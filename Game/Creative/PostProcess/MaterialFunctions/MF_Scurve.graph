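# Material function that remaps a scalar Gradient through an S-curve with adjustable Contrast, stored among Creative post-process functions.
# Contrast is clamped just below one and converted into a reciprocal exponent that steepens as contrast rises.
# Two mirrored power curves shape the lower and upper halves of the gradient.
# A Sign test around the 0.5 midpoint selects between them arithmetically, without a branch, and the result is rescaled.
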MaterialFunction /Game/Creative/PostProcess/MaterialFunctions/MF_Scurve
graph {nodes [
  n1 = Multiply(A=NamedRerouteUsage(), B=Subtract(B=NamedRerouteUsage()) {constA=2})
  n2 = Multiply(A=NamedRerouteUsage(), B=NamedRerouteUsage())
  FunctionOutput(None=Multiply(A=Add(A=@n1, B=@n2)) {constB=0.25})
  NamedRerouteDeclaration(None=FunctionInput() {inputName=Gradient, inputType=FunctionInput_Scalar, sortPriority=1}) {name=x, nodeColor=<rgba 0.7344 0 1 1>}
  n3 = Clamp(None=FunctionInput() {bUsePreviewValueAsDefault=true, inputName=Contrast, inputType=FunctionInput_Scalar, sortPriority=2}) {maxDefault=0.99}
  NamedRerouteDeclaration(None=Divide(B=OneMinus(None=@n3))) {name=e, nodeColor=<rgba 0 1 0.4141 1>}
  n4 = NamedRerouteUsage()
  NamedRerouteDeclaration(None=Power(Base=Add(A=@n4, B=@n4), Exp=NamedRerouteUsage())) {name=a, nodeColor=<rgba 1 0.0938 0 1>}
  n5 = NamedRerouteUsage()
  n6 = Subtract(B=Add(A=@n5, B=@n5)) {constA=2}
  n7 = NamedRerouteUsage()
  NamedRerouteDeclaration(None=Subtract(B=Power(Base=@n6, Exp=@n7)) {constA=2}) {name=b, nodeColor=<rgba 0 0.2266 1 1>}
  n8 = Subtract(A=NamedRerouteUsage()) {constB=0.5}
  NamedRerouteDeclaration(None=Add(A=Sign(None=@n8))) {name=c, nodeColor=<rgba 0.5469 1 0 1>}
 ]}
