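# Post-process cel shader that runs before depth of field and posterizes the scene into tonal bands.
# It quantizes log2 of a desaturated input into steps, either rounded or S-curve softened via a static switch, then converts back with a power of two.
# Global Strength and Master FX Control blend the effect over the scene.
# A depth-based mask excludes the sky, except on mobile, which also has its own step count.
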Material /Game/Creative/PostProcess/Materials/M_DV_PP_CelShader
graph {
 properties {bAutomaticallySetUsageInEditor=false, blendableLocation=BL_SceneColorBeforeDOF, materialDomain=MD_PostProcess, usedShadingModels=Default Lit}
 nodes [
  n1 = ShadingPathSwitch(
   Default=ScalarParameter() {defaultValue=1, parameterName=Steps}
   Mobile=ScalarParameter() {defaultValue=1.25, parameterName="Steps (mobile)"}
  ) {desc=Migrated from FeatureLevelSwitch}
  n2 = Desaturation(None=NamedRerouteUsage())
  n3 = Multiply(A=@n1, B=Logarithm2!(X=@n2))
  n4 = Frac(None=@n3)
  n5 = ScalarParameter() {defaultValue=0.5, parameterName=Contrast}
  n6 = Divide(
   A=Add(
    A=Floor(None=@n3)
    B=MaterialFunctionCall!(Gradient=@n4, Contrast=@n5) {materialFunction=/Game/Creative/PostProcess/MaterialFunctions/MF_Scurve}.Result
   )
   B=@n1
  )
  n7 = Logarithm2(X=@n2)
  n8 = Divide(A=Round(None=Multiply(A=@n7, B=@n1)), B=@n1)
  n9 = LinearInterpolate(
   A=ComponentMask(None=NamedRerouteUsage()) {b=true, g=true, r=true}
   B=Power(Base=Constant() {r=2}, Exp=StaticSwitchParameter(True=@n6, False=@n8) {parameterName=Scurve})
   Alpha=ScalarParameter() {defaultValue=1, parameterName=Global Strength}
  )
  n10 = ComponentMask(None=NamedRerouteUsage()) {b=true, g=true, r=true}
  n11 = NamedRerouteUsage()
  LinearInterpolate(
   A=ComponentMask(None=SceneTexture() {sceneTextureId=PPI_PostProcessInput0}.Color) {b=true, g=true, r=true}
   B=ShadingPathSwitch(Default=LinearInterpolate(A=@n9, B=@n10, Alpha=@n11) {desc=Mask out sky}, Mobile=@n9) {desc=Migrated from FeatureLevelSwitch}
   Alpha=ScalarParameter() {parameterName=Master FX Control, sortPriority=0}
  )
  n12 = ComponentMask(None=NamedRerouteUsage()) {b=true, g=true, r=true}
  Max(A=ComponentMask(None=@n12) {r=true}, B=Max(A=ComponentMask(None=@n12) {g=true}, B=ComponentMask(None=@n12) {b=true}))
  NamedRerouteDeclaration(None=SceneTexture() {bFiltered=true, sceneTextureId=PPI_PostProcessInput0}.Color) {name=PP0, nodeColor=<rgba 0.5469 1 0 1>}
  NamedRerouteDeclaration(None=MaterialFunctionCall!(
    Input=SceneTexture() {sceneTextureId=PPI_SceneDepth}.Color
    Input Low=ScalarParameter() {defaultValue=25000, parameterName=Falloff Near Point}
    Input High=ScalarParameter() {defaultValue=30000, parameterName=Falloff Far Point}
   ) {materialFunction=/Engine/Functions/Engine_MaterialFunctions03/Math/RemapValueRangeNormalized}.Result) {name=Skybox Mask, nodeColor=<rgba 1 0.9141 0 1>}
 ]
}
