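# Material function that projects a texture in world space for floors and walls, with adjustable size, offset, rotation and blend contrast.
# Offset world position over negated absolute size forms XZ, YZ and XY UVs, each rotated by CustomRotator and sampled with the shared wrap sampler.
# Absolute world normal X blends the wall samples, then Z blends in the floor, both through CheapContrast.
# Outputs wall, floor and combined color and alpha.
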MaterialFunction /Game/Creative/PostProcess/MaterialFunctions/MF_WorldAlignedTexRot
graph {nodes [
  Constant() {r=1}
  n1 = WorldPosition()
  n2 = FunctionInput() {
   bUsePreviewValueAsDefault=true
   description=Offsets the texture XYZ
   inputName=Texture Offset
   inputType=FunctionInput_Vector2
   sortPriority=5
   previewValue=<20 20 20 6>
  }
  n3 = Add(A=ComponentMask(None=@n1.XYZ) {b=true, r=true}, B=@n2) {constB=0}
  n4 = FunctionInput() {
   bUsePreviewValueAsDefault=true
   description=Size of texture in worldspace units
   inputName=TextureSize1
   inputType=FunctionInput_Vector2
   sortPriority=1
   previewValue=<256 256 256 6>
  }
  n5 = Multiply(A=Abs(None=@n4)) {constB=-1}
  n6 = FunctionInput(Preview=Constant()) {
   bUsePreviewValueAsDefault=true
   description="amount of rotation wanted (0 - 360)"
   inputName=Rotation Degrees
   inputType=FunctionInput_Scalar
   sortPriority=4
   previewValue=<0 0 0 0>
  }
  n7 = FunctionInput() {
   description="Put a TextureObject into this. (right click on a texture sample to convert, for a parameter version you will need to find it in the palette)"
   inputName=TextureObject
   inputType=FunctionInput_Texture2D
  }
  n8 = TextureSample(
   UVs=MaterialFunctionCall(UVs=Divide(A=@n3, B=@n5) {constB=360}, "Rotation Angle (0-1)"=Divide(A=@n6) {constB=360}) {materialFunction=&CustomRotator}.Rotated Values
   Tex=@n7
  ) {samplerSource=SSM_Wrap_WorldGroupSettings}
  n9 = Add(A=ComponentMask(None=@n1.XYZ) {b=true, g=true}, B=@n2) {constB=0}
  n10 = TextureSample(
   UVs=MaterialFunctionCall(UVs=Divide(A=@n9, B=@n5) {constB=360}, "Rotation Angle (0-1)"=Divide(A=@n6) {constB=360}) {materialFunction=&CustomRotator}.Rotated Values
   Tex=@n7
  ) {samplerSource=SSM_Wrap_WorldGroupSettings}
  n11 = FunctionInput(Preview=VertexNormalWS()) {bUsePreviewValueAsDefault=true, inputName=World Space Normal}
  n12 = ComponentMask(None=@n11) {r=true}
  n13 = FunctionInput() {bUsePreviewValueAsDefault=true, inputName=Contrast, inputType=FunctionInput_Scalar, previewValue=<1 0 0 1>}
  n14 = ComponentMask(None=MaterialFunctionCall(In=Abs(None=@n12), Contrast=@n13) {materialFunction=&CheapContrast}.Result) {r=true}
  n15 = LinearInterpolate(A=@n8.RGB, B=@n10.RGB, Alpha=@n14) {desc=Float3}
  n16 = Add(A=ComponentMask(None=@n1.XYZ) {g=true, r=true}, B=@n2) {constB=0}
  n17 = TextureSample(
   UVs=MaterialFunctionCall(UVs=Divide(A=@n16, B=@n5) {constB=360}, "Rotation Angle (0-1)"=Divide(A=@n6) {constB=360}) {materialFunction=&CustomRotator}.Rotated Values
   Tex=@n7
  ) {samplerSource=SSM_Wrap_WorldGroupSettings}
  n18 = ComponentMask(None=@n11) {b=true}
  n19 = ComponentMask(None=MaterialFunctionCall(In=Abs(None=@n18), Contrast=@n13) {materialFunction=&CheapContrast}.Result) {r=true}
  FunctionOutput(None=LinearInterpolate(A=@n15, B=@n17.RGB, Alpha=@n19)) {description=Use this for floor and walls., outputName=XYZ Texture, sortPriority=3}
  FunctionOutput(None=@n17.RGB) {description=Use this for floors, outputName=Z Texture, sortPriority=1}
  FunctionOutput(None=@n15) {description=Use this for walls, outputName=XY Texture, sortPriority=2}
  n20 = LinearInterpolate(A=@n8.A, B=@n10.A, Alpha=@n14) {desc=Float3}
  FunctionOutput(None=LinearInterpolate(A=@n20, B=@n17.A, Alpha=@n19)) {description=Use this for floor and walls., outputName=XYZ Texture Alpha, sortPriority=6}
  FunctionOutput(None=@n17.A) {description=Use this for floors, outputName="Z Texture Alpha ", sortPriority=4}
  FunctionOutput(None=@n20) {description=Use this for walls, outputName=XY Texture Alpha, sortPriority=5}
 ]}
