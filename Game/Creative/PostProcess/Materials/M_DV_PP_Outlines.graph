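# Post-process material, applied before depth of field, that draws outlines and AO-driven hatching over the scene.
# Four-neighbour scene-depth and world-normal differences form edge masks, combined by Max; mobile takes one mask and skips hatching.
# World-aligned rotated hatching, masked by ambient occlusion and a custom-depth test to exclude the player, multiplies the result.
# Lines fade beyond a depth cutoff, a Debug switch isolates them, and Master FX Control blends the effect.
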Material /Game/Creative/PostProcess/Materials/M_DV_PP_Outlines
graph {
 properties {blendableLocation=BL_SceneColorBeforeDOF, materialDomain=MD_PostProcess, pixelDepthOffsetMode=PDOM_Legacy, usedShadingModels=Default Lit}
 nodes [
  n1 = SceneTexture() {sceneTextureId=PPI_PostProcessInput0}
  n2 = Constant()
  n3 = NamedRerouteUsage()
  n4 = Max(
   A=StaticSwitchParameter(True=NamedRerouteUsage(), False=@n2) {defaultValue=true, parameterName=WN Lines}
   B=StaticSwitchParameter(True=@n3, False=@n2) {defaultValue=true, parameterName=SD Lines}
  )
  n5 = OneMinus(None=LinearInterpolate(Alpha=ShadingPathSwitch(Default=@n4, Mobile=@n3) {desc=Migrated from FeatureLevelSwitch}) {desc="Pre TM Adjust (if needed)"})
  n6 = StaticSwitchParameter(
   True=@n5
   False=LinearInterpolate(A=Constant3Vector() {desc=OutlineColour}, B=ComponentMask(None=@n1.Color) {b=true, g=true, r=true}, Alpha=@n5)
  ) {parameterName=Debug}
  n7 = MaterialFunctionCall!(
   ThresholdValue=ScalarParameter() {defaultValue=100000, group=Ranges, parameterName=Range Cutoff}
   ThresholdRange=Constant() {r=500}
   Alpha=SceneDepth().R
   TwoSidedThreshold
  ) {materialFunction=/Engine/Functions/Engine_MaterialFunctions02/Masking/ThresholdWithRange}
  n8 = LinearInterpolate(A=@n6, Alpha=@n7.Result)
  n9 = LinearInterpolate(A=@n6, B=ComponentMask(None=SceneTexture() {sceneTextureId=PPI_PostProcessInput0}.Color) {b=true, g=true, r=true}, Alpha=@n7.Result)
  LinearInterpolate(
   A=ComponentMask(None=SceneTexture() {sceneTextureId=PPI_PostProcessInput0}.Color) {b=true, g=true, r=true}
   B=Multiply(A=NamedRerouteUsage(), B=StaticSwitchParameter(True=@n8, False=@n9) {parameterName=Debug})
   Alpha=ScalarParameter() {parameterName=Master FX Control, sortPriority=0}
  )
  n10 = TextureObject() {texture=/Game/Creative/PostProcess/Textures/T_HatchingA}
  n11 = ScalarParameter() {defaultValue=1, parameterName=TriP Contrast}
  n12 = SceneTexture() {sceneTextureId=PPI_WorldNormal}
  n13 = ScalarParameter() {defaultValue=50, parameterName=TriP Size}
  n14 = Constant() {r=85}
  n15 = SceneTexture() {sceneTextureId=PPI_AmbientOcclusion}
  n16 = OneMinus(None=Power(Base=ComponentMask(None=@n15.Color) {r=true}, Exp=ScalarParameter() {defaultValue=2, parameterName=AO Power}) {constExponent=0.3})
  n17 = ScalarParameter() {defaultValue=0.2, parameterName=AO Contrast}
  n18 = LinearInterpolate(
   B=LinearInterpolate(
    B=MaterialFunctionCall!(TextureObject=@n10, Contrast=@n11, World Space Normal=@n12.Color, TextureSize1=@n13, Rotation Degrees=@n14, Texture Offset) {materialFunction=/Game/Creative/PostProcess/MaterialFunctions/MF_WorldAlignedTexRot}.XYZ Texture
    Alpha=MaterialFunctionCall(In=@n16, Contrast=@n17) {materialFunction=&CheapContrast}.Result
   ) {constA=1}
   Alpha=ScalarParameter() {defaultValue=1, parameterName=AO Hatching}
  ) {constA=1}
  n19 = Constant() {r=1}
  n20 = ComponentMask(None=SceneTexture() {sceneTextureId=PPI_SceneDepth}.Color) {r=true}
  n21 = ComponentMask(None=SceneTexture() {sceneTextureId=PPI_CustomDepth}.Color) {r=true}
  n22 = Constant()
  n23 = Constant() {r=1}
  NamedRerouteDeclaration(None=LinearInterpolate(
    B=ShadingPathSwitch(Default=@n18, Mobile=@n19) {desc=Migrated from FeatureLevelSwitch}
    Alpha=If(A=@n20, B=@n21, "A > B"=@n22, "A == B"=@n22, "A < B"=@n23) {desc=Mask Player out}
   ) {constA=1}) {name=AO Hatching, nodeColor=<rgba 1 0.1172 0 1>}
  n24 = SceneDepth()
  n25 = Multiply(A=SceneTexture() {sceneTextureId=PPI_SceneDepth}.InvSize, B=Constant2Vector() {g=1})
  n26 = Constant() {r=1}
  n27 = SceneDepth(UVs=Add(A=TextureCoordinate(), B=Multiply(A=@n25, B=@n26)))
  n28 = SceneDepth()
  n29 = Multiply(A=SceneTexture() {sceneTextureId=PPI_SceneDepth}.InvSize, B=Constant2Vector() {g=-1})
  n30 = SceneDepth(UVs=Add(A=TextureCoordinate(), B=Multiply(A=@n29, B=@n26)))
  n31 = SceneDepth()
  n32 = Multiply(A=SceneTexture() {sceneTextureId=PPI_SceneDepth}.InvSize, B=Constant2Vector() {r=1})
  n33 = SceneDepth(UVs=Add(A=TextureCoordinate(), B=Multiply(A=@n32, B=@n26)))
  n34 = SceneDepth()
  n35 = Multiply(A=SceneTexture() {sceneTextureId=PPI_SceneDepth}.InvSize, B=Constant2Vector() {r=-1})
  n36 = SceneDepth(UVs=Add(A=TextureCoordinate(), B=Multiply(A=@n35, B=@n26)))
  n37 = Add(A=Add(A=Subtract(A=@n24.R, B=@n27.R), B=Subtract(A=@n28.R, B=@n30.R)), B=Add(A=Subtract(A=@n31.R, B=@n33.R), B=Subtract(A=@n34.R, B=@n36.R)))
  n38 = ShadingPathSwitch(
   Default=ScalarParameter() {defaultValue=1, parameterName=SD Threshold}
   Mobile=ScalarParameter() {defaultValue=1, parameterName=SD Threshold Mobile}
  ) {desc=Migrated from FeatureLevelSwitch}
  NamedRerouteDeclaration(None=Saturate(None=Divide(A=@n37, B=@n38))) {name=SD Lines, nodeColor=<rgba 0.7812 0 1 1>}
  n39 = TextureCoordinate()
  n40 = ScalarParameter() {defaultValue=1, parameterName=WN Thickness}
  n41 = Multiply(A=Multiply(A=SceneTexture() {sceneTextureId=PPI_SceneDepth}.InvSize, B=Constant2Vector() {g=1}), B=@n40)
  n42 = Subtract(A=SceneTexture() {sceneTextureId=PPI_WorldNormal}.Color, B=SceneTexture(UVs=Add(A=@n39, B=@n41)) {sceneTextureId=PPI_WorldNormal}.Color)
  n43 = SceneTexture() {sceneTextureId=PPI_WorldNormal}
  n44 = Multiply(A=SceneTexture() {sceneTextureId=PPI_SceneDepth}.InvSize, B=Constant2Vector() {g=-1})
  n45 = SceneTexture(UVs=Add(A=TextureCoordinate(), B=Multiply(A=@n44, B=@n40))) {sceneTextureId=PPI_WorldNormal}
  n46 = SceneTexture() {sceneTextureId=PPI_WorldNormal}
  n47 = Multiply(A=SceneTexture() {sceneTextureId=PPI_SceneDepth}.InvSize, B=Constant2Vector() {r=1})
  n48 = SceneTexture(UVs=Add(A=TextureCoordinate(), B=Multiply(A=@n47, B=@n40))) {sceneTextureId=PPI_WorldNormal}
  n49 = SceneTexture() {sceneTextureId=PPI_WorldNormal}
  n50 = Multiply(A=SceneTexture() {sceneTextureId=PPI_SceneDepth}.InvSize, B=Constant2Vector() {r=-1})
  n51 = SceneTexture(UVs=Add(A=TextureCoordinate(), B=Multiply(A=@n50, B=@n40))) {sceneTextureId=PPI_WorldNormal}
  n52 = Add(A=Add(A=@n42, B=Subtract(A=@n43.Color, B=@n45.Color)), B=Add(A=Subtract(A=@n46.Color, B=@n48.Color), B=Subtract(A=@n49.Color, B=@n51.Color)))
  n53 = Multiply(
   A=Desaturation(None=StaticSwitchParameter(True=@n42, False=@n52) {parameterName=WN Simplify})
   B=ScalarParameter() {defaultValue=1, parameterName=WN Intensity}
  )
  n54 = ScalarParameter() {defaultValue=1, parameterName=WN Threshold}
  NamedRerouteDeclaration(None=Saturate(None=Power(Base=@n53, Exp=@n54) {constExponent=1, desc=Line Detail})) {name=WN Lines, nodeColor=<rgba 0 1 0.4609 1>}
 ]
}
