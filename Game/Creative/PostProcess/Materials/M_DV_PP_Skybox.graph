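# M_DV_PP_Skybox is a Creative post-process material, applied after DOF, that paints a stylized sky and depth fog, faded by Master FX Control.
# World-direction gradients form sky, horizon and ground; panned, projected noise forms clouds with offset-sample shading and sun-direction rim light.
# A scene-depth mask bounds the skybox, a curve atlas colors fog, and time of day darkens the effects.
# Mobile skips sky and fog, brightening its input instead.
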Material /Game/Creative/PostProcess/Materials/M_DV_PP_Skybox
graph {
 properties {bAutomaticallySetUsageInEditor=false, blendableLocation=BL_SceneColorAfterDOF, materialDomain=MD_PostProcess, usedShadingModels=Default Lit}
 nodes [
  n1 = NamedRerouteUsage()
  n2 = NamedRerouteUsage()
  n3 = NamedRerouteUsage()
  n4 = ComponentMask(None=NamedRerouteUsage() {desc=No Custom Sky or Fog on Mobile}) {b=true, g=true, r=true}
  LinearInterpolate(
   A=ComponentMask(None=SceneTexture() {sceneTextureId=PPI_PostProcessInput0}.Color) {b=true, g=true, r=true}
   B=ShadingPathSwitch(Default=LinearInterpolate(A=@n1, B=@n2, Alpha=@n3), Mobile=Multiply(A=@n4) {constB=1.15, desc=Add Brightness to account for no GI or Fog}) {desc=Migrated from FeatureLevelSwitch}
   Alpha=ScalarParameter() {parameterName=Master FX Control, sortPriority=0}
  )
  n5 = NamedRerouteUsage()
  n6 = Power(Base=Add(A=@n5, B=Constant() {r=0.5}), Exp=Constant() {r=12})
  n7 = Multiply(
   A=LinearInterpolate(A=NamedRerouteUsage(), B=NamedRerouteUsage(), Alpha=NamedRerouteUsage())
   B=ScalarParameter() {defaultValue=1, parameterName=Global Brightness}
  )
  Multiply(A=Step(Y=Subtract(A=@n5, B=@n6), X=ScalarParameter() {defaultValue=0.1991, group=Clouds, parameterName=Cloud Outline Thickness}), B=@n7) {desc=Not used}
  n8 = VectorParameter() {parameterName=Colour Top Day, defaultValue=<rgba 0 0.0353 0.0417 1>}
  n9 = WorldPosition()
  n10 = ComponentMask(None=Normalize(VectorInput=@n9.XYZ)) {b=true}
  n11 = Multiply(A=Power(Base=OneMinus(None=@n10), Exp=ScalarParameter() {defaultValue=6.6218, parameterName=SkyBlend}), B=Constant() {r=1})
  n12 = Normalize(VectorInput=@n9.XYZ)
  n13 = ScalarParameter() {defaultValue=0.048, desc=Ground Height, parameterName=Ground Height}
  n14 = OneMinus(None=ComponentMask(None=Add(A=@n12, B=@n13) {constB=1.1821}) {b=true})
  n15 = ScalarParameter() {defaultValue=904.3628, parameterName=Ground}
  n16 = LinearInterpolate(
   A=LinearInterpolate(A=@n8.RGB, B=VectorParameter() {parameterName=Colour Bottom, defaultValue=<rgba 0.0495 0.2969 0.1988 1>}.RGB, Alpha=Saturate(None=@n11))
   B=VectorParameter() {parameterName=Ground, defaultValue=<rgba 0 0 0 1>}.RGB
   Alpha=Saturate(None=Power(Base=@n14, Exp=@n15))
  )
  n17 = Multiply(
   A=VectorParameter() {parameterName=Horizon Colour, defaultValue=<rgba 0.4217 0.526 0.244 1>}.RGB
   B=ScalarParameter() {defaultValue=1, parameterName=Horizon Power}
  )
  n18 = Add(A=@n13, B=Normalize(VectorInput=@n9.XYZ))
  n19 = OneMinus(None=ComponentMask(None=Abs(None=@n18)) {b=true})
  n20 = ScalarParameter() {defaultValue=15, parameterName=Horizon}
  n21 = Saturate(None=Multiply(A=Power(Base=@n19, Exp=@n20), B=Constant() {r=1}))
  n22 = Normalize(VectorInput=WorldPosition().XYZ)
  n23 = Sign(None=ComponentMask(None=Add(A=@n13, B=@n22)) {b=true})
  n24 = StaticSwitchParameter(True=LinearInterpolate(B=@n21, Alpha=Saturate(None=@n23)) {constB=0}, False=@n21) {parameterName=Hard Horizon}
  n25 = VectorParameter() {parameterName=Colour Top Night, defaultValue=<rgba 0.125 0.125 0.125 1>}
  n26 = MaterialFunctionCall() {materialFunction=&Scalar_Time_of_Day}
  NamedRerouteDeclaration(None=StaticSwitchParameter(True=LinearInterpolate(A=@n8.RGB, B=@n25.RGB, Alpha=@n26.Result), False=LinearInterpolate(A=@n16, B=@n17, Alpha=@n24)) {parameterName=Colour Top Only}) {name=Sky Gradients, nodeColor=<rgba 0 0.5781 1 1>}
  NamedRerouteDeclaration(None=SceneTexture() {sceneTextureId=PPI_SceneDepth}.Color) {name=SceneDepth, nodeColor=<rgba 1 0.0469 0 1>}
  NamedRerouteDeclaration(None=@n7) {name=Sky Output, nodeColor=<rgba 1 0.0469 0 1>}
  n27 = SceneTexture() {sceneTextureId=PPI_PostProcessInput1}
  n28 = CurveAtlasRowParameter(CurveTime=MaterialFunctionCall!(
    Input=NamedRerouteUsage()
    Input Low=ScalarParameter() {defaultValue=500, group=Fog, parameterName=Curve Near Point}
    Input High=ScalarParameter() {defaultValue=100000, group=Fog, parameterName=Curve Far Point}
   ) {materialFunction=/Engine/Functions/Engine_MaterialFunctions03/Math/RemapValueRangeNormalized}.Result) {
   atlas=/Game/Creative/PostProcess/ColorCurve/A_PostProcess_Atlas
   curve=/Game/Creative/PostProcess/ColorCurve/C_DV_Fog
   defaultValue=6
   group=Fog
   parameterName=CurveInput
  }
  n29 = LinearInterpolate(
   A=ComponentMask(None=NamedRerouteUsage()) {b=true, g=true, r=true}
   B=LinearInterpolate(
    A=@n28
    B=Constant() {r=-0.3}
    Alpha=MaterialFunctionCall() {materialFunction=&Scalar_Time_of_Day}.Result
   )
   Alpha=@n28.A
  )
  n30 = ComponentMask(None=NamedRerouteUsage()) {b=true, g=true, r=true}
  n31 = MaterialFunctionCall!(
   ThresholdValue=ScalarParameter() {defaultValue=100000, group=Fog, parameterName=Range Cutoff}
   ThresholdRange=Constant() {r=500}
   Alpha=NamedRerouteUsage()
   TwoSidedThreshold
  ) {materialFunction=/Engine/Functions/Engine_MaterialFunctions02/Masking/ThresholdWithRange}
  NamedRerouteDeclaration(None=LinearInterpolate(
    A=ComponentMask(None=@n27.Color) {b=true, g=true, r=true}
    B=LinearInterpolate(A=@n29, B=@n30, Alpha=@n31.Result)
    Alpha=ScalarParameter() {defaultValue=1, group=Fog, parameterName=Fog Global Amount, sliderMax=1}
   )) {name=Fog, nodeColor=<rgba 0.9453 1 0 1>}
  NamedRerouteDeclaration(None=SceneTexture() {sceneTextureId=PPI_PostProcessInput0}.Color) {name=PP0, nodeColor=<rgba 0.9219 1 0 1>}
  NamedRerouteDeclaration(None=MaterialFunctionCall(
    Input=NamedRerouteUsage()
    Input Low=ScalarParameter() {defaultValue=35000, parameterName=Falloff Near Point}
    Input High=ScalarParameter() {defaultValue=40000, parameterName=Falloff Far Point}
   ) {materialFunction=&RemapValueRangeNormalized}.Result) {name=Skybox Mask, nodeColor=<rgba 1 0.9141 0 1>}
  n32 = MaterialFunctionCall(Float3=Normalize(VectorInput=WorldPosition().XYZ)) {materialFunction=&BreakOutFloat3Components}
  n33 = AppendVector(A=@n32.R, B=@n32.R)
  n34 = Power(Base=@n32.B, Exp=ScalarParameter() {defaultValue=0.5154, parameterName=Sky Curvature})
  n35 = ScalarParameter() {defaultValue=0.57, group=Clouds, parameterName=Cloud Direction A}
  n36 = ScalarParameter() {defaultValue=-0.16, group=Clouds, parameterName=Cloud Direction B}
  n37 = Panner(
   Coordinate=Multiply(A=ScalarParameter() {defaultValue=0.1154, group=Clouds, parameterName=Cloud Height}, B=Divide(A=@n33, B=@n34))
   Time=Time()
   Speed=Multiply(A=AppendVector(A=@n35, B=@n36)) {constB=0.01}
  ) {speedX=-0.01}
  n38 = TextureSample(UVs=@n37) {samplerType=SAMPLERTYPE_LinearColor, texture=/Game/Creative/PostProcess/Textures/T_Noise_CloudsB}
  n39 = ScalarParameter() {defaultValue=2.2301, group=Clouds, parameterName=Cloud Coverage}
  n40 = Saturate(None=Power(Base=Multiply(A=@n38.R, B=@n39), Exp=ScalarParameter() {defaultValue=300, group=Clouds, parameterName=Cloud Edge Hardness}))
  n41 = ScalarParameter() {defaultValue=0.048, group=Clouds, parameterName=Cloud Ground Height}
  n42 = Normalize(VectorInput=WorldPosition().XYZ)
  n43 = ComponentMask(None=Abs(None=Add(A=@n41, B=@n42))) {b=true}
  n44 = Multiply(
   A=Power(Base=OneMinus(None=@n43), Exp=ScalarParameter() {defaultValue=15, group=Clouds, parameterName=Cloud Horizon})
   B=ScalarParameter() {defaultValue=0.5351, group=Clouds, parameterName=Cloud Horizon Falloff}
  )
  n45 = WorldPosition()
  n46 = ComponentMask(None=Add(B=Normalize(VectorInput=@n45.XYZ))) {b=true}
  n47 = LinearInterpolate(B=OneMinus(None=Saturate(None=@n44)), Alpha=Saturate(None=Sign(None=@n46))) {constB=0}
  NamedRerouteDeclaration(None=LinearInterpolate(B=Multiply(A=@n40, B=@n47), Alpha=ScalarParameter() {defaultValue=1, group=Clouds, parameterName=Cloud Opacity})) {name=Cloud Mask, nodeColor=<rgba 1 0.2812 0 1>}
  n48 = Normalize(VectorInput=Multiply(A=SkyAtmosphereLightDirection() {lightIndex=1}, B=Constant() {r=-1}))
  n49 = CameraVectorWS()
  n50 = Constant()
  n51 = ScalarParameter() {defaultValue=0.5186, group=Clouds, parameterName=Cloud Rim A}
  n52 = NamedRerouteUsage()
  n53 = ScalarParameter() {defaultValue=0.128, group=Clouds, parameterName=Cloud Rim B}
  n54 = Constant() {r=1}
  n55 = Multiply(
   A=Power(Base=DotProduct(A=@n48, B=@n49), Exp=ScalarParameter() {defaultValue=38.4452, group=Clouds, parameterName=Cloud Rim Exp})
   B=Subtract(A=SmoothStep(Min=@n50, Max=@n51, Value=@n52), B=SmoothStep(Min=@n53, Max=@n54, Value=@n52))
  )
  n56 = VectorParameter() {group=Clouds, parameterName=Cloud Rim Tint, defaultValue=<rgba 1 0.8466 0.6125 1>}
  n57 = ScalarParameter() {defaultValue=0.65, group=Clouds, parameterName=Cloud Shading edge Pos}
  n58 = ScalarParameter() {defaultValue=0.01, group=Clouds, parameterName=Cloud Shading edge Sharpness}
  n59 = TextureSample(UVs=Add(A=@n37, B=Constant() {desc=Offset cloud shadow, r=0.01})) {samplerType=SAMPLERTYPE_LinearColor, texture=&T_Noise_CloudsB}
  n60 = Multiply(A=ScalarParameter() {defaultValue=0.596, desc=Percent, group=Clouds, parameterName=Cloud Shading Percent}, B=@n39)
  n61 = LinearInterpolate(
   A=VectorParameter() {group=Clouds, parameterName=Cloud Light, defaultValue=<rgba 0.0377 0.1386 0.1458 1>}.RGB
   B=VectorParameter() {group=Clouds, parameterName=Cloud Dark, defaultValue=<rgba 0 0 0 1>}.RGB
   Alpha=SmoothStep(Min=Subtract(A=@n57, B=@n58), Max=@n57, Value=Multiply(A=@n59.R, B=@n60))
  )
  n62 = Multiply(A=@n61) {constB=0.1}
  n63 = MaterialFunctionCall() {materialFunction=&Scalar_Time_of_Day}
  NamedRerouteDeclaration(None=Add(A=Multiply(A=@n55, B=@n56.RGB), B=LinearInterpolate(A=@n61, B=@n62, Alpha=@n63.Result))) {name=Cloud Colour, nodeColor=<rgba 0.7578 0 1 1>}
  NamedRerouteDeclaration() {name=Debug, nodeColor=<rgba 0.8281 0 1 1>}
 ]
}
