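# Post-process material that restyles the scene as a faded 1970s magazine print, blended over the original by Master FX Control.
# The scene gets a contrast power, desaturation and warm tint, then lerps toward textured paper color using print-line and edge masks.
# Print lines combine repeating screen-Y bands with a rotatable static texture whose channels merge through SmoothMax.
# Film grain, a vertical gradient and power-based screen-border masks are shared through named reroutes.
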Material /Game/Creative/PostProcess/Materials/M_PP_70sMagazine_Print
graph {
 properties {materialDomain=MD_PostProcess, pixelDepthOffsetMode=PDOM_Legacy, usedShadingModels=Default Lit}
 nodes [
  n1 = SceneTexture() {sceneTextureId=PPI_PostProcessInput0}
  n2 = ScalarParameter() {defaultValue=1, group=Print, parameterName=ValueRange}
  n3 = Multiply(
   A=VectorParameter() {group=Print, parameterName=Print Tint, defaultValue=<rgba 1 0.8688 0.5208 1>}.RGB
   B=Desaturation(
    None=Power(Base=@n1.Color, Exp=@n2) {constExponent=1.5}
    Fraction=ScalarParameter() {defaultValue=0.45, group=Print, parameterName=Desaturation}
   )
  )
  n4 = NamedRerouteUsage()
  n5 = Multiply(A=VectorParameter() {group=Print, parameterName=Paper Color, defaultValue=<rgba 1 0.857 0.5938 1>}.RGB, B=NamedRerouteUsage())
  n6 = TextureSample(UVs=Multiply(A=ScreenPosition().ViewportUV, B=ScalarParameter() {defaultValue=1.5, group=Grain, parameterName=NoiseTiling})) {texture=/Game/Packages/Fortress_Transylvania/Haunted_Forest/Swamps/Textures/2ndInfectedMapLinear2}
  n7 = Multiply(A=NamedRerouteUsage(), B=ScalarParameter() {defaultValue=1, group=X. Print Lines, parameterName=Print Lines Opacity})
  n8 = NamedRerouteUsage()
  n9 = Multiply(A=@n8, B=ScalarParameter() {defaultValue=0.25, group=Print, parameterName=Edge Opacity})
  LinearInterpolate(
   A=ComponentMask(None=SceneTexture() {sceneTextureId=PPI_PostProcessInput0}.Color) {b=true, g=true, r=true}
   B=LinearInterpolate(A=Multiply(A=@n3, B=@n4), B=Multiply(A=@n5, B=@n6.RGB), Alpha=Add(A=@n7, B=@n9))
   Alpha=ScalarParameter() {group=Main, parameterName=Master FX Control, sortPriority=0}
  )
  n10 = ScreenPosition()
  n11 = ComponentMask(None=@n10.ViewportUV) {g=true}
  NamedRerouteDeclaration(None=Power(Base=@n11, Exp=ScalarParameter() {defaultValue=2, group=X. Print Lines, parameterName=Gradiant Mask Power}) {constExponent=3}) {name=Gradiant Mask, nodeColor=<rgba 1 0.2109 0 1>}
  n12 = TextureSample(UVs=Multiply(A=ScreenPosition().ViewportUV, B=ScalarParameter() {defaultValue=20, group=Grain, parameterName=GrainSize}) {constB=20}) {
   samplerSource=SSM_Wrap_WorldGroupSettings
   samplerType=SAMPLERTYPE_LinearColor
   texture=/Engine/EngineResources/FilmGrains/Marcie_Grain_v3_128_M2_000
  }
  NamedRerouteDeclaration(None=Multiply(A=Add(A=@n12.RGB) {constB=0.5}, B=ScalarParameter() {defaultValue=1, group=Grain, parameterName=GrainIntensity, sliderMax=1})) {name=Grain Noise, nodeColor=<rgba 0.8984 1 0 1>}
  n13 = Frac(None=Multiply(A=@n11, B=ScalarParameter() {defaultValue=100, group=Print, parameterName=Print Line Amount}))
  n14 = Multiply(
   A=@n10.ViewportUV
   B=ComponentMask(None=VectorParameter() {group=X. Print Lines, parameterName=UVScale, defaultValue=<rgba 3 3 0 1>}.RGB) {g=true, r=true}
  )
  n15 = ComponentMask(None=VectorParameter() {group=X. Print Lines, parameterName=UVOffset, defaultValue=<rgba -0.5 0 0 1>}.RGB) {g=true, r=true}
  n16 = ChannelMaskParameter(None=AppendVector(A=Constant() {r=6.2832}, B=Constant())) {group=X. Print Lines, maskChannel=Alpha, parameterName=Print Line Direction, defaultValue=<rgba 0 0 0 1>}
  n17 = TextureSample(UVs=Rotator(Coordinate=Add(A=@n14, B=@n15), Time=@n16.R)) {samplerType=SAMPLERTYPE_Masks, texture=/Game/Missions/Primary/RepairTheShelter/Textures/Cinematic_Screen_Static}
  n18 = ScalarParameter() {defaultValue=0.5, group=X. Print Lines, parameterName=Print Error Scale}
  n19 = MaterialFunctionCall!(
   A=MaterialFunctionCall(A=@n17.R, B=@n17.R, distance=@n18) {materialFunction=/Engine/Functions/Substrate/MF_SmoothMax1}.SmoothMax
   B=@n17.B
   distance=@n18
  ) {materialFunction=&MF_SmoothMax1}
  n20 = Saturate(None=Power(Base=@n13))
  n21 = NamedRerouteUsage()
  NamedRerouteDeclaration(None=Multiply(A=Multiply(A=@n13, B=@n19.SmoothMax), B=Multiply(A=@n20, B=@n21))) {name=Print Lines, nodeColor=<rgba 0.0781 0 1 1>}
  n22 = ScreenPosition()
  n23 = ScalarParameter() {defaultValue=8, group=Grain, parameterName=Edge Mask Intensity}
  n24 = ComponentMask(None=@n22.ViewportUV) {g=true}
  n25 = Add(
   A=Power(Base=ComponentMask(None=@n22.ViewportUV) {g=true}, Exp=@n23) {constExponent=15}
   B=Power(Base=OneMinus(None=@n24), Exp=@n23) {constExponent=15}
  )
  n26 = ComponentMask(None=@n22.ViewportUV) {r=true}
  n27 = Add(
   A=Power(Base=ComponentMask(None=@n22.ViewportUV) {r=true}, Exp=@n23) {constExponent=15}
   B=Power(Base=OneMinus(None=@n26), Exp=@n23) {constExponent=15}
  )
  n28 = Constant() {r=0.8}
  NamedRerouteDeclaration(None=Subtract(A=MaterialFunctionCall(A=@n25, B=@n27, distance=@n28) {materialFunction=&MF_SmoothMax1}.SmoothMax) {constB=0.25}) {name=Edges, nodeColor=<rgba 0.2656 0 1 1>}
  OneMinus(None=@n8)
 ]
}
