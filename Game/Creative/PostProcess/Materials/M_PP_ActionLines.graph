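# Post-process material that overlays animated, comic-style radial action lines on the scene after depth of field.
# Radial UVs, scaled by line count, are rotated by looping time quantized into frames, then feed value noise for stepped flicker that ignores pause.
# The lines use only the color parameter's red channel, so they render grayscale.
# A remapped vignette mask blends them with scene color; Master FX Control, zero by default, fades the effect.
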Material /Game/Creative/PostProcess/Materials/M_PP_ActionLines
graph {
 properties {blendableLocation=BL_SceneColorAfterDOF, materialDomain=MD_PostProcess, pixelDepthOffsetMode=PDOM_Legacy, usedShadingModels=Default Lit}
 nodes [
  n1 = Multiply(
   A=MaterialFunctionCall(Swizzle Coordinate Output=StaticBool()) {materialFunction=&VectorToRadialValue}.Radial Coordinates
   B=AppendVector(
    A=Constant2Vector() {g=400, r=1}.R
    B=ScalarParameter() {defaultValue=400, group=Action Lines, parameterName=Action Lines Number, sliderMax=1200, sliderMin=100}
   )
  )
  n2 = Frac(None=Multiply(A=ScalarParameter() {defaultValue=0.25, group=Action Lines, parameterName=Action Lines Speed, sliderMax=20}, B=Time() {bIgnorePause=true}))
  n3 = ScalarParameter() {defaultValue=24, group=Action Lines, parameterName=Action Lines Frames, sliderMax=120, sliderMin=2}
  n4 = Multiply(A=Floor(None=Multiply(A=@n2, B=@n3) {constB=12}), B=Divide(A=@n3) {constB=1})
  n5 = Noise!(World Position=AppendVector(A=Rotator(Coordinate=@n1, Time=@n4), B=Constant()), FilterWidth) {levels=1, noiseFunction=NOISEFUNCTION_ValueALU, outputMax=0.8, outputMin=0.2}
  n6 = VectorParameter() {group=Action Lines, parameterName=Action Lines Color, defaultValue=<rgba 0.9 0.9 0.9 1>}
  n7 = MaterialFunctionCall(X=@n6.R, Y=@n6.R, Z=@n6.R, A=@n6.R) {materialFunction=&MakeFloat4}
  n8 = SceneTexture() {sceneTextureId=PPI_PostProcessInput0}
  n9 = MaterialFunctionCall(
   Input=MaterialFunctionCall() {materialFunction=/Game/Creative/PostProcess/MaterialFunctions/MF_VignetteMask}.VignetteMask
   Input Low=Constant()
   Input High=Constant() {r=1}
   Target Low=OneMinus(None=ScalarParameter() {group=Vingette, parameterName=Vingette In Intensity, sliderMax=1, sortPriority=2})
   Target High=OneMinus(None=ScalarParameter() {defaultValue=0.666, group=Vingette, parameterName=Vingette Out Intensity, sliderMax=1, sortPriority=1})
  ) {materialFunction=&RemapValueRange}
  LinearInterpolate(
   A=@n8.Color
   B=LinearInterpolate(A=Multiply(A=@n5, B=@n7.Result), B=@n8.Color, Alpha=Saturate(None=@n9.Result))
   Alpha=ScalarParameter() {group=Action Lines, parameterName=Master FX Control, sortPriority=0}
  )
 ]
}
