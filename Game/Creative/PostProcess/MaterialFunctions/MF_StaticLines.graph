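# Material function for a screen-space glitch effect in post process, outputting animated static lines and a UV distortion offset.
# Scaled, offset screen UVs, rotated by a channel-mask direction parameter, drive three panning samples of a static mask texture combined into noise.
# A scrolling Frac and SmoothStep band on screen Y adds a sweep.
# Noise drives horizontal distortion and the sweep vertical, with animation ignoring game pause.
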MaterialFunction /Game/Creative/PostProcess/MaterialFunctions/MF_StaticLines
graph {nodes [
  n1 = ScreenPosition()
  n2 = ComponentMask(None=VectorParameter() {group=X. Static Lines, parameterName=UVScale, defaultValue=<rgba 3 3 0 1>}.RGB) {g=true, r=true}
  n3 = VectorParameter() {group=X. Static Lines, parameterName=UVOffset, defaultValue=<rgba -0.5 0 0 1>}
  n4 = ChannelMaskParameter(None=AppendVector(A=Constant() {r=6.2832}, B=Constant())) {group=X. Static Lines, maskChannel=Alpha, parameterName=Glitch Direction, defaultValue=<rgba 0 0 0 1>}
  n5 = Rotator(Coordinate=Add(A=Multiply(A=@n1.ViewportUV, B=@n2), B=ComponentMask(None=@n3.RGB) {g=true, r=true}), Time=@n4.R)
  n6 = Time() {bIgnorePause=true}
  n7 = Multiply(
   A=TextureSample(UVs=Panner(Coordinate=@n5, Time=@n6) {speedX=6, speedY=-0.05}) {samplerType=SAMPLERTYPE_Masks, texture=/Game/Missions/Primary/RepairTheShelter/Textures/Cinematic_Screen_Static}.G
   B=TextureSample(UVs=Panner(Coordinate=@n5, Time=@n6) {speedX=-3, speedY=-0.03}) {samplerType=SAMPLERTYPE_Masks, texture=&Cinematic_Screen_Static}.R
  )
  n8 = Multiply(A=@n7, B=ScalarParameter() {defaultValue=0.2, group=X. Static Lines, parameterName=Static UV Distort})
  n9 = Constant2Vector() {r=1}
  n10 = ComponentMask(None=ScreenPosition().ViewportUV) {g=true}
  n11 = Multiply(A=Time() {bIgnorePause=true}) {constB=-0.2}
  n12 = SmoothStep(Value=Frac(None=Add(A=@n10, B=@n11))) {constMax=1.2, constMin=0.9}
  n13 = Multiply(A=@n12, B=ScalarParameter() {defaultValue=0.01, group=X. Static Lines, parameterName=Sweep UV Distort}) {constB=0.02}
  n14 = Constant2Vector() {g=1}
  FunctionOutput(None=Add(A=Multiply(A=@n8, B=@n9), B=Multiply(A=@n13, B=@n14))) {outputName=DistortedUVs}
  n15 = Panner(Coordinate=@n5, Time=@n6) {speedX=5, speedY=0.1}
  n16 = Add(
   A=Multiply(A=TextureSample(UVs=@n15) {samplerType=SAMPLERTYPE_Masks, texture=&Cinematic_Screen_Static}.B) {constB=0.1}
   B=@n7
  )
  n17 = ScalarParameter() {defaultValue=1, group=X. Static Lines, parameterName=Static Scale}
  n18 = ScalarParameter() {defaultValue=0.2, group=X. Static Lines, parameterName=Sweep Add}
  FunctionOutput(None=Add(A=Multiply(A=@n16, B=@n17), B=Multiply(A=@n12, B=@n18) {constB=0.2})) {outputName=Static}
 ]}
