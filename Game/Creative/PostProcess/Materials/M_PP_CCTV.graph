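# Post-process material that gives the scene a CCTV look, blended against the original image by a Master FX Control parameter.
# MF_StaticLines offsets the scene UVs, and the resampled image is desaturated.
# White noise from a sine-dot-frac UV hash, with no time input, is mixed in by an opacity parameter.
# A radial-gradient fisheye UV branch with an invert switch is present but does not feed that scene sample.
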Material /Game/Creative/PostProcess/Materials/M_PP_CCTV
graph {
 properties {materialDomain=MD_PostProcess, pixelDepthOffsetMode=PDOM_Legacy, usedShadingModels=Default Lit}
 nodes [
  n1 = MaterialFunctionCall() {materialFunction=/Game/Creative/PostProcess/MaterialFunctions/MF_StaticLines}
  n2 = ComponentMask(None=ScreenPosition().ViewportUV) {g=true, r=true}
  n3 = ComponentMask(None=SceneTexture(UVs=Add(A=@n2, B=@n1.DistortedUVs)) {sceneTextureId=PPI_PostProcessInput0}.Color) {b=true, g=true, r=true}
  n4 = Constant() {r=1}
  n5 = NamedRerouteUsage()
  n6 = DotProduct(
   A=MaterialFunctionCall(X=@n5, Y=@n5, Z=NamedRerouteUsage()) {materialFunction=&MakeFloat3}.Result
   B=Constant3Vector() {constant=<rgba 127.1 311.7 74.7 0>}
  )
  n7 = NamedRerouteUsage()
  n8 = DotProduct(
   A=MaterialFunctionCall(X=NamedRerouteUsage(), Y=@n7, Z=@n7) {materialFunction=&MakeFloat3}.Result
   B=Constant3Vector() {constant=<rgba 269.5 183.3 246.1 0>}
  )
  n9 = NamedRerouteUsage()
  n10 = DotProduct(
   A=MaterialFunctionCall(X=NamedRerouteUsage(), Y=@n9, Z=@n9) {materialFunction=&MakeFloat3}.Result
   B=Constant3Vector() {constant=<rgba 114 272 126 0>}
  )
  n11 = Multiply(
   A=Sine(None=MaterialFunctionCall(X=@n6, Y=@n8, Z=@n10) {materialFunction=&MakeFloat3}.Result)
   B=Constant() {r=1400}
  )
  LinearInterpolate(
   A=ComponentMask(None=SceneTexture() {sceneTextureId=PPI_PostProcessInput0}.Color) {b=true, g=true, r=true}
   B=LinearInterpolate(
    A=Desaturation(None=@n3, Fraction=@n4)
    B=Frac(None=@n11)
    Alpha=ScalarParameter() {defaultValue=0.25, group=2. White Noise, parameterName=White Noise Opacity}
   )
   Alpha=ScalarParameter() {group=0. Material Features, parameterName=Master FX Control, sortPriority=0}
  )
  n12 = TextureCoordinate()
  NamedRerouteDeclaration(None=ComponentMask(None=@n12) {r=true}) {name=red, nodeColor=<rgba 0 0.2969 1 1>}
  NamedRerouteDeclaration(None=ComponentMask(None=@n12) {g=true}) {name=green, nodeColor=<rgba 0 1 0 1>}
  n13 = Add(A=@n2, B=Constant() {r=-0.5}) {constB=-1}
  n14 = MaterialFunctionCall(
   UVs=@n13
   CenterPosition=Constant()
   Radius=ScalarParameter() {defaultValue=1, group=1. Fish Eye Distortion, parameterName=FishEye Radius}
   Density=ScalarParameter() {group=1. Fish Eye Distortion, parameterName=FishEye Falloff}
   Invert Density=StaticBool()
  ) {materialFunction=&RadialGradientExponential}
  n15 = StaticSwitchParameter(True=OneMinus(None=@n14.RadialGradientExponential), False=@n14.RadialGradientExponential) {group=0. Material Features, parameterName=invertFishEye}
  Subtract(
   A=@n2
   B=Multiply(A=ScalarParameter() {defaultValue=0.5, group=1. Fish Eye Distortion, parameterName=Fish Eye Intensity}, B=Multiply(A=@n13, B=@n15))
  )
 ]
}
