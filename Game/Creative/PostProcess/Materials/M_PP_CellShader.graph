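# Post-process cel shader material, run after DOF, that turns scene color into banded tones with an optional posterize.
# Luminance from a four-tap blurred scene over base color, threshold bands, a normalized scene/base-color tone blend and log2 tone quantization build the look.
# Depth gives skybox masks, and custom depth/stencil give a character mask using screen UVs from a thermal material layer.
# Feature-level switches replace base-color reads on lower-end targets, and static switches expose debug views.
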Material /Game/Creative/PostProcess/Materials/M_PP_CellShader
graph {
 properties {bAutomaticallySetUsageInEditor=false, blendableLocation=BL_SceneColorAfterDOF, materialDomain=MD_PostProcess, usedShadingModels=Default Lit}
 nodes [
  n1 = NamedRerouteUsage()
  n2 = StaticSwitchParameter(True=AppendVector(A=@n1, B=Constant()), False=Multiply(A=Desaturation(None=@n1), B=NamedRerouteUsage())) {group=Thresholds, parameterName=Debug Threshold}
  n3 = SceneTexture() {sceneTextureId=PPI_PostProcessInput0}
  n4 = Round(None=ScalarParameter() {defaultValue=25, group=Sky, parameterName=Sky Steps})
  n5 = Multiply(A=@n3.Color, B=@n4)
  n6 = StaticSwitchParameter(True=Divide(A=Floor(None=@n5), B=@n4), False=@n3.Color) {parameterName=Posterize}
  n7 = NamedRerouteUsage()
  LinearInterpolate(
   A=SceneTexture() {sceneTextureId=PPI_PostProcessInput0}.Color
   B=LinearInterpolate(A=SceneTexture() {sceneTextureId=PPI_PostProcessInput0}.Color, B=LinearInterpolate(A=@n2, B=@n6, Alpha=@n7), Alpha=NamedRerouteUsage())
   Alpha=ScalarParameter() {defaultValue=1, group=Tone Blending, parameterName=Global Strength, sliderMax=1}
  )
  n8 = ComponentMask(None=SceneTexture() {sceneTextureId=PPI_PostProcessInput0}.Color) {b=true, g=true, r=true}
  n9 = Max(A=ComponentMask(None=@n8) {r=true}, B=Max(A=ComponentMask(None=@n8) {g=true}, B=ComponentMask(None=@n8) {b=true}))
  n10 = Constant() {r=2}
  n11 = Logarithm2!(X=@n9)
  n12 = ScalarParameter() {defaultValue=1, group=Tone Blending, parameterName=Steps}
  n13 = Divide(A=Round(None=Multiply(A=@n11, B=@n12)), B=@n12)
  n14 = LinearInterpolate(
   A=ComponentMask(None=SceneTexture() {sceneTextureId=PPI_PostProcessInput0}.Color) {b=true, g=true, r=true}
   B=Multiply(A=Divide(A=@n8, B=@n9), B=Power(Base=@n10, Exp=@n13))
   Alpha=ScalarParameter() {defaultValue=1, group=Tone Blending, parameterName=Global Strength}
  )
  n15 = ComponentMask(None=SceneTexture() {sceneTextureId=PPI_PostProcessInput0}.Color) {b=true, g=true, r=true}
  n16 = NamedRerouteUsage()
  LinearInterpolate(
   A=ComponentMask(None=SceneTexture() {sceneTextureId=PPI_PostProcessInput0}.Color) {b=true, g=true, r=true}
   B=ComponentMask(None=LinearInterpolate(A=@n14, B=@n15, Alpha=@n16)) {b=true, g=true, r=true}
   Alpha=ScalarParameter() {group=Main, parameterName=Master FX Control, sortPriority=0}
  )
  n17 = Constant()
  n18 = Multiply(A=ScalarParameter() {defaultValue=2, parameterName=Blur Base}) {constB=0.001}
  n19 = AppendVector(A=@n17, B=@n18)
  n20 = TextureCoordinate()
  n21 = Multiply(A=@n18) {constB=-1}
  n22 = AppendVector(A=@n17, B=@n21)
  n23 = Add(
   A=SceneTexture(UVs=Add(A=@n19, B=@n20)) {sceneTextureId=PPI_PostProcessInput0}.Color
   B=SceneTexture(UVs=Add(A=@n22, B=@n20)) {sceneTextureId=PPI_PostProcessInput0}.Color
  )
  n24 = AppendVector(A=@n18, B=@n17)
  n25 = AppendVector(A=@n21, B=@n17)
  n26 = Add(
   A=SceneTexture(UVs=Add(A=@n24, B=@n20)) {sceneTextureId=PPI_PostProcessInput0}.Color
   B=SceneTexture(UVs=Add(A=@n25, B=@n20)) {sceneTextureId=PPI_PostProcessInput0}.Color
  )
  n27 = Desaturation(None=Divide(A=Add(A=@n23, B=@n26), B=Constant() {r=4}))
  n28 = SceneTexture() {sceneTextureId=PPI_PostProcessInput0}
  n29 = Divide(A=@n27, B=LinearInterpolate(A=Constant() {r=0.03}, B=Constant() {r=0.1}, Alpha=Desaturation(None=@n28.Color)) {constA=0.08, constB=0.1})
  n30 = Divide(A=@n27, B=Desaturation(None=SceneTexture() {sceneTextureId=PPI_BaseColor}.Color))
  NamedRerouteDeclaration(None=StaticSwitchParameter(True=@n29, False=FeatureLevelSwitch(Default=@n30, ES2=@n29, ES3_1=@n29, SM4_REMOVED=@n29)) {group=Thresholds, parameterName=Debug NSwitch}) {name=Luminance, nodeColor=<rgba 0.6172 0 1 1>}
  n31 = SceneTexture() {sceneTextureId=PPI_PostProcessInput0}
  n32 = MaterialFunctionCall(Vector=@n31.Color, Default=@n31.Color) {materialFunction=&SafeNormalize}
  n33 = LinearInterpolate(
   A=@n32.Result
   B=VectorParameter() {group=Tone Blending, parameterName=Normalize Fallback Colour, defaultValue=<rgba 0 0 0 1>}.RGB
   Alpha=@n32.Result
  )
  n34 = Constant()
  n35 = LinearInterpolate(
   A=@n31.Color
   B=ComponentMask(None=AppendVector(A=@n33, B=@n34)) {a=true, b=true, g=true, r=true}
   Alpha=ScalarParameter() {defaultValue=1, group=Tone Blending, parameterName=PP0 Normalisze Amount, sliderMax=1, sortPriority=11}
  )
  n36 = ScalarParameter() {defaultValue=2.5, group=Tone Blending, parameterName=PP0 Brightness Switch, sortPriority=11}
  n37 = StaticSwitchParameter(
   True=@n36
   False=FeatureLevelSwitch(
    Default=ScalarParameter() {defaultValue=2.5, group=Tone Blending, parameterName=PP0 Brightness, sortPriority=12}
    ES2=@n36
    ES3_1=@n36
    SM4_REMOVED=@n36
   )
  ) {group=Thresholds, parameterName=Debug NSwitch}
  n38 = Multiply(
   A=SceneTexture() {sceneTextureId=PPI_StoredBaseColor}.Color
   B=ScalarParameter() {defaultValue=2, group=Tone Blending, parameterName=BC Brightness, sortPriority=13}
  )
  n39 = SceneTexture() {sceneTextureId=PPI_PostProcessInput0}
  NamedRerouteDeclaration(None=LinearInterpolate(
    A=Multiply(A=@n35, B=@n37)
    B=FeatureLevelSwitch(Default=@n38, ES2=@n39.Color, ES3_1=@n39.Color, SM4_REMOVED=@n39.Color)
    Alpha=ScalarParameter() {defaultValue=0.5, group=Tone Blending, parameterName=Blend PP0 / BC, sliderMax=1, sortPriority=10}
   )) {name=SceneTone, nodeColor=<rgba 1 0.7676 0 1>}
  NamedRerouteDeclaration(None=MaterialFunctionCall!(
    Input=SceneTexture() {sceneTextureId=PPI_SceneDepth}.Color
    Input Low=ScalarParameter() {defaultValue=10000, group=Tone Blending, parameterName=Falloff Near Point}
    Input High=ScalarParameter() {defaultValue=15000, group=Tone Blending, parameterName=Falloff Far Point}
   ) {materialFunction=/Engine/Functions/Engine_MaterialFunctions03/Math/RemapValueRangeNormalized}.Result) {name=Skybox Mask, nodeColor=<rgba 1 0.9141 0 1>}
  n40 = NamedRerouteUsage()
  n41 = Subtract(A=ComponentMask(None=SceneTexture(UVs=@n40) {sceneTextureId=PPI_CustomDepth}.Color) {r=true}, B=SceneDepth(UVs=@n40).R)
  n42 = Constant() {r=1}
  n43 = Multiply(A=OneMinus(None=Saturate(None=@n41)), B=If(A=NamedRerouteUsage(), "A > B"=@n42, "A == B"=Constant(), "A < B"=@n42) {constB=22})
  n44 = MaterialAttributeLayers!(None) {defaultLayers {
    blends []
    editorOnly {
     deletedParentLayerGuids []
     layerGuids [#1]
     layerLinkStates [NotFromParent]
     layerNames [Background]
     layerStates [true]
     restrictToBlendRelatives []
     restrictToLayerRelatives [false]
    }
    layers [/Game/Weapons/FORT_Rifles/ThermalVision/ThermalAnimation/ML_Thermal_Init]
   }}
  n45 = GetMaterialAttributes(None=@n44) {
   desc="UV 0: Screen Space UV\r\nUV 1 (R): Post Process Mask\r\nUV 1 (G): Toggle Thermal Scale\r\nEmissive: Color Multiplier"
   attributeGetTypes [#2]
  }
  n46 = NamedRerouteDeclaration(None=ComponentMask(None=@n45.Customized UV 0) {g=true, r=true}) {name=Screen Space UV, nodeColor=<rgba 1 0 0.1875 1>}
  n47 = NamedRerouteDeclaration(None=ComponentMask(None=SceneTexture(UVs=@n46) {sceneTextureId=PPI_CustomStencil}.Color) {r=true}) {name=Custom Stencil, nodeColor=<rgba 1 0 0.0625 1>}
  n48 = Constant() {r=1}
  n49 = If(A=@n47, B=Constant() {r=2}, "A > B"=@n48, "A == B"=Constant(), "A < B"=@n48) {desc=buildings highlight}
  n50 = Constant() {r=1}
  n51 = If(A=@n47, B=Constant() {r=123}, "A > B"=@n50, "A == B"=Constant(), "A < B"=@n50) {desc=Damage numbers highlight}
  n52 = Abs(None=Subtract(A=@n47) {constB=27})
  n53 = Constant() {r=1}
  n54 = Constant()
  n55 = Multiply(
   A=SmoothStep(Value=SceneDepth(UVs=@n46).R) {constMax=5000000, constMin=10000000, desc=Sky mask}
   B=Multiply(A=Multiply(A=@n49, B=@n51), B=If(A=@n52, "A > B"=@n53, "A == B"=@n54, "A < B"=@n54) {constB=1.01, desc="BelongTreaty FX (27-28)"})
  )
  n56 = Constant()
  n57 = If(A=NamedRerouteUsage(), "A > B"=@n56, "A == B"=Constant() {r=1}, "A < B"=@n56) {constB=24}
  n58 = Max(
   A=@n55
   B=LinearInterpolate(A=@n55, B=Max(B=@n57), Alpha=ScalarParameter() {group=X. Native Functionality, parameterName=Reveal Cloaking Players})
  )
  NamedRerouteDeclaration(None=OneMinus(None=Multiply(A=@n58, B=@n43))) {name=Character Mask, nodeColor=<rgba 0 0.9297 1 1>}
  n59 = LinearInterpolate(A=Constant() {r=0.08}, B=Constant() {r=0.1}, Alpha=SceneTexture() {sceneTextureId=PPI_PostProcessInput0}.Color) {constA=0.08, constB=0.1}
  n60 = NamedRerouteUsage()
  n61 = ScalarParameter() {defaultValue=4, parameterName=2 Threshold}
  n62 = If(
   A=NamedRerouteUsage()
   B=ScalarParameter() {defaultValue=8, parameterName=1 Threshold}
   "A > B"=StaticSwitchParameter(
    True=VectorParameter() {group=Thresholds, parameterName=0 Colour, defaultValue=<rgba 1 0.3184 0.9001 1>}.RGB
    False=ScalarParameter() {defaultValue=3, parameterName=1 Value}
   ) {group=Thresholds, parameterName=Debug Threshold}
   "A < B"=StaticSwitchParameter(
    True=VectorParameter() {group=Thresholds, parameterName=1 Colour, defaultValue=<rgba 1 0 0 1>}.RGB
    False=ScalarParameter() {defaultValue=4, parameterName=2 Value}
   ) {group=Thresholds, parameterName=Debug Threshold}
  )
  n63 = StaticSwitchParameter(
   True=VectorParameter() {group=Thresholds, parameterName=2 Colour, defaultValue=<rgba 0 1 0 1>}.RGB
   False=ScalarParameter() {defaultValue=2, parameterName=3 Value}
  ) {group=Thresholds, parameterName=Debug Threshold}
  n64 = VectorParameter() {group=Thresholds, parameterName=3 Colour, defaultValue=<rgba 0 0 1 1>}
  n65 = ScalarParameter() {defaultValue=1.25, parameterName=4 Value}
  n66 = If(
   A=NamedRerouteUsage()
   B=ScalarParameter() {defaultValue=0.6, parameterName=4 Threshold}
   "A > B"=If(
    A=NamedRerouteUsage()
    B=ScalarParameter() {defaultValue=1.3, parameterName=3 Threshold}
    "A > B"=If(A=@n60, B=@n61, "A > B"=@n62, "A < B"=@n63)
    "A < B"=StaticSwitchParameter(True=@n64.RGB, False=@n65) {group=Thresholds, parameterName=Debug Threshold}
   )
   "A < B"=StaticSwitchParameter(
    True=VectorParameter() {group=Thresholds, parameterName=4 Colour, defaultValue=<rgba 1 1 0 1>}.RGB
    False=ScalarParameter() {defaultValue=0.6, parameterName=5 Value}
   ) {group=Thresholds, parameterName=Debug Threshold}
  )
  n67 = FeatureLevelSwitch(Default=SceneTexture() {sceneTextureId=PPI_BaseColor}.Color, ES2=@n59, ES3_1=@n59, SM4_REMOVED=@n59)
  NamedRerouteDeclaration(None=StaticSwitchParameter(True=Multiply(A=@n59, B=@n66), False=Multiply(A=@n67, B=@n66)) {group=Thresholds, parameterName=Debug NSwitch}) {name=Thresholds, nodeColor=<rgba 1 0 0.8203 1>}
  n68 = Divide(
   A=SceneTexture(UVs=NamedRerouteUsage()) {sceneTextureId=PPI_SceneDepth}.Color
   B=ScalarParameter() {defaultValue=150000, group=Sky, parameterName=Skybox Mask Range}
  ) {constB=100000}
  n69 = Constant() {r=1}
  Saturate(None=If(A=ComponentMask(None=@n68) {b=true, g=true, r=true}, B=Constant() {r=1}, "A > B"=@n69, "A == B"=@n69, "A < B"=Constant())) {desc=old skybox}
 ]
}
