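# Post-process material that gives the scene a comic cel-shaded look, applied to scene color before depth of field.
# It bands brightness by rounding log2 of the largest RGB channel to a Steps-controlled interval, then rescales the color, keeping channel ratios.
# A remapped scene-depth skybox mask fades distant pixels back to the original, unless Cel Shade Sky overrides it.
# Master FX Control blends the result over the unmodified scene.
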Material /Game/Creative/PostProcess/Materials/M_PP_ComicCellShader
graph {
 properties {
  bAutomaticallySetUsageInEditor=false
  blendableLocation=BL_SceneColorBeforeDOF
  materialDomain=MD_PostProcess
  pixelDepthOffsetMode=PDOM_Legacy
  usedShadingModels=Default Lit
 }
 nodes [
  n1 = ComponentMask(None=SceneTexture() {sceneTextureId=PPI_PostProcessInput0}.Color) {b=true, g=true, r=true}
  n2 = Max(A=ComponentMask(None=@n1) {r=true}, B=Max(A=ComponentMask(None=@n1) {g=true}, B=ComponentMask(None=@n1) {b=true}))
  n3 = ScalarParameter() {defaultValue=1, parameterName=Steps}
  n4 = Round(None=Multiply(A=Logarithm2!(X=@n2), B=@n3))
  n5 = Multiply(A=Divide(A=@n1, B=@n2), B=Power(Base=Constant() {r=2}, Exp=Divide(A=@n4, B=@n3)))
  n6 = LinearInterpolate(A=@n5, B=ComponentMask(None=SceneTexture() {sceneTextureId=PPI_PostProcessInput0}.Color) {b=true, g=true, r=true}, Alpha=NamedRerouteUsage())
  n7 = ScalarParameter() {parameterName=Cel Shade Sky}
  LinearInterpolate(
   A=ComponentMask(None=SceneTexture() {sceneTextureId=PPI_PostProcessInput0}.Color) {b=true, g=true, r=true}
   B=ComponentMask(None=LinearInterpolate(A=@n6, B=@n5, Alpha=@n7)) {b=true, g=true, r=true}
   Alpha=ScalarParameter() {group=Main, parameterName=Master FX Control, sortPriority=0}
  )
  NamedRerouteDeclaration(None=MaterialFunctionCall!(
    Input=SceneTexture() {sceneTextureId=PPI_SceneDepth}.Color
    Input Low=ScalarParameter() {defaultValue=100000, parameterName=Falloff Near Point}
    Input High=ScalarParameter() {defaultValue=150000, parameterName=Falloff Far Point}
   ) {materialFunction=/Engine/Functions/Engine_MaterialFunctions03/Math/RemapValueRangeNormalized}.Result) {name=Skybox Mask, nodeColor=<rgba 1 0.9141 0 1>}
 ]
}
