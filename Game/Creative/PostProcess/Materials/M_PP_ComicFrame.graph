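# M_PP_ComicFrame is a post-process material that overlays a comic-style frame on the screen.
# T_ComicFrame's red channel blends from white to the scene color, then its green channel blends that result to black.
# The Master FX Control scalar parameter fades between the untouched scene and the framed result.
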Material /Game/Creative/PostProcess/Materials/M_PP_ComicFrame
graph {
 properties {
  bAutomaticallySetUsageInEditor=false
  blendablePriority=1
  materialDomain=MD_PostProcess
  pixelDepthOffsetMode=PDOM_Legacy
  usedShadingModels=Default Lit
 }
 nodes [
  n1 = SceneTexture() {sceneTextureId=PPI_PostProcessInput0}
  n2 = TextureSample() {texture=/Game/Creative/PostProcess/Textures/T_ComicFrame}
  LinearInterpolate(
   A=ComponentMask(None=SceneTexture() {sceneTextureId=PPI_PostProcessInput0}.Color) {a=true, b=true, g=true, r=true}
   B=LinearInterpolate(A=LinearInterpolate(B=@n1.Color, Alpha=@n2.R) {constA=1}, Alpha=@n2.G) {constB=0}
   Alpha=ScalarParameter() {group=Main, parameterName=Master FX Control, sortPriority=0}
  )
 ]
}
