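# Material function for post-process cross-hatching that outputs a line-darkening factor from screen UVs, a light vector and an inverted mask.
# Screen UVs dotted with the rotated light vector, scaled by line amount and viewport height, form repeating stripes shaped by thickness, contrast and opacity.
# Voronoi noise fed into a second vector noise yields random per-cell values, shared through named reroutes to vary stripe angle and density.
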MaterialFunction /Game/Creative/PostProcess/MaterialFunctions/MF_CrossHatch
graph {nodes [
  n1 = FunctionInput() {inputName=LightVector, inputType=FunctionInput_Vector2}
  n2 = Multiply(A=Subtract(A=NamedRerouteUsage()) {constB=0.5}, B=FunctionInput() {inputName=Line Direction, inputType=FunctionInput_Scalar, sortPriority=2})
  n3 = FunctionInput() {inputName=ScreenUVs, inputType=FunctionInput_Vector2}
  n4 = FunctionInput() {inputName=Line Amount, inputType=FunctionInput_Scalar, sortPriority=2}
  n5 = Multiply(A=Multiply(A=@n4) {constB=0.1}, B=Subtract(A=NamedRerouteUsage()) {constB=0.5})
  n6 = ComponentMask(None=ViewSize()) {g=true}
  n7 = Constant() {r=1000}
  n8 = Multiply(
   A=DotProduct(A=@n3, B=Rotator(Coordinate=@n1, Time=@n2) {centerX=0, centerY=0, speed=6.2})
   B=Multiply(A=Add(A=@n4, B=@n5), B=Divide(A=@n6, B=@n7))
  )
  n9 = Subtract(A=Multiply(A=Frac(None=@n8)) {constB=2})
  n10 = Divide(
   A=Subtract(A=Abs(None=@n9), B=FunctionInput() {inputName=Line Thickness, inputType=FunctionInput_Scalar, sortPriority=2})
   B=FunctionInput() {inputName=Line Contrast, inputType=FunctionInput_Scalar, sortPriority=3}
  )
  n11 = Multiply(A=OneMinus(None=Saturate(None=@n10)), B=OneMinus(None=FunctionInput() {inputName=Mask, inputType=FunctionInput_Scalar, sortPriority=1}))
  n12 = FunctionInput() {inputName=Line Opacity, inputType=FunctionInput_Scalar, sortPriority=5}
  FunctionOutput(None=OneMinus(None=Multiply(A=@n11, B=@n12)))
  n13 = AppendVector(
   A=Multiply(A=@n3, B=FunctionInput() {inputName=Voronoi Cell Amount, inputType=FunctionInput_Scalar, sortPriority=1})
   B=FunctionInput() {inputName=Random Voronoi Value, inputType=FunctionInput_Scalar, sortPriority=1}
  )
  n14 = VectorNoise(World Position=ComponentMask(None=VectorNoise(World Position=@n13) {noiseFunction=VNF_VoronoiALU}) {b=true, g=true, r=true})
  NamedRerouteDeclaration(None=ComponentMask(None=@n14) {r=true}) {name=NoiseA, nodeColor=<rgba 0.2656 0 1 1>}
  NamedRerouteDeclaration(None=ComponentMask(None=@n14) {g=true}) {name=NoiseB, nodeColor=<rgba 1 0 0.7969 1>}
 ]}
