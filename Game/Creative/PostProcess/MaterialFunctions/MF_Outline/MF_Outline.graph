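# Post-process material function that draws depth-based and normal-based outlines over an input scene color, outputting Color and the depth-edge mask as Alpha.
# Depth edges use modulated thickness and threshold, are distance-masked, and blend in an outline color.
# A second pass thresholds summed WorldNormal differences between center and four neighboring pixels for another distance-masked color blend.
# A shading path switch keeps this normal pass for deferred, giving forward and mobile only depth outlines.
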MaterialFunction /Game/Creative/PostProcess/MaterialFunctions/MF_Outline/MF_Outline
graph {nodes [
  Constant() {r=0.5}
  n1 = ScalarParameter() {defaultValue=6500, group=3.0 Mesh Outline, parameterName=MaxOutlineDistance}
  n2 = MaterialFunctionCall!(
   Min Thickness=ScalarParameter() {defaultValue=0.5, group=3.0 Mesh Outline, parameterName=Edge Thickness Min}
   Max Depth=ScalarParameter() {defaultValue=5000, group=3.0 Mesh Outline, parameterName=Max Depth Outline}
   Max Thickness=ScalarParameter() {defaultValue=3, group=3.0 Mesh Outline, parameterName=Edge Thickness}
  ) {materialFunction=&/MF_LineModulation}
  n3 = ScalarParameter() {defaultValue=1, group=3.0 Mesh Outline, parameterName=Edge Threshold}
  n4 = MaterialFunctionCall() {materialFunction=&/MF_DepthModulation}
  n5 = Multiply(
   A=MaterialFunctionCall!(Max Thickness=@n2.Result, MaxOutlineDistance=NamedRerouteDeclaration(None=@n1) {name=maxDistance, nodeColor=<rgba 0 1 0.3906 1>}) {materialFunction=&/MF_DistanceMask}.Result
   B=MaterialFunctionCall!(Edge Threshold=Multiply(A=@n3, B=@n4.Result), Thickness=@n2.Result) {materialFunction=&/MF_DetectEdges_Depth}.Result
  )
  n6 = LinearInterpolate(
   A=ComponentMask(None=FunctionInput() {inputName=SceneTexture}) {b=true, g=true, r=true}
   B=VectorParameter() {group=3. Outline, parameterName=Color Outline, defaultValue=<rgba 0 0 0 1>}.RGB
   Alpha=@n5
  )
  n7 = NamedRerouteUsage()
  n8 = MaterialFunctionCall(
   Min Thickness=ScalarParameter() {group=3.1 Normal Outline, parameterName=Edge Normal Thickness Min}
   Max Depth=ScalarParameter() {group=3.1 Normal Outline, parameterName=Max Normal Depth Outline}
   Max Thickness=ScalarParameter() {group=3.1 Normal Outline, parameterName=Edge Normal Thickness}
  ) {materialFunction=&MF_LineModulation}
  n9 = ScalarParameter() {defaultValue=1, group=3.1 Normal Outline, parameterName=Edge Normal Threshold}
  n10 = NamedRerouteUsage()
  n11 = MaterialFunctionCall!(Pixel Width=@n8.Result) {materialFunction=&/MF_Kernel}
  n12 = ComponentMask(None=SceneTexture(UVs=@n11.Left Pixel) {sceneTextureId=PPI_WorldNormal}.Color) {b=true, g=true, r=true}
  n13 = NamedRerouteUsage()
  n14 = ComponentMask(None=SceneTexture(UVs=@n11.Right Pixel) {sceneTextureId=PPI_WorldNormal}.Color) {b=true, g=true, r=true}
  n15 = NamedRerouteUsage()
  n16 = ComponentMask(None=SceneTexture(UVs=@n11.Top Pixel) {sceneTextureId=PPI_WorldNormal}.Color) {b=true, g=true, r=true}
  n17 = NamedRerouteDeclaration(None=ComponentMask(None=SceneTexture(UVs=@n11.Center Pixel) {sceneTextureId=PPI_WorldNormal}.Color) {b=true, g=true, r=true}) {name=centerNormal, nodeColor=<rgba 0 0.2969 1 1>}
  n18 = ComponentMask(None=SceneTexture(UVs=@n11.Bottom Pixel) {sceneTextureId=PPI_WorldNormal}.Color) {b=true, g=true, r=true}
  n19 = Add(A=Add(A=Distance(A=@n10, B=@n12), B=Distance(A=@n13, B=@n14)), B=Add(A=Distance(A=@n15, B=@n16), B=Distance(A=@n17, B=@n18)))
  n20 = LinearInterpolate(
   A=@n6
   B=VectorParameter() {group=3. Outline, parameterName=Normal Color Outline, defaultValue=<rgba 0 0 0 1>}.RGB
   Alpha=Multiply(
    A=MaterialFunctionCall(Max Thickness=@n8.Result, MaxOutlineDistance=@n7) {materialFunction=&MF_DistanceMask}.Result
    B=Step(Y=@n9, X=@n19)
   )
  )
  n21 = NamedRerouteUsage()
  FunctionOutput(None=ShadingPathSwitch(Default=@n20, Deferred=@n20, Forward=@n21, Mobile=@n21)) {outputName=Color}
  FunctionOutput(None=@n5) {outputName=Alpha, sortPriority=1}
  NamedRerouteDeclaration(None=@n6) {name=justOutline, nodeColor=<rgba 0 1 0.2969 1>}
 ]}
