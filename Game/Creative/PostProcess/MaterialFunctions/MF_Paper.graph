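# Material function in the Creative post-process folder that builds a tiled paper grain pattern and outputs it as PaperTexture.
# Texture coordinates are scaled by the PaperVFX Tiling parameter and shared by two paper grain texture samples.
# The dojo ricepaper grain is brightened by a constant offset, then multiplied by the Creative paper grain to produce the RGB result.
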MaterialFunction /Game/Creative/PostProcess/MaterialFunctions/MF_Paper
graph {nodes [
  n1 = Multiply(A=TextureCoordinate(), B=ScalarParameter() {defaultValue=3, group=1. Paper VFX, parameterName=PaperVFX Tiling})
  n2 = TextureSample(UVs=@n1) {texture=/Game/Environments/Asteria/Sets/Dojo/Materials/RicepaperMaterial/Pattern_PaperGrain}
  FunctionOutput(None=Multiply(A=Add(A=@n2.RGB) {constB=0.4}, B=TextureSample(UVs=@n1) {texture=/Game/Creative/PostProcess/Textures/T_Pattern_PaperGrain}.RGB)) {outputName=PaperTexture}
 ]}
