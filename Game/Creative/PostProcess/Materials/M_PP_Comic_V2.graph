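# Comic-style post-process material that layers paper texture, outlines and cross-hatching over the scene, with per-effect and master strength controls.
# Scene color gets a saturation boost through negative desaturation, then blends through MF_Paper and MF_Outline.
# MF_CrossHatch builds Voronoi-varied lines from aspect-corrected screen UVs, a brightness mask from the maximum RGB channel, and the sky atmosphere light direction projected into view space.
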Material /Game/Creative/PostProcess/Materials/M_PP_Comic_V2
graph {
 properties {materialDomain=MD_PostProcess, pixelDepthOffsetMode=PDOM_Legacy, usedShadingModels=Default Lit}
 nodes [
  n1 = NamedRerouteUsage()
  n2 = LinearInterpolate(
   A=@n1
   B=MaterialFunctionCall() {materialFunction=/Game/Creative/PostProcess/MaterialFunctions/MF_Paper}.PaperTexture
   Alpha=ScalarParameter() {defaultValue=0.3, group=0. VFX Adjustments, parameterName=PaperVFXStrength}
  )
  n3 = MaterialFunctionCall!(SceneTexture=@n2) {materialFunction=/Game/Creative/PostProcess/MaterialFunctions/MF_Outline/MF_Outline}
  n4 = ScalarParameter() {defaultValue=0.9, group=0. VFX Adjustments, parameterName=OutlineVFXStrength}
  LinearInterpolate(
   A=@n1
   B=LinearInterpolate(
    A=LinearInterpolate(A=@n2, B=@n3.Color, Alpha=@n4) {constAlpha=0}
    B=NamedRerouteUsage()
    Alpha=ScalarParameter() {defaultValue=0.2, group=0. VFX Adjustments, parameterName=CrossHatchVFXStrength}
   )
   Alpha=ScalarParameter() {group=0. VFX Adjustments, parameterName=MasterVFXAdjustment, sortPriority=1}
  )
  n5 = Subtract(A=TextureCoordinate()) {constB=0.5}
  n6 = ComponentMask(None=@n5) {r=true}
  n7 = ViewSize()
  n8 = Divide(A=ComponentMask(None=@n7) {r=true}, B=ComponentMask(None=@n7) {g=true})
  n9 = ComponentMask(None=@n5) {g=true}
  NamedRerouteDeclaration(None=AppendVector(A=Multiply(A=@n6, B=@n8), B=Multiply(A=@n9) {constB=-1})) {name=screenUVS, nodeColor=<rgba 0 0.3437 1 1>}
  n10 = Transform(None=SkyAtmosphereLightDirection()) {transformSourceType=TRANSFORMSOURCE_World, transformType=TRANSFORM_View}
  NamedRerouteDeclaration(None=Normalize(VectorInput=ComponentMask(None=@n10) {g=true, r=true})) {name=screenLightVector, nodeColor=<rgba 0 0.6016 1 1>}
  n11 = SceneTexture() {sceneTextureId=PPI_PostProcessInput0}
  n12 = ScalarParameter() {defaultValue=1.8, group=0. VFX Adjustments, parameterName=Color Boost}
  NamedRerouteDeclaration(None=Desaturation(None=ComponentMask(None=@n11.Color) {b=true, g=true, r=true}, Fraction=OneMinus(None=@n12))) {name=CameraDiffuse, nodeColor=<rgba 0 0.6953 1 1>}
  n13 = SceneTexture() {sceneTextureId=PPI_PostProcessInput0}
  n14 = ComponentMask(None=@n13.Color) {r=true}
  n15 = Max(A=ComponentMask(None=@n13.Color) {g=true}, B=ComponentMask(None=@n13.Color) {b=true})
  NamedRerouteDeclaration(None=InverseLinearInterpolate(Value=Max(A=@n14, B=@n15)) {constA=0.5, constB=0.8}) {name=LayerAMask, nodeColor=<rgba 0 1 0.1562 1>}
  NamedRerouteDeclaration(None=MaterialFunctionCall!(
    LightVector=NamedRerouteUsage()
    ScreenUVs=NamedRerouteUsage()
    Mask=NamedRerouteUsage()
    Random Voronoi Value=ScalarParameter() {defaultValue=15, group=1. Layer A Cross Hatch, parameterName=Layer A Voronoi Random Value}
    Voronoi Cell Amount=ScalarParameter() {defaultValue=20, group=1. Layer A Cross Hatch, parameterName=Layer A Voronoi Cell Amount}
    Line Thickness=ScalarParameter() {defaultValue=0.6, group=1. Layer A Cross Hatch, parameterName=Layer A Line Thickness}
    Line Direction=ScalarParameter() {group=1. Layer A Cross Hatch, parameterName=Layer A Line Direction}
    Line Amount=ScalarParameter() {defaultValue=200, group=1. Layer A Cross Hatch, parameterName=Layer A Line Amount}
    Line Contrast=ScalarParameter() {defaultValue=0.7, group=1. Layer A Cross Hatch, parameterName=Layer A Line Contrast}
    Line Opacity=ScalarParameter() {defaultValue=0.8, group=1. Layer A Cross Hatch, parameterName=Layer A Line Opacity}
   ) {materialFunction=/Game/Creative/PostProcess/MaterialFunctions/MF_CrossHatch}.Result) {name=Diagonal Lines, nodeColor=<rgba 0 1 0.5547 1>}
 ]
}
