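# M_PP_Desaturate is a post-process material that fades the scene image toward grayscale.
# It samples PostProcessInput0 and desaturates it with equal RGB luminance weights rather than perceptual ones.
# A LinearInterpolate blends the original RGB with the desaturated RGB, using the "Master FX Control" scalar parameter as the effect strength.
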Material /Game/Creative/PostProcess/Materials/M_PP_Desaturate
graph {
 properties {bAutomaticallySetUsageInEditor=false, materialDomain=MD_PostProcess, usedShadingModels=Default Lit}
 nodes [
  n1 = SceneTexture() {sceneTextureId=PPI_PostProcessInput0}
  LinearInterpolate(
   A=ComponentMask(None=SceneTexture() {sceneTextureId=PPI_PostProcessInput0}.Color) {b=true, g=true, r=true}
   B=ComponentMask(None=Desaturation(None=@n1.Color) {luminanceFactors=<rgba 0.33 0.33 0.33 1>}) {b=true, g=true, r=true}
   Alpha=ScalarParameter() {group=Main, parameterName=Master FX Control, sortPriority=0}
  )
 ]
}
