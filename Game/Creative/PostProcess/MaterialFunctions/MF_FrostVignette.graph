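# Material function that builds a radial vignette mask, named for a frost post-process effect.
# It measures UV distance from (0.5, 0.5), with a static switch choosing aspect-scaled screen-aligned UVs or plain texture coordinates.
# An inverse lerp between the Radius Control and Radius inputs remaps that distance and saturates it, darkest at center and brightest toward the edges with default inputs.
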MaterialFunction /Game/Creative/PostProcess/MaterialFunctions/MF_FrostVignette
graph {nodes [
  n1 = FunctionInput(Preview=Constant()) {bUsePreviewValueAsDefault=true, inputName=Radius Control}
  n2 = FunctionInput(Preview=Constant() {r=0.5}) {bUsePreviewValueAsDefault=true, inputName=Radius, inputType=FunctionInput_Scalar}
  n3 = Distance(
   A=StaticSwitchParameter(
    True=MaterialFunctionCall() {materialFunction=&ScreenAlignedUVs}."X 100%, Y Scale to Ratio"
    False=TextureCoordinate()
   ) {group=0. Vignette, parameterName=useScreenAlignedUVs}
   B=Constant2Vector() {g=0.5, r=0.5}
  )
  FunctionOutput(None=Saturate(None=InverseLinearInterpolate(A=@n1, B=@n2, Value=@n3)))
 ]}
