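# M_PP_Frost_Simple is a post-process material that layers a frost and ice effect over the screen.
# MF_FrostVignette masks shape screen-space noise from a channel-packed texture and Voronoi noise into ice, with scene UV distortion from screen derivatives.
# Distance-based fog, secondary noise, procedural speckling and snowflakes are blended on top, followed by a desaturated tint.
# Parameters grouped per layer control tiling, color, strength and vignette radius.
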Material /Game/Creative/PostProcess/Materials/M_PP_Frost_Simple
graph {
 properties {materialDomain=MD_PostProcess, pixelDepthOffsetMode=PDOM_Legacy, usedShadingModels=Default Lit}
 nodes [
  n1 = VectorParameter() {group=6. Snowflakes, parameterName=Snowflake Color, defaultValue=<rgba 1 1 1 1>}
  n2 = TextureSampleParameter2D() {group=0. Material, parameterName=Noise, texture=/Game/Creative/PostProcess/Textures/T_FrostNoise_Packed}
  n3 = Multiply(A=ComponentMask(None=@n2.RGB) {g=true}, B=ScalarParameter() {group=6. Snowflakes, parameterName=Snowflake Shape Strength})
  n4 = Multiply(A=NamedRerouteUsage(), B=ScalarParameter() {defaultValue=325, group=7. Frost Speckling, parameterName=Frost Speckling Tiling})
  n5 = Constant()
  n6 = Noise!(World Position=AppendVector(A=@n4, B=@n5), FilterWidth)
  n7 = AppendVector(
   A=Multiply(A=@n4, B=ScalarParameter() {defaultValue=0.015, group=7. Frost Speckling, parameterName=Frost Speckling Breakup}) {constB=0.5}
   B=@n5
  )
  n8 = Multiply(
   A=Subtract(A=Round(None=@n6), B=Noise(World Position=@n7))
   B=ScalarParameter() {defaultValue=2, group=7. Frost Speckling, parameterName=Frost Speckling Strength}
  )
  n9 = MaterialFunctionCall!(
   Radius Control=ScalarParameter() {defaultValue=0.35, group=7. Frost Speckling, parameterName=Frost Speckling Vignette Radius Edge Control}
   Radius=ScalarParameter() {defaultValue=40, group=7. Frost Speckling, parameterName=Frost Speckling Vignette Radius}
  ) {materialFunction=/Game/Creative/PostProcess/MaterialFunctions/MF_FrostVignette}
  n10 = SceneTexture(UVs=NamedRerouteUsage()) {sceneTextureId=PPI_PostProcessInput0}
  n11 = VectorParameter() {group=1. Ice Effect, parameterName=Ice Color, defaultValue=<rgba 0.7117 0.7823 0.8281 1>}
  n12 = NamedRerouteUsage()
  n13 = Constant()
  n14 = Add(
   A=Noise(World Position=AppendVector(A=@n12, B=@n13)) {noiseFunction=NOISEFUNCTION_VoronoiALU}
   B=ScalarParameter() {group=0. Material, parameterName=Voronoi Value Boost}
  )
  n15 = ScalarParameter() {defaultValue=0.15, group=0. Material, parameterName=Voronoi Strength}
  n16 = Add(
   A=Add(A=NamedRerouteUsage(), B=ScalarParameter() {defaultValue=0.2, group=0. Material, parameterName=Noise Value Boost, sliderMax=1})
   B=Saturate(None=Multiply(A=@n14, B=@n15))
  )
  n17 = Add(
   A=NamedRerouteUsage()
   B=MaterialFunctionCall(
    Radius Control=ScalarParameter() {defaultValue=0.8, group=0. Material, parameterName=Vignette Radius Edge Control}
    Radius=ScalarParameter() {defaultValue=3, group=0. Material, parameterName=Vignette Radius, sliderMax=3}
   ) {materialFunction=&MF_FrostVignette}.Result
  )
  n18 = NamedRerouteUsage()
  n19 = MaterialFunctionCall(
   Radius Control=ScalarParameter() {defaultValue=0.35, group=1. Ice Effect, parameterName=Ice Vignette Radius Edge Control}
   Radius=ScalarParameter() {defaultValue=3, group=1. Ice Effect, parameterName=Ice Vignette Radius, sliderMax=3}
  ) {materialFunction=&MF_FrostVignette}
  n20 = LinearInterpolate(
   A=ComponentMask(None=SceneTexture() {sceneTextureId=PPI_PostProcessInput0}.Color) {b=true, g=true, r=true}
   B=LinearInterpolate(A=ComponentMask(None=@n10.Color) {b=true, g=true, r=true}, B=Multiply(A=@n11.RGB, B=@n16), Alpha=Saturate(None=@n17))
   Alpha=Saturate(None=Add(A=@n18, B=@n19.Result))
  )
  n21 = NamedRerouteUsage()
  n22 = ScalarParameter() {group=3. Fog Effect, parameterName=Fog Opacity, sliderMax=1}
  n23 = VectorParameter() {group=2. Secondary Noise, parameterName=Secondary Noise Color, defaultValue=<rgba 1.25 1.25 1.25 1>}
  n24 = ScalarParameter() {defaultValue=0.25, group=2. Secondary Noise, parameterName=Secondary Noise Radius Edge Control}
  n25 = ScalarParameter() {defaultValue=10, group=2. Secondary Noise, parameterName=Secondary Noise Radius}
  n26 = Multiply(
   A=Multiply(
    A=MaterialFunctionCall(Radius Control=@n24, Radius=@n25) {materialFunction=&MF_FrostVignette}.Result
    B=NamedRerouteUsage()
   )
   B=ScalarParameter() {defaultValue=2, group=2. Secondary Noise, parameterName=Secondary Noise Intensity}
  )
  n27 = LinearInterpolate(
   A=LinearInterpolate(A=LinearInterpolate(A=@n20, B=@n21, Alpha=@n22), B=Saturate(None=@n23.RGB), Alpha=Saturate(None=@n26))
   B=Constant() {r=2}
   Alpha=Saturate(None=Multiply(A=@n8, B=@n9.Result))
  )
  n28 = ScalarParameter() {defaultValue=2, group=6. Snowflakes, parameterName=Snowflake Alpha Stength}
  n29 = LinearInterpolate(A=@n27, B=Saturate(None=LinearInterpolate(A=@n27, B=@n1.RGB, Alpha=@n3)), Alpha=Saturate(None=Multiply(A=@n2.A, B=@n28)))
  LinearInterpolate(
   A=@n29
   B=Multiply(A=Desaturation(None=@n29), B=VectorParameter() {group=5. Tint, parameterName=Tint Color, defaultValue=<rgba 0.8021 0.9402 1 1>}.RGB)
   Alpha=ScalarParameter() {group=5. Tint, parameterName=Tint Strength}
  )
  n30 = MaterialFunctionCall(Vector 3=Subtract(A=WorldPosition().XYZ)) {materialFunction=/Engine/Functions/Engine_MaterialFunctions02/Utility/VectorLength}
  n31 = ScalarParameter() {defaultValue=2500, group=3. Fog Effect, parameterName=Fog Density}
  n32 = ScreenPosition()
  n33 = NamedRerouteDeclaration(None=ComponentMask(None=SceneTexture(UVs=@n32.ViewportUV) {sceneTextureId=PPI_PostProcessInput0}.Color) {b=true, g=true, r=true}) {name=PostProcessRGB, nodeColor=<rgba 0.3359 0 1 1>}
  n34 = LinearInterpolate(
   A=@n33
   B=VectorParameter() {group=3. Fog Effect, parameterName=Fog Color, defaultValue=<rgba 0.771 0.9546 0.974 1>}.RGB
   Alpha=Clamp(None=Divide(A=@n30.V3 Length, B=@n31))
  ) {constAlpha=1}
  n35 = ScalarParameter() {defaultValue=20000, group=3. Fog Effect, parameterName=Fog Start Distance}
  n36 = NamedRerouteDeclaration(None=Saturate(None=Divide(A=@n30.V3 Length, B=@n35))) {name=Distancealpha, nodeColor=<rgba 0 1 0.4609 1>}
  n37 = ScalarParameter() {group=3. Fog Effect, parameterName=Fog Vignette Radius Edge Control}
  n38 = ScalarParameter() {defaultValue=0.4, group=3. Fog Effect, parameterName=Fog Vignette Radius}
  NamedRerouteDeclaration(None=LinearInterpolate(
    A=@n33
    B=LinearInterpolate(A=@n33, B=@n34, Alpha=@n36)
    Alpha=MaterialFunctionCall(Radius Control=@n37, Radius=@n38) {materialFunction=&MF_FrostVignette}.Result
   )) {name=Fog, nodeColor=<rgba 1 0 0.1641 1>}
  n39 = NamedRerouteUsage()
  n40 = Add(A=DDX(Value=@n39), B=DDY(Value=@n39))
  n41 = ScalarParameter() {group=1. Ice Effect, parameterName=Ice Intensity}
  NamedRerouteDeclaration(None=Subtract(
    A=MaterialFunctionCall() {materialFunction=&ScreenAlignedUVs}."X 100%, Y 100%"
    B=Multiply(A=@n40, B=@n41)
   )) {name=iceUVs, nodeColor=<rgba 0.2891 1 0 1>}
  n42 = NamedRerouteDeclaration(None=MaterialFunctionCall() {materialFunction=&ScreenAlignedUVs}."X 100%, Y Scale to Ratio") {name=screenUVS, nodeColor=<rgba 0.8281 1 0 1>}
  n43 = ScalarParameter() {defaultValue=20, group=0. Material, parameterName=Tiling}
  n44 = ScalarParameter() {group=0. Material, parameterName=U Offset}
  n45 = ScalarParameter() {group=0. Material, parameterName=V Offset}
  n46 = NamedRerouteDeclaration(None=Add(A=Multiply(A=@n42, B=@n43), B=AppendVector(A=@n44, B=@n45))) {name=UVs, nodeColor=<rgba 0 0.3906 1 1>}
  NamedRerouteDeclaration(None=StaticComponentMaskParameter(None=TextureSampleParameter2D(UVs=@n46) {group=0. Material, parameterName=Noise, texture=&T_FrostNoise_Packed}.RGB) {defaultR=true, group=0. Material, parameterName=Noise Channel}) {name=NoiseTexture, nodeColor=<rgba 0.6406 1 0 1>}
  n47 = Multiply(A=NamedRerouteUsage(), B=ScalarParameter() {defaultValue=10, group=2. Secondary Noise, parameterName=Secondary Noise Tiling})
  NamedRerouteDeclaration(None=StaticComponentMaskParameter(None=TextureSampleParameter2D(UVs=@n47) {group=0. Material, parameterName=Noise, texture=&T_FrostNoise_Packed}.RGB) {defaultR=true, group=2. Secondary Noise, parameterName=Secondary Noise Channel}) {name=secondaryNoise, nodeColor=<rgba 1 0 0.3281 1>}
  n48 = TextureSampleParameter2D(UVs=Multiply(A=NamedRerouteUsage(), B=ScalarParameter() {defaultValue=7, group=4. Shape, parameterName=Shape Noise Tiling})) {group=0. Material, parameterName=Noise, texture=&T_FrostNoise_Packed}
  n49 = ScalarParameter() {defaultValue=0.18, group=4. Shape, parameterName=Shape Noise Radius Edge Control}
  n50 = ScalarParameter() {defaultValue=2, group=4. Shape, parameterName=Shape Noise Radius}
  NamedRerouteDeclaration(None=Multiply(
    A=StaticComponentMaskParameter(None=@n48.RGB) {defaultB=true, group=4. Shape, parameterName=Shape Channel}
    B=MaterialFunctionCall(Radius Control=@n49, Radius=@n50) {materialFunction=&MF_FrostVignette}.Result
   )) {name=shapeNoise, nodeColor=<rgba 0 1 0.1797 1>}
 ]
}
