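# Post-process material that lays a tiled halftone pattern over the rendered scene.
# A screen-aligned sphere mask, eased through a sine, varies pattern strength between edge and center parameters, and a tiled grayscale texture turns it into the pattern.
# The scene is both multiplied by the pattern and reduced by its scaled inverse, and the two results are blended.
# Master FX Control fades between the untouched scene and the effect.
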Material /Game/Creative/PostProcess/Materials/M_PP_Halftone
graph {
 properties {materialDomain=MD_PostProcess, pixelDepthOffsetMode=PDOM_Legacy, usedShadingModels=Default Lit}
 nodes [
  n1 = SceneTexture() {sceneTextureId=PPI_PostProcessInput0}
  n2 = MaterialFunctionCall() {materialFunction=&ScreenAlignedUVs}
  n3 = Multiply(A=SphereMask(A=@n2."X 100%, Y Scale to Ratio", B=Constant() {r=0.5}, Radius=Constant() {r=0.707}, Hardness=Constant())) {constB=0.25}
  n4 = ScalarParameter() {defaultValue=150, parameterName=Tiling}
  n5 = LinearInterpolate(
   A=LinearInterpolate(
    A=ScalarParameter() {defaultValue=0.55, parameterName=Pattern Edge}
    B=ScalarParameter() {defaultValue=0.95, parameterName=Pattern Center}
    Alpha=Sine(None=@n3)
   ) {constA=1, constB=0}
   Alpha=TextureSample(UVs=Multiply(A=@n2."X 100%, Y Scale to Ratio", B=@n4)) {samplerType=SAMPLERTYPE_LinearGrayscale, texture=/Game/Creative/PostProcess/Textures/T_PP_Halftone_Pattern}.R
  )
  n6 = Multiply(A=OneMinus(None=@n5), B=ScalarParameter() {defaultValue=0.35, parameterName=Pattern Subtract})
  LinearInterpolate(
   A=SceneTexture() {sceneTextureId=PPI_PostProcessInput0}.Color
   B=LinearInterpolate(A=Multiply(A=@n1.Color, B=@n5), B=Subtract(A=@n1.Color, B=@n6))
   Alpha=ScalarParameter() {parameterName=Master FX Control, sortPriority=0}
  )
 ]
}
