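# Post process material, blended after depth of field, that adds a heat wave distortion to the scene.
# A time-panned cell noise texture has its R and G channels weighted by a remapped vignette mask, then summed into one offset applied to both screen UV axes.
# Scene depth over a fade distance ramps the distorted image in away from the camera.
# Master FX Control, zero by default, blends the whole effect over the original scene.
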Material /Game/Creative/PostProcess/Materials/M_PP_Heatwave
graph {
 properties {
  blendMode=BLEND_Translucent
  blendableLocation=BL_SceneColorAfterDOF
  blendableOutputAlpha=true
  materialDomain=MD_PostProcess
  pixelDepthOffsetMode=PDOM_Legacy
  usedShadingModels=Default Lit
 }
 nodes [
  n1 = SceneTexture(UVs=TextureCoordinate()) {sceneTextureId=PPI_PostProcessInput0}
  n2 = TextureCoordinate()
  n3 = ComponentMask(None=@n2) {r=true}
  n4 = Multiply(A=Time(), B=ScalarParameter() {group=Heat Wave, parameterName=X UV Distort, sliderMax=0.25, sliderMin=-0.25})
  n5 = ComponentMask(None=@n2) {g=true}
  n6 = ScalarParameter() {defaultValue=0.25, group=Heat Wave, parameterName=Y UV Distort Offset, sliderMax=0.25, sliderMin=-0.25, sortPriority=33}
  n7 = Multiply(A=Time(), B=@n6)
  n8 = TextureSampleParameter2D(UVs=AppendVector(A=Add(A=@n3, B=@n4), B=Add(A=@n5, B=@n7))) {group=Heat Wave, parameterName="Heat Wave Texture [R=X, G=Y]", sortPriority=4, texture=/Game/Creative/PostProcess/Textures/Cell_Noise_T1}
  n9 = Saturate(None=MaterialFunctionCall(
    Input=MaterialFunctionCall() {materialFunction=/Game/Creative/PostProcess/MaterialFunctions/MF_VignetteMask}.VignetteMask
    Input Low=Constant()
    Input High=Constant() {r=1}
    Target Low=ScalarParameter() {defaultValue=1, group=Vingette, parameterName=Vingette In Intensity, sliderMax=1, sortPriority=2}
    Target High=ScalarParameter() {defaultValue=1, group=Vingette, parameterName=Vingette Out Intensity, sliderMax=1, sortPriority=1}
   ) {materialFunction=&RemapValueRange}.Result)
  n10 = Multiply(A=@n8.R, B=@n9)
  n11 = Multiply(A=@n8.G, B=@n9)
  n12 = ScalarParameter() {defaultValue=0.005, group=Heat Wave, parameterName=Heat Intensity, sliderMax=0.05, sliderMin=0.001, sortPriority=1}
  n13 = Add(A=TextureCoordinate(), B=Multiply(A=Add(A=@n10, B=@n11), B=@n12) {constB=0.01})
  n14 = Divide(
   A=SceneTexture() {sceneTextureId=PPI_SceneDepth}.Color
   B=ScalarParameter() {defaultValue=500, group=Heat Wave, parameterName=Depth Fade Distance, sliderMax=1000, sortPriority=2}
  ) {constB=5000}
  LinearInterpolate(
   A=@n1.Color
   B=LinearInterpolate(A=@n1.Color, B=SceneTexture(UVs=@n13) {sceneTextureId=PPI_PostProcessInput0}.Color, Alpha=Saturate(None=@n14))
   Alpha=ScalarParameter() {group=Heat Wave, parameterName=Master FX Control, sortPriority=0}
  )
 ]
}
