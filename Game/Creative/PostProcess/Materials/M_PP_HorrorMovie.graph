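# M_PP_HorrorMovie is a post-process material that applies a color tint and gain to the rendered scene, named for a horror movie look.
# It takes the RGB of PostProcessInput0 and multiplies it by the Tint parameter, warm by default, scaled by Gain.
# A lerp driven by the Master FX Control scalar blends from the untouched scene color to the graded result.
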Material /Game/Creative/PostProcess/Materials/M_PP_HorrorMovie
graph {
 properties {materialDomain=MD_PostProcess, pixelDepthOffsetMode=PDOM_Legacy, usedShadingModels=Default Lit}
 nodes [
  n1 = ComponentMask(None=SceneTexture() {sceneTextureId=PPI_PostProcessInput0}.Color) {b=true, g=true, r=true}
  n2 = VectorParameter() {parameterName=Tint, defaultValue=<rgba 1 0.8723 0.7708 1>}
  n3 = ScalarParameter() {defaultValue=1.3, parameterName=Gain}
  LinearInterpolate(A=@n1, B=Multiply(A=Multiply(A=@n2.RGB, B=@n3), B=@n1), Alpha=ScalarParameter() {parameterName=Master FX Control, sortPriority=0})
 ]
}
