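# MF_CMYKConversionToRGB is a material function in the Creative post-process set that converts a CMYK color to RGB.
# It takes one Vector4 input and splits it into CMY (RGB channels) and K (alpha).
# Each of C, M and Y is maxed against K, then inverted with OneMinus, so black acts as a floor on ink amount rather than a multiplicative darkening.
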MaterialFunction /Game/Creative/PostProcess/MaterialFunctions/MF_CMYKConversionToRGB
graph {nodes [
  n1 = FunctionInput() {inputName=CMYK, inputType=FunctionInput_Vector4}
  n2 = ComponentMask(None=@n1) {b=true, g=true, r=true}
  n3 = ComponentMask(None=@n1) {a=true}
  FunctionOutput(None=OneMinus(None=Max(A=@n2, B=@n3)))
 ]}
