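# Post-process material function that draws a comic halftone dot pattern from the scene color for the ink channel selected by a CMYK mask input.
# It builds an aspect-corrected screen grid at an input rotation, scaled by the Grid Resolution parameter.
# Scene color sampled at each cell center and its four direct neighbors converts to CMYK and sets dot radius.
# A Min merges the five soft-edged dots so they overlap across cell borders, then inverts.
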MaterialFunction /Game/Creative/PostProcess/MaterialFunctions/MF_ComicDots
graph {nodes [
  n1 = Rotator(Coordinate=NamedRerouteUsage(), Time=Multiply(A=NamedRerouteUsage()) {constB=-1}) {centerX=0, centerY=0, speed=0.0174}
  n2 = Multiply(A=Distance(A=NamedRerouteUsage(), B=@n1)) {constB=1.414}
  n3 = NamedRerouteUsage()
  n4 = SceneTexture(UVs=Divide(A=Divide(A=@n1, B=@n3), B=NamedRerouteUsage())) {sceneTextureId=PPI_PostProcessInput0}
  n5 = DotProduct(
   A=MaterialFunctionCall!(RGB=ComponentMask(None=@n4.Color) {b=true, g=true, r=true}) {materialFunction=&/MF_RGBConversionToCMYK}.CMYK
   B=NamedRerouteUsage()
  )
  n6 = Saturate(None=Divide(A=Subtract(A=@n2, B=@n5)) {constB=0.01})
  n7 = NamedRerouteUsage()
  n8 = Rotator(Coordinate=Add(A=NamedRerouteUsage(), B=Constant2Vector() {r=1}), Time=Multiply(A=NamedRerouteUsage()) {constB=-1}) {centerX=0, centerY=0, speed=0.0174}
  n9 = Divide(A=@n8, B=NamedRerouteUsage())
  n10 = NamedRerouteUsage()
  n11 = ComponentMask(None=SceneTexture(UVs=Divide(A=@n9, B=@n10)) {sceneTextureId=PPI_PostProcessInput0}.Color) {b=true, g=true, r=true}
  n12 = Subtract(
   A=Multiply(A=Distance(A=@n7, B=@n8)) {constB=1.414}
   B=DotProduct(
    A=MaterialFunctionCall(RGB=@n11) {materialFunction=&MF_RGBConversionToCMYK}.CMYK
    B=NamedRerouteUsage()
   )
  )
  n13 = Rotator(Coordinate=Add(A=NamedRerouteUsage(), B=Constant2Vector() {r=-1}), Time=Multiply(A=NamedRerouteUsage()) {constB=-1}) {centerX=0, centerY=0, speed=0.0174}
  n14 = Distance(A=NamedRerouteUsage(), B=@n13)
  n15 = Divide(A=Divide(A=@n13, B=NamedRerouteUsage()), B=NamedRerouteUsage())
  n16 = MaterialFunctionCall(RGB=ComponentMask(None=SceneTexture(UVs=@n15) {sceneTextureId=PPI_PostProcessInput0}.Color) {b=true, g=true, r=true}) {materialFunction=&MF_RGBConversionToCMYK}
  n17 = NamedRerouteUsage()
  n18 = Divide(A=Subtract(A=Multiply(A=@n14) {constB=1.414}, B=DotProduct(A=@n16.CMYK, B=@n17))) {constB=0.01}
  n19 = Rotator(Coordinate=Add(A=NamedRerouteUsage(), B=Constant2Vector() {g=1}), Time=Multiply(A=NamedRerouteUsage()) {constB=-1}) {centerX=0, centerY=0, speed=0.0174}
  n20 = Multiply(A=Distance(A=NamedRerouteUsage(), B=@n19)) {constB=1.414}
  n21 = NamedRerouteUsage()
  n22 = SceneTexture(UVs=Divide(A=Divide(A=@n19, B=@n21), B=NamedRerouteUsage())) {sceneTextureId=PPI_PostProcessInput0}
  n23 = DotProduct(
   A=MaterialFunctionCall(RGB=ComponentMask(None=@n22.Color) {b=true, g=true, r=true}) {materialFunction=&MF_RGBConversionToCMYK}.CMYK
   B=NamedRerouteUsage()
  )
  n24 = Saturate(None=Divide(A=Subtract(A=@n20, B=@n23)) {constB=0.01})
  n25 = Rotator(Coordinate=Add(A=NamedRerouteUsage(), B=Constant2Vector() {g=-1}), Time=Multiply(A=NamedRerouteUsage()) {constB=-1}) {centerX=0, centerY=0, speed=0.0174}
  n26 = Multiply(A=Distance(A=NamedRerouteUsage(), B=@n25)) {constB=1.414}
  n27 = NamedRerouteUsage()
  n28 = SceneTexture(UVs=Divide(A=Divide(A=@n25, B=@n27), B=NamedRerouteUsage())) {sceneTextureId=PPI_PostProcessInput0}
  n29 = DotProduct(
   A=MaterialFunctionCall(RGB=ComponentMask(None=@n28.Color) {b=true, g=true, r=true}) {materialFunction=&MF_RGBConversionToCMYK}.CMYK
   B=NamedRerouteUsage()
  )
  n30 = Saturate(None=Divide(A=Subtract(A=@n26, B=@n29)) {constB=0.01})
  n31 = Min(A=Saturate(None=Divide(A=@n12) {constB=0.01}), B=Min(A=Saturate(None=@n18), B=Min(A=@n24, B=@n30)))
  FunctionOutput(None=OneMinus(None=Min(A=@n6, B=@n31)))
  n32 = Multiply(A=TextureCoordinate(), B=NamedRerouteUsage())
  n33 = NamedRerouteUsage()
  n34 = Rotator(Coordinate=NamedRerouteDeclaration(None=Multiply(A=@n32, B=@n33)) {name=scaledUVs, nodeColor=<rgba 1 0 0.2109 1>}, Time=NamedRerouteUsage()) {centerX=0, centerY=0, speed=0.0174}
  NamedRerouteDeclaration(None=Add(A=Floor(None=@n34)) {constB=0.5}) {name=cellCord, nodeColor=<rgba 0.3828 1 0 1>}
  n35 = ViewSize()
  n36 = ComponentMask(None=@n35) {r=true}
  n37 = ComponentMask(None=@n35) {g=true}
  NamedRerouteDeclaration(None=AppendVector(A=Divide(A=@n36, B=@n37), B=Constant() {r=1})) {name=aspectRatio, nodeColor=<rgba 0.8281 1 0 1>}
  NamedRerouteDeclaration(None=ScalarParameter() {defaultValue=200, group=2. Comic Dots, parameterName=Grid Resolution}) {name=gridResolution, nodeColor=<rgba 1 0 0.6328 1>}
  NamedRerouteDeclaration(None=FunctionInput() {inputName=gridRotation, inputType=FunctionInput_Scalar, sortPriority=1}) {name=gridRotation, nodeColor=<rgba 1 0 0.2344 1>}
  NamedRerouteDeclaration(None=FunctionInput() {inputName=CMYK Mask, inputType=FunctionInput_Vector4, sortPriority=2}) {name=CMYKMask, nodeColor=<rgba 0 0.4844 1 1>}
 ]}
