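# Post-process material that gives the rendered scene a comic-book print look with paper texture, halftone dots and outlines.
# Four MF_ComicDots passes, one per CMYK channel at its own screen angle, recombine to RGB as the halftone layer.
# Paper, dots and MF_Outline layer through strength lerps, with the outline taking the paper-and-dots composite as input.
# MasterVFXAdjustment, zero by default, fades between the untouched scene and the full effect.
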Material /Game/Creative/PostProcess/Materials/M_PP_NeoComic_V2
graph {
 properties {materialDomain=MD_PostProcess, pixelDepthOffsetMode=PDOM_Legacy, usedShadingModels=Default Lit}
 nodes [
  n1 = NamedRerouteUsage()
  n2 = LinearInterpolate(
   A=LinearInterpolate(
    A=@n1
    B=MaterialFunctionCall() {materialFunction=/Game/Creative/PostProcess/MaterialFunctions/MF_Paper}.PaperTexture
    Alpha=ScalarParameter() {defaultValue=0.6, group=0. VFX Adjustments, parameterName=PaperVFXStrength}
   )
   B=NamedRerouteUsage()
   Alpha=ScalarParameter() {defaultValue=0.35, group=0. VFX Adjustments, parameterName=ComicVFXStrength}
  )
  LinearInterpolate(
   A=@n1
   B=LinearInterpolate(
    A=@n2
    B=MaterialFunctionCall!(SceneTexture=@n2) {materialFunction=/Game/Creative/PostProcess/MaterialFunctions/MF_Outline/MF_Outline}.Color
    Alpha=ScalarParameter() {defaultValue=0.65, group=0. VFX Adjustments, parameterName=OutlineVFXStrength}
   ) {constAlpha=0}
   Alpha=ScalarParameter() {group=0. VFX Adjustments, parameterName=MasterVFXAdjustment, sortPriority=1}
  )
  n3 = MaterialFunctionCall!(gridRotation=Constant() {r=15}, CMYK Mask=Constant4Vector() {constant=<rgba 1 0 0 0>}.RGBA) {materialFunction=/Game/Creative/PostProcess/MaterialFunctions/MF_ComicDots}
  n4 = MaterialFunctionCall(gridRotation=Constant() {r=75}, CMYK Mask=Constant4Vector() {constant=<rgba 0 1 0 0>}.RGBA) {materialFunction=&MF_ComicDots}
  n5 = MaterialFunctionCall(gridRotation=Constant(), CMYK Mask=Constant4Vector() {constant=<rgba 0 0 1 0>}.RGBA) {materialFunction=&MF_ComicDots}
  n6 = MaterialFunctionCall(gridRotation=Constant() {r=45}, CMYK Mask=Constant4Vector() {constant=<rgba 0 0 0 1>}.RGBA) {materialFunction=&MF_ComicDots}
  NamedRerouteDeclaration(None=MaterialFunctionCall!(CMYK=MaterialFunctionCall(R=@n3.Result, G=@n4.Result, B=@n5.Result, A=@n6.Result) {materialFunction=&AppendMany}.RGBA) {materialFunction=/Game/Creative/PostProcess/MaterialFunctions/MF_CMYKConversionToRGB}.Result) {name=comicDots, nodeColor=<rgba 1 0.6563 0 1>}
  NamedRerouteDeclaration(None=ComponentMask(None=SceneTexture() {sceneTextureId=PPI_PostProcessInput0}.Color) {b=true, g=true, r=true}) {name=CameraDiffuse, nodeColor=<rgba 0 0.6953 1 1>}
 ]
}
