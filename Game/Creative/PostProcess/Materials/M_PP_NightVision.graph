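# Post-process material that turns the scene into a tinted night-vision view, blended in by Master FX Control, which defaults to zero.
# Desaturated scene color combines with the tint through summed Linear Light and Pin Light blends.
# Named reroutes carry this image, an aspect-corrected radial vignette, panning scanlines that ignore pause, and sine-hash UV white noise into the final composite.
# Scalar opacities control the layers, with scanlines and noise subtle by default.
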Material /Game/Creative/PostProcess/Materials/M_PP_NightVision
graph {
 properties {materialDomain=MD_PostProcess, pixelDepthOffsetMode=PDOM_Legacy, usedShadingModels=Default Lit}
 nodes [
  n1 = ViewProperty() {property=MEVP_ViewSize}
  n2 = ComponentMask(None=@n1.Property) {r=true}
  n3 = ComponentMask(None=@n1.Property) {g=true}
  n4 = AppendVector(A=Min(A=Divide(A=@n2, B=@n3)), B=Min(A=Divide(A=@n3, B=@n2)))
  n5 = Add(A=Multiply(A=ScreenPosition().ViewportUV, B=@n4), B=Divide(A=OneMinus(None=@n4)))
  n6 = Constant2Vector() {g=0.5, r=0.5}
  n7 = OneMinus(None=Divide(A=Distance(A=@n5, B=@n6), B=ScalarParameter() {defaultValue=0.75, group=2. Vignette, parameterName=Scale}))
  AppendVector(A=@n7, B=Constant() {r=1})
  n8 = Multiply(A=NamedRerouteUsage(), B=NamedRerouteUsage())
  n9 = Multiply(A=@n8, B=NamedRerouteUsage())
  n10 = ScalarParameter() {defaultValue=0.1, group=0. Material Features, parameterName=Scanline Opacity}
  LinearInterpolate(
   A=ComponentMask(None=SceneTexture() {sceneTextureId=PPI_PostProcessInput0}.Color) {b=true, g=true, r=true}
   B=LinearInterpolate(
    A=LinearInterpolate(A=@n8, B=@n9, Alpha=@n10)
    B=NamedRerouteUsage()
    Alpha=ScalarParameter() {defaultValue=0.15, group=0. Material Features, parameterName=WhiteNoise Opacity}
   )
   Alpha=ScalarParameter() {group=0. Material Features, parameterName=Master FX Control, sortPriority=0}
  )
  n11 = TextureCoordinate()
  NamedRerouteDeclaration(None=ComponentMask(None=@n11) {r=true}) {name=red, nodeColor=<rgba 0.2656 1 0 1>}
  n12 = ComponentMask(None=SceneTexture() {sceneTextureId=PPI_PostProcessInput0}.Color) {b=true, g=true, r=true}
  n13 = Desaturation(None=@n12)
  n14 = Multiply(
   A=VectorParameter() {group=1. Night Vision, parameterName=Night Vision Color, defaultValue=<rgba 0.0273 0.1912 0.0003 1>}.RGB
   B=ScalarParameter() {defaultValue=1, group=1. Night Vision, parameterName=Nightvision Strength}
  )
  n15 = MaterialFunctionCall!(Base=@n13, Blend=@n14) {materialFunction=/Engine/Functions/Engine_MaterialFunctions03/Blends/Blend_LinearLight}
  n16 = MaterialFunctionCall!(Base=@n13, Blend=@n14) {materialFunction=/Engine/Functions/Engine_MaterialFunctions03/Blends/Blend_PinLight}
  NamedRerouteDeclaration(None=LinearInterpolate(
    A=@n12
    B=Add(A=@n15.Result, B=@n16.Result)
    Alpha=ScalarParameter() {defaultValue=1, group=0. Material Features, parameterName=NightVision Opacity}
   )) {name=nightVision, nodeColor=<rgba 0.6641 1 0 1>}
  NamedRerouteDeclaration(None=ComponentMask(None=@n11) {g=true}) {name=green, nodeColor=<rgba 0 1 0 1>}
  NamedRerouteDeclaration(None=LinearInterpolate(B=@n7, Alpha=ScalarParameter() {defaultValue=1.3, group=0. Material Features, parameterName=VignetteAmount}) {constA=1}) {name=Vignette, nodeColor=<rgba 0.7109 0 1 1>}
  n17 = Panner(
   Coordinate=ScreenPosition().ViewportUV
   Time=Time() {bIgnorePause=true}
   Speed=AppendVector(A=Constant(), B=ScalarParameter() {defaultValue=0.02, group=3. Scanlines, parameterName=Scanline Speed})
  ) {speedY=0.03}
  n18 = ScalarParameter() {defaultValue=50, group=3. Scanlines, parameterName=Scanline Amount}
  n19 = Frac(None=ComponentMask(None=Multiply(A=@n17, B=@n18) {constB=200}) {g=true})
  n20 = Add(A=Multiply(A=@n19, B=OneMinus(None=@n19)), B=ScalarParameter() {defaultValue=-0.2, group=3. Scanlines, parameterName=ScanlineSize})
  NamedRerouteDeclaration(None=LinearInterpolate(
    A=ScalarParameter() {defaultValue=0.1, group=3. Scanlines, parameterName=Scanlines Dark}
    B=ScalarParameter() {defaultValue=0.5, group=3. Scanlines, parameterName=Scanlines Light}
    Alpha=Ceil(None=@n20)
   )) {name=scanLines, nodeColor=<rgba 1 0.9609 0 1>}
  n21 = NamedRerouteUsage()
  n22 = MaterialFunctionCall(X=@n21, Y=@n21, Z=NamedRerouteUsage()) {materialFunction=&MakeFloat3}
  n23 = Constant3Vector() {constant=<rgba 127.1 311.7 74.7 0>}
  n24 = NamedRerouteUsage()
  n25 = MaterialFunctionCall(X=NamedRerouteUsage(), Y=@n24, Z=@n24) {materialFunction=&MakeFloat3}
  n26 = Constant3Vector() {constant=<rgba 269.5 183.3 246.1 0>}
  n27 = NamedRerouteUsage()
  n28 = MaterialFunctionCall(X=NamedRerouteUsage(), Y=@n27, Z=@n27) {materialFunction=&MakeFloat3}
  n29 = Constant3Vector() {constant=<rgba 114 272 126 0>}
  n30 = Sine(None=MaterialFunctionCall(X=DotProduct(A=@n22.Result, B=@n23), Y=DotProduct(A=@n25.Result, B=@n26), Z=DotProduct(A=@n28.Result, B=@n29)) {materialFunction=&MakeFloat3}.Result)
  n31 = Constant() {r=1400}
  NamedRerouteDeclaration(None=Frac(None=Multiply(A=@n30, B=@n31))) {name=whiteNoise, nodeColor=<rgba 0 1 0.3906 1>}
 ]
}
