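# Post-process material that gives the scene an old-cartoon film look, blended in by Master FX Control.
# Desaturated, tinted, contrast-adjusted color is multiplied by distance-fading depth-edge outlines, sampling four or eight neighbors by quality level.
# A jittering vignette, flipbook noise and detail, and scrolling vertical scratches from one packed texture layer on top.
# Side branches for base-color shading, world-aligned AO noise and alternate edges feed nothing visible in the graph.
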Material /Game/Creative/PostProcess/Materials/M_PP_OldCartoon
graph {
 properties {blendableLocation=BL_SceneColorAfterDOF, materialDomain=MD_PostProcess, pixelDepthOffsetMode=PDOM_Legacy, usedShadingModels=Default Lit}
 nodes [
  n1 = Subtract(A=MaterialFunctionCall() {materialFunction=&ScreenAlignedUVs}."X 100%, Y 100%") {constB=0.5}
  n2 = Power(Base=Multiply(A=Abs(None=@n1)) {constB=2}) {constExponent=3}
  n3 = Multiply(A=Time() {bIgnorePause=true}, B=ScalarParameter() {defaultValue=1, group=Vignette, parameterName=Vignette Speed})
  n4 = Sine(None=@n3)
  n5 = Sine(None=@n3) {period=0.3}
  n6 = Add(A=Add(A=Multiply(A=@n4) {constB=0.1}, B=Multiply(A=@n5) {constB=0.1}), B=Multiply(A=Sine(None=@n3) {period=0.1}) {constB=0.1})
  n7 = Multiply(A=Sine(None=@n3) {period=0.21}) {constB=0.1}
  n8 = Add(
   A=ScalarParameter() {defaultValue=0.7, group=Vignette, parameterName=Vignette}
   B=Multiply(A=Add(A=@n6, B=@n7), B=ScalarParameter() {defaultValue=0.1, group=Vignette, parameterName=Vignette Jitter}) {constB=0.1}
  )
  n9 = OneMinus(None=Multiply(A=ComponentMask(None=@n2) {g=true}, B=@n8))
  n10 = OneMinus(None=Multiply(A=ComponentMask(None=@n2) {r=true}, B=@n8))
  n11 = Multiply(A=Time() {bIgnorePause=true}, B=ScalarParameter() {defaultValue=24, group=Color, parameterName=Frame Speed})
  n12 = Subtract(A=@n11, B=Frac(None=@n11) {desc=Frame Lerp}) {desc=IndexA}
  n13 = Constant2Vector() {g=8, r=8}
  n14 = Divide(A=Constant2Vector() {g=1, r=1}, B=@n13) {desc=Scale}
  n15 = ComponentMask(None=@n14) {desc="## Horz Images", r=true}
  n16 = AppendVector(
   A=Fmod(A=@n12, B=ComponentMask(None=@n13) {desc="## Horz Images", r=true}) {desc=IndexA_Horz}
   B=Floor(None=Multiply(A=@n12, B=@n15)) {desc=IndexA_Vert}
  ) {desc=IndexA_Combined}
  n17 = TextureCoordinate()
  n18 = TextureSample(UVs=Multiply(A=Add(A=@n16, B=@n17), B=@n14) {desc=UV_A}) {samplerType=SAMPLERTYPE_LinearColor, texture=/Game/Creative/PostProcess/Textures/T_OldCartoon_M}
  n19 = Multiply(
   A=LinearInterpolate(B=@n18.R, Alpha=ScalarParameter() {defaultValue=0.15, group=Color, parameterName=Noise Strength}) {constA=1}
   B=LinearInterpolate(B=@n18.G, Alpha=ScalarParameter() {defaultValue=0.4, group=Color, parameterName=Detail Strength}) {constA=1}
  )
  n20 = TextureCoordinate()
  n21 = Time() {bIgnorePause=true}
  n22 = AppendVector(
   A=Add(A=ComponentMask(None=@n20) {r=true}, B=Multiply(A=@n21) {constB=0.02})
   B=Multiply(A=@n21, B=ScalarParameter() {defaultValue=0.05, group=Vertical Lines, parameterName=Vertical Line Speed})
  )
  n23 = ComponentMask(None=Desaturation(
    None=SceneTexture() {sceneTextureId=PPI_PostProcessInput0}.Color
    Fraction=ScalarParameter() {defaultValue=1, group=Color, parameterName=Desaturation}
   )) {b=true, g=true, r=true}
  n24 = VectorParameter() {group=Color, parameterName=Color Tint, defaultValue=<rgba 1 1 1 0>}
  n25 = Multiply(
   A=Power(Base=Multiply(A=@n23, B=@n24.RGB), Exp=ScalarParameter() {defaultValue=1.4, group=Color, parameterName=Power Exp})
   B=ScalarParameter() {defaultValue=2, group=Color, parameterName=Intensity}
  )
  n26 = ScalarParameter() {defaultValue=3, group=Lines, parameterName=Line Thickness Near}
  n27 = ScalarParameter() {group=Lines, parameterName=Line Thickness Far}
  n28 = ComponentMask(None=SceneDepth().R) {r=true}
  n29 = Multiply(A=ScalarParameter() {defaultValue=15, group=Lines, parameterName="Scene Depth Bias (m)"}) {constB=100}
  n30 = ScalarParameter() {defaultValue=30, group=Lines, parameterName="Scene Depth Distance (m)"}
  n31 = Saturate(None=Divide(A=Subtract(A=@n28, B=@n29), B=Multiply(A=@n30) {constB=100}))
  n32 = Multiply(A=Max(A=LinearInterpolate(A=@n26, B=@n27, Alpha=@n31)) {constB=0}, B=SceneTexelSize()) {constA=2}
  n33 = ComponentMask(None=@n32) {desc=x, r=true}
  n34 = Constant()
  n35 = AppendVector(A=@n33, B=@n34)
  n36 = ComponentMask(None=ScreenPosition().ViewportUV) {g=true, r=true}
  n37 = SceneDepth(UVs=Add(A=@n36, B=@n35) {desc=x+})
  n38 = SceneDepth()
  n39 = ComponentMask(None=@n32) {desc=y, g=true}
  n40 = AppendVector(A=@n34, B=@n39) {desc=y+}
  n41 = SceneDepth(UVs=Add(A=@n36, B=@n40) {desc=y+})
  n42 = Constant() {r=-1}
  n43 = SceneDepth(UVs=Add(A=@n36, B=Multiply(A=@n35, B=@n42)) {desc=x-})
  n44 = SceneDepth(UVs=Add(A=@n36, B=Multiply(A=@n42, B=@n40)) {desc=y-})
  n45 = Add(A=Add(A=Subtract(A=@n37.R, B=@n38.R), B=Subtract(A=@n41.R, B=@n38.R)), B=Add(A=Subtract(A=@n43.R, B=@n38.R), B=Subtract(A=@n44.R, B=@n38.R)))
  n46 = AppendVector(A=@n33, B=@n39)
  n47 = SceneDepth(UVs=Add(A=@n36, B=@n46) {desc=x+})
  n48 = Subtract(A=@n47.R, B=@n38.R)
  n49 = SceneDepth(UVs=Add(A=@n36, B=Multiply(A=@n46) {constB=-1}) {desc=y+})
  n50 = Subtract(A=@n49.R, B=@n38.R)
  n51 = Constant2Vector() {g=1, r=-1}
  n52 = SceneDepth(UVs=Add(A=@n36, B=Multiply(A=@n46, B=@n51) {constB=-1}) {desc=x-})
  n53 = Subtract(A=@n52.R, B=@n38.R)
  n54 = Constant2Vector() {g=-1, r=1}
  n55 = SceneDepth(UVs=Add(A=@n36, B=Multiply(A=@n46, B=@n54) {constB=-1}) {desc=y-})
  n56 = Subtract(A=@n55.R, B=@n38.R)
  n57 = Add(A=@n45, B=Add(A=Add(A=@n48, B=@n50), B=Add(A=@n53, B=@n56)))
  n58 = ScalarParameter() {defaultValue=300, parameterName=Divisor}
  n59 = Saturate(None=Divide(A=@n57, B=@n58) {constB=100})
  n60 = Saturate(None=Divide(A=@n45, B=@n58) {constB=100})
  n61 = OneMinus(None=Multiply(A=QualitySwitch(Default=@n59, Low=@n60, High=@n59, Medium=@n60, Epic=@n59), B=OneMinus(None=@n31)))
  n62 = Multiply(
   A=OneMinus(None=TextureSample(UVs=@n22) {samplerType=SAMPLERTYPE_LinearColor, texture=&T_OldCartoon_M}.B)
   B=Multiply(A=Clamp(None=@n25), B=ComponentMask(None=@n61) {r=true})
  )
  LinearInterpolate(
   A=ComponentMask(None=SceneTexture() {sceneTextureId=PPI_PostProcessInput0}.Color) {b=true, g=true, r=true}
   B=Multiply(A=Multiply(A=@n9, B=@n10), B=Multiply(A=@n19, B=@n62))
   Alpha=ScalarParameter() {parameterName=Master FX Control, sortPriority=0}
  )
  n63 = Constant() {r=1}
  n64 = Divide(
   A=Desaturation(None=SceneTexture() {sceneTextureId=PPI_PostProcessInput0}.Color, Fraction=@n63)
   B=Desaturation(None=SceneTexture() {sceneTextureId=PPI_DiffuseColor}.Color, Fraction=@n63)
  )
  n65 = ScalarParameter() {defaultValue=1, parameterName=Color Scale}
  n66 = LinearInterpolate(A=ScalarParameter() {parameterName=Shadow Low}, Alpha=Saturate(None=Divide(A=@n64, B=@n65)))
  n67 = SceneTexture() {sceneTextureId=PPI_SceneDepth}
  n68 = Saturate(None=Subtract(A=ComponentMask(None=@n67.Color) {r=true}, B=ScalarParameter() {parameterName=SkyDepth}))
  LinearInterpolate(
   B=Multiply(A=@n66, B=SceneTexture() {sceneTextureId=PPI_StoredBaseColor}.Color)
   Alpha=Multiply(A=ScalarParameter() {parameterName=Base Color, sliderMax=1}, B=OneMinus(None=@n68))
  )
  n69 = TextureObject() {samplerType=SAMPLERTYPE_Masks, texture=/Game/Global/Textures/Default/Tileable_Patterns/OakPP_Post}
  n70 = ScalarParameter() {defaultValue=800, parameterName=Size}
  n71 = ComponentMask(None=SceneTexture() {sceneTextureId=PPI_WorldNormal}.Color) {b=true, g=true, r=true}
  n72 = TextureObject() {samplerType=SAMPLERTYPE_Masks, texture=&OakPP_Post}
  n73 = ScalarParameter() {defaultValue=8000, parameterName=Size2}
  n74 = ComponentMask(None=SceneTexture() {sceneTextureId=PPI_WorldNormal}.Color) {b=true, g=true, r=true}
  n75 = Divide(
   A=Add(A=SceneDepth().R, B=ScalarParameter() {defaultValue=-3000, parameterName=DistanceOffset})
   B=ScalarParameter() {defaultValue=5000, parameterName=Distance}
  )
  n76 = ComponentMask(None=LinearInterpolate(
    A=MaterialFunctionCall(TextureObject=@n69, TextureSize=@n70, World Space Normal=@n71) {materialFunction=&WorldAlignedTexture}.XYZ Texture
    B=MaterialFunctionCall(TextureObject=@n72, TextureSize=@n73, World Space Normal=@n74) {materialFunction=&WorldAlignedTexture}.XYZ Texture
    Alpha=Saturate(None=@n75)
   )) {r=true}
  n77 = LinearInterpolate(
   A=ScalarParameter() {parameterName=Ambient Occlusion Low}
   B=ScalarParameter() {defaultValue=1, parameterName=Ambient Occlusion High}
   Alpha=ComponentMask(None=SceneTexture() {sceneTextureId=PPI_AmbientOcclusion}.Color) {r=true}
  )
  n78 = Multiply(A=OneMinus(None=Saturate(None=@n76)), B=OneMinus(None=Saturate(None=@n77)))
  n79 = ScalarParameter() {defaultValue=0.9, parameterName=DarkenNoise}
  OneMinus(None=Saturate(None=Multiply(A=@n78, B=@n79)))
  n80 = ScalarParameter() {defaultValue=4000, parameterName=Divisor 2}
  n81 = SceneDepth()
  n82 = Subtract(A=@n81.R, B=@n37.R)
  n83 = Subtract(A=@n81.R, B=@n41.R)
  n84 = Subtract(A=@n81.R, B=@n43.R)
  n85 = Subtract(A=@n81.R, B=@n44.R)
  n86 = Subtract(A=@n81.R, B=@n47.R)
  n87 = Subtract(A=@n81.R, B=@n49.R)
  n88 = Subtract(A=@n81.R, B=@n52.R)
  n89 = Subtract(A=@n81.R, B=@n55.R)
  n90 = Add(A=Add(A=Add(A=@n82, B=@n83), B=Add(A=@n84, B=@n85)), B=Add(A=Add(A=@n86, B=@n87), B=Add(A=@n88, B=@n89)))
  n91 = Add(A=Saturate(None=Divide(A=@n57, B=@n80) {constB=100}), B=Saturate(None=Divide(A=@n90, B=@n80) {constB=100}))
  QualitySwitch(Default=@n91, Low=@n60, High=@n91, Medium=@n59, Epic=@n91)
  Saturate(None=@n66)
 ]
}
