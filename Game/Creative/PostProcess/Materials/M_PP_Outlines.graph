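# Post-process material that draws coloured outlines over the scene after depth of field, blended by a Master FX Control parameter.
# Four-neighbour differences in scene depth and world normals form two switchable edge masks, with depth line thickness stepping between near and far values at a distance threshold.
# The masks combine by maximum, but mobile uses only depth lines.
# Outlines fade out across a depth range, and a Debug switch shows them on white.
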Material /Game/Creative/PostProcess/Materials/M_PP_Outlines
graph {
 properties {
  bAutomaticallySetUsageInEditor=false
  blendableLocation=BL_SceneColorAfterDOF
  blendablePriority=1
  materialDomain=MD_PostProcess
  pixelDepthOffsetMode=PDOM_Legacy
  usedShadingModels=Default Lit
 }
 nodes [
  n1 = VectorParameter() {parameterName=Outline Colour, defaultValue=<rgba 0 0 0 1>}
  n2 = Constant()
  n3 = NamedRerouteUsage()
  n4 = Max(
   A=StaticSwitchParameter(True=NamedRerouteUsage(), False=@n2) {defaultValue=true, parameterName=WN Lines}
   B=StaticSwitchParameter(True=@n3, False=@n2) {defaultValue=true, parameterName=SD Lines}
  )
  n5 = LinearInterpolate(
   B=LinearInterpolate(Alpha=ShadingPathSwitch(Default=@n4, Mobile=@n3) {desc=Migrated from FeatureLevelSwitch}) {desc="Pre TM Adjust (if needed)"}
   Alpha=@n1.A
  )
  n6 = SceneTexture() {sceneTextureId=PPI_PostProcessInput0}
  n7 = StaticSwitchParameter(
   True=LinearInterpolate(B=@n1.RGB, Alpha=@n5) {constA=1}
   False=LinearInterpolate(A=ComponentMask(None=@n6.Color) {b=true, g=true, r=true}, B=@n1.RGB, Alpha=@n5)
  ) {parameterName=Debug}
  n8 = MaterialFunctionCall!(
   Input=SceneDepth().R
   Input Low=ScalarParameter() {defaultValue=8000, parameterName=Near Range}
   Input High=ScalarParameter() {defaultValue=12000, parameterName=Far Range}
  ) {materialFunction=/Engine/Functions/Engine_MaterialFunctions03/Math/RemapValueRangeNormalized}
  n9 = ComponentMask(None=SceneTexture() {sceneTextureId=PPI_PostProcessInput0}.Color) {b=true, g=true, r=true}
  LinearInterpolate(
   A=ComponentMask(None=SceneTexture() {sceneTextureId=PPI_PostProcessInput0}.Color) {b=true, g=true, r=true}
   B=StaticSwitchParameter(True=LinearInterpolate(A=@n7, Alpha=@n8.Result), False=LinearInterpolate(A=@n7, B=@n9, Alpha=@n8.Result)) {parameterName=Debug}
   Alpha=ScalarParameter() {group=Main, parameterName=Master FX Control, sortPriority=0}
  )
  n10 = SceneDepth()
  n11 = Multiply(A=SceneTexture() {sceneTextureId=PPI_SceneDepth}.InvSize, B=Constant2Vector() {g=1})
  n12 = Constant()
  n13 = LinearInterpolate(
   A=ScalarParameter() {defaultValue=1, parameterName=SD Far Thickness}
   B=ScalarParameter() {defaultValue=2, parameterName=SD Near Thickness}
   Alpha=If(
    A=SceneTexture() {sceneTextureId=PPI_SceneDepth}.Color
    B=ScalarParameter() {defaultValue=10000, group=Ranges, parameterName=Line thickness distance}
    "A > B"=@n12
    "A == B"=@n12
    "A < B"=Constant() {r=1}
   )
  ) {constA=1, constB=10}
  n14 = SceneDepth(UVs=Add(A=TextureCoordinate(), B=Multiply(A=@n11, B=@n13)))
  n15 = SceneDepth()
  n16 = Multiply(A=SceneTexture() {sceneTextureId=PPI_SceneDepth}.InvSize, B=Constant2Vector() {g=-1})
  n17 = SceneDepth(UVs=Add(A=TextureCoordinate(), B=Multiply(A=@n16, B=@n13)))
  n18 = SceneDepth()
  n19 = Multiply(A=SceneTexture() {sceneTextureId=PPI_SceneDepth}.InvSize, B=Constant2Vector() {r=1})
  n20 = SceneDepth(UVs=Add(A=TextureCoordinate(), B=Multiply(A=@n19, B=@n13)))
  n21 = SceneDepth()
  n22 = Multiply(A=SceneTexture() {sceneTextureId=PPI_SceneDepth}.InvSize, B=Constant2Vector() {r=-1})
  n23 = SceneDepth(UVs=Add(A=TextureCoordinate(), B=Multiply(A=@n22, B=@n13)))
  n24 = Add(A=Add(A=Subtract(A=@n10.R, B=@n14.R), B=Subtract(A=@n15.R, B=@n17.R)), B=Add(A=Subtract(A=@n18.R, B=@n20.R), B=Subtract(A=@n21.R, B=@n23.R)))
  n25 = ScalarParameter() {defaultValue=1, parameterName=SD Threshold}
  NamedRerouteDeclaration(None=Saturate(None=Divide(A=@n24, B=@n25))) {name=SD Lines, nodeColor=<rgba 0.7812 0 1 1>}
  n26 = TextureCoordinate()
  n27 = ScalarParameter() {defaultValue=1, parameterName=WN Thickness}
  n28 = Multiply(A=Multiply(A=SceneTexture() {sceneTextureId=PPI_SceneDepth}.InvSize, B=Constant2Vector() {g=1}), B=@n27)
  n29 = Subtract(A=SceneTexture() {sceneTextureId=PPI_WorldNormal}.Color, B=SceneTexture(UVs=Add(A=@n26, B=@n28)) {sceneTextureId=PPI_WorldNormal}.Color)
  n30 = TextureCoordinate()
  n31 = Multiply(A=Multiply(A=SceneTexture() {sceneTextureId=PPI_SceneDepth}.InvSize, B=Constant2Vector() {g=-1}), B=@n27)
  n32 = Subtract(A=SceneTexture() {sceneTextureId=PPI_WorldNormal}.Color, B=SceneTexture(UVs=Add(A=@n30, B=@n31)) {sceneTextureId=PPI_WorldNormal}.Color)
  n33 = TextureCoordinate()
  n34 = Multiply(A=Multiply(A=SceneTexture() {sceneTextureId=PPI_SceneDepth}.InvSize, B=Constant2Vector() {r=1}), B=@n27)
  n35 = Subtract(A=SceneTexture() {sceneTextureId=PPI_WorldNormal}.Color, B=SceneTexture(UVs=Add(A=@n33, B=@n34)) {sceneTextureId=PPI_WorldNormal}.Color)
  n36 = TextureCoordinate()
  n37 = Multiply(A=Multiply(A=SceneTexture() {sceneTextureId=PPI_SceneDepth}.InvSize, B=Constant2Vector() {r=-1}), B=@n27)
  n38 = Subtract(A=SceneTexture() {sceneTextureId=PPI_WorldNormal}.Color, B=SceneTexture(UVs=Add(A=@n36, B=@n37)) {sceneTextureId=PPI_WorldNormal}.Color)
  n39 = Desaturation(None=Add(A=Add(A=@n29, B=@n32), B=Add(A=@n35, B=@n38)))
  n40 = ScalarParameter() {defaultValue=1, parameterName=WN Threshold}
  NamedRerouteDeclaration(None=Saturate(None=Power(Base=@n39, Exp=@n40) {constExponent=1, desc=Line Detail})) {name=WN Lines, nodeColor=<rgba 0 1 0.4609 1>}
 ]
}
