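# M_PP_Pixelizer is a post-process material that gives the scene a retro pixelated look with reduced color levels.
# It snaps screen UVs to an adjustable grid, samples the scene there, then partly blends in a quantized color.
# A static switch picks either RGB-weighted checkerboard dither brightness or a scanline-style modulation from a tiled 2x2 checker texture.
# Master FX Control blends between the original scene and the styled result.
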Material /Game/Creative/PostProcess/Materials/M_PP_Pixelizer
graph {
 properties {materialDomain=MD_PostProcess, pixelDepthOffsetMode=PDOM_Legacy, usedShadingModels=Default Lit}
 nodes [
  n1 = ScalarParameter() {defaultValue=1.5, parameterName=PixelWidth}
  n2 = ScalarParameter() {defaultValue=1, parameterName=PixelHeight}
  n3 = Floor(None=Multiply(A=ScalarParameter() {defaultValue=350, parameterName=PixelAmount}, B=AppendVector(A=@n1, B=@n2)))
  n4 = Multiply(A=ScreenPosition().ViewportUV, B=Multiply(A=Constant() {r=0.2}, B=@n3))
  Add(A=ScalarParameter() {parameterName=agre34r22}, B=@n4)
  Constant3Vector() {constant=<rgba 0.3 0.6 0.1 0>}
  Constant3Vector() {constant=<rgba 0.75 0.75 0.75 0>}
  Constant() {r=0.05}
  Constant() {r=0.1}
  Constant() {r=250}
  Constant() {r=0.8}
  Constant() {r=400}
  Constant() {r=0.4375}
  n5 = Floor(None=Multiply(A=ScreenPosition().ViewportUV, B=@n3))
  n6 = ComponentMask(None=SceneTexture(UVs=Divide(A=@n5, B=@n3)) {sceneTextureId=PPI_PostProcessInput0}.Color) {b=true, g=true, r=true}
  n7 = ScalarParameter() {defaultValue=12, parameterName=DitherSceneLevels}
  n8 = Multiply(A=@n6, B=@n7)
  n9 = LinearInterpolate(A=@n6, B=Divide(A=Floor(None=@n8), B=@n7), Alpha=ScalarParameter() {defaultValue=0.5, parameterName=DitherWholeSceneBlend})
  n10 = DotProduct(
   A=@n9
   B=ComponentMask(None=VectorParameter() {parameterName=CheckerDitherPerRGB, defaultValue=<rgba 0.3 0.3 0.8 0>}.RGB) {b=true, g=true, r=true}
  )
  n11 = MaterialFunctionCall!(
   UVs=ScreenPosition().ViewportUV
   Tiling=Multiply(
    A=MaterialFunctionCall() {materialFunction=&ScreenResolution}.Visible Resolution
    B=ScalarParameter() {defaultValue=0.5, parameterName=CheckerDither-Tiling}
   )
  ) {materialFunction=/Engine/ArtTools/RenderToTexture/MaterialFunctions/CheckerPattern}
  n12 = LinearInterpolate(
   A=ScalarParameter() {defaultValue=2.5, parameterName=CheckerDither-A}
   B=ScalarParameter() {defaultValue=1, parameterName=CheckerDither-B}
   Alpha=Clamp(None=Max(A=@n10, B=@n11.Result))
  )
  n13 = Multiply(A=ScalarParameter() {defaultValue=0.8, parameterName=ScanlineBrightness}, B=@n9)
  n14 = Multiply(A=@n9, B=ScalarParameter() {defaultValue=1.25, parameterName=ScanlineBrightness2})
  n15 = TextureSample(UVs=@n4) {samplerType=SAMPLERTYPE_LinearGrayscale, texture=/Game/Creative/PostProcess/Textures/Checker_2x2}
  n16 = ScalarParameter() {defaultValue=2, parameterName=ScanlinePower}
  n17 = Clamp(None=Multiply(A=Power(Base=@n15.R, Exp=@n16), B=ScalarParameter() {defaultValue=2, parameterName=ScanlineX}))
  LinearInterpolate(
   A=ComponentMask(None=SceneTexture() {sceneTextureId=PPI_PostProcessInput0}.Color) {b=true, g=true, r=true}
   B=StaticSwitchParameter(True=Multiply(A=@n9, B=@n12), False=LinearInterpolate(A=@n13, B=@n14, Alpha=@n17)) {parameterName=UseStyleB}
   Alpha=ScalarParameter() {parameterName=Master FX Control, sortPriority=0}
  )
 ]
}
