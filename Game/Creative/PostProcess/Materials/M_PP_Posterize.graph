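# Post-process material that, despite the Posterize name, overlays an aspect-corrected halftone dot pattern on the screen.
# Screen UVs form a Grid Res cell grid, apparently sampling scene colour at cell centres for desaturated red-green brightness.
# Darker samples give larger black dots, and a Min over each cell and four neighbours lets dots overlap borders.
# Posterize Amount blends the pattern into scene colour, then MasterVFXAdjustment scales the whole effect.
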Material /Game/Creative/PostProcess/Materials/M_PP_Posterize
graph {
 properties {
  bAutomaticallySetUsageInEditor=false
  blendablePriority=1
  materialDomain=MD_PostProcess
  pixelDepthOffsetMode=PDOM_Legacy
  usedShadingModels=Default Lit
 }
 nodes [
  n1 = Multiply(A=Distance(A=NamedRerouteUsage(), B=NamedRerouteUsage())) {constB=1.4142, desc=Normalise}
  n2 = Divide(A=NamedRerouteUsage(), B=NamedRerouteUsage())
  n3 = NamedRerouteUsage()
  n4 = ComponentMask(None=SceneTexture(UVs=Divide(A=@n2, B=@n3)) {sceneTextureId=PPI_PostProcessInput0}.Color) {g=true, r=true}
  n5 = OneMinus(None=Saturate(None=Desaturation(None=@n4)))
  n6 = Saturate(None=Divide(A=Subtract(A=@n1, B=@n5)) {constB=0.01})
  n7 = NamedRerouteUsage()
  n8 = Add(A=NamedRerouteUsage(), B=Constant2Vector() {r=1})
  n9 = Divide(A=Divide(A=@n8, B=NamedRerouteUsage()), B=NamedRerouteUsage())
  n10 = Desaturation(None=ComponentMask(None=SceneTexture(UVs=@n9) {sceneTextureId=PPI_PostProcessInput0}.Color) {g=true, r=true})
  n11 = Subtract(A=Multiply(A=Distance(A=@n7, B=@n8)) {constB=1.4142, desc=Normalise}, B=OneMinus(None=Saturate(None=@n10)))
  n12 = Add(A=NamedRerouteUsage(), B=Constant2Vector() {r=-1})
  n13 = Distance(A=NamedRerouteUsage(), B=@n12)
  n14 = NamedRerouteUsage()
  n15 = SceneTexture(UVs=Divide(A=Divide(A=@n12, B=@n14), B=NamedRerouteUsage())) {sceneTextureId=PPI_PostProcessInput0}
  n16 = Saturate(None=Desaturation(None=ComponentMask(None=@n15.Color) {g=true, r=true}))
  n17 = Divide(A=Subtract(A=Multiply(A=@n13) {constB=1.4142, desc=Normalise}, B=OneMinus(None=@n16))) {constB=0.01}
  n18 = Add(A=NamedRerouteUsage(), B=Constant2Vector() {g=1})
  n19 = Multiply(A=Distance(A=NamedRerouteUsage(), B=@n18)) {constB=1.4142, desc=Normalise}
  n20 = Divide(A=@n18, B=NamedRerouteUsage())
  n21 = NamedRerouteUsage()
  n22 = ComponentMask(None=SceneTexture(UVs=Divide(A=@n20, B=@n21)) {sceneTextureId=PPI_PostProcessInput0}.Color) {g=true, r=true}
  n23 = OneMinus(None=Saturate(None=Desaturation(None=@n22)))
  n24 = Saturate(None=Divide(A=Subtract(A=@n19, B=@n23)) {constB=0.01})
  n25 = Add(A=NamedRerouteUsage(), B=Constant2Vector() {g=-1})
  n26 = Multiply(A=Distance(A=NamedRerouteUsage(), B=@n25)) {constB=1.4142, desc=Normalise}
  n27 = Divide(A=@n25, B=NamedRerouteUsage())
  n28 = NamedRerouteUsage()
  n29 = ComponentMask(None=SceneTexture(UVs=Divide(A=@n27, B=@n28)) {sceneTextureId=PPI_PostProcessInput0}.Color) {g=true, r=true}
  n30 = OneMinus(None=Saturate(None=Desaturation(None=@n29)))
  n31 = Saturate(None=Divide(A=Subtract(A=@n26, B=@n30)) {constB=0.01})
  n32 = Min(A=Saturate(None=Divide(A=@n11) {constB=0.01}), B=Min(A=Saturate(None=@n17), B=Min(A=@n24, B=@n31)))
  LinearInterpolate(
   A=ComponentMask(None=SceneTexture() {sceneTextureId=PPI_PostProcessInput0}.Color) {a=true, b=true, g=true, r=true}
   B=LinearInterpolate(
    A=SceneTexture() {sceneTextureId=PPI_PostProcessInput0}.Color
    B=Min(A=@n6, B=@n32)
    Alpha=ScalarParameter() {defaultValue=0.2, parameterName=Posterize Amount, sliderMax=1}
   )
   Alpha=ScalarParameter() {group=0. VFX Adjustments, parameterName=MasterVFXAdjustment, sortPriority=1}
  )
  n33 = ViewSize()
  n34 = ComponentMask(None=@n33) {r=true}
  n35 = ComponentMask(None=@n33) {g=true}
  NamedRerouteDeclaration(None=AppendVector(A=Divide(A=@n34, B=@n35), B=Constant() {r=1})) {name=Aspect Ratio, nodeColor=<rgba 0 0.7656 1 1>}
  n36 = TextureCoordinate()
  n37 = NamedRerouteUsage()
  n38 = NamedRerouteDeclaration(None=Multiply(A=Multiply(A=@n36, B=@n37), B=NamedRerouteUsage())) {name=Scaled UVs, nodeColor=<rgba 0 1 0.4609 1>}
  NamedRerouteDeclaration(None=Add(A=Floor(None=@n38)) {constB=0.5}) {name=CellCoord, nodeColor=<rgba 0.6875 1 0 1>}
  NamedRerouteDeclaration(None=ScalarParameter() {defaultValue=256, desc=Grid Res, parameterName=Grid Res}) {name=Grid Res, nodeColor=<rgba 0.8281 0 1 1>}
 ]
}
