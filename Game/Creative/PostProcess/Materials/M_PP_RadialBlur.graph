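# Post-process material, applied after depth of field, that adds a radial blur toward the screen center.
# A Custom HLSL node averages eleven scene color taps along the pixel-to-center direction, scaled by Blur, with a mobile shading path branch.
# A centered RadialGradientExponential, set by Blur Radius and Density, masks blurred against unblurred color.
# Master FX Control fades the whole effect in and defaults to off.
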Material /Game/Creative/PostProcess/Materials/M_PP_RadialBlur
graph {
 properties {blendableLocation=BL_SceneColorAfterDOF, materialDomain=MD_PostProcess, pixelDepthOffsetMode=PDOM_Legacy, usedShadingModels=Default Lit}
 nodes [
  MaterialExpressionScreenPosition_0 = ScreenPosition()
  n1 = SceneTexture(UVs=@MaterialExpressionScreenPosition_0.ViewportUV) {sceneTextureId=PPI_PostProcessInput0}
  MaterialExpressionScalarParameter_2 = ScalarParameter() {defaultValue=3, group=Radial Blur, parameterName=Blur, sortPriority=1}
  n2 = Custom!(Blur=@MaterialExpressionScalarParameter_2, ScreenUV=@MaterialExpressionScreenPosition_0.ViewportUV) {
   code="const float2 ScreenMult = View.ViewSizeAndInvSize.xy * View.BufferSizeAndInvSize.zw;\r\n\r\nconst int TexIndex = 14;\r\nconst float Samples [11] = {-0.08, -0.05, -0.03, -0.02, -0.01, 0, 0.01, 0.02, 0.03, 0.05, 0.08};\r\n\r\nfloat2 dir = float2(0.5,0.5) - ScreenUV;\r\nfloat4 sum = float4(0.0,0.0,0.0,0.0);\r\nfor(int i = 0; i<11; i++)\r\n{\r\n    float2 pos = ScreenUV + dir * Samples [i] * Blur;\r\n    pos = clamp(pos, float2(0.0,0.0), float2(1.0, 1.0));\r\n\r\n##if SHADING_PATH_MOBILE\r\n\tsum += MobileSceneTextureLookup(Parameters, TexIndex, pos * ScreenMult);\r\n##else\r\n\tsum += SceneTextureLookup(pos * ScreenMult, TexIndex, false);\r\n##endif\r\n}\r\n\r\nreturn sum / 11.0;"
   description=Radial Blur
   inputs [
    {inputName=Blur, input {expression=&self:MaterialExpressionScalarParameter_2}}
    {inputName=ScreenUV, input {expression=&self:MaterialExpressionScreenPosition_0}}
   ]
  }
  n3 = ScreenPosition()
  n4 = Constant2Vector() {g=0.5, r=0.5}
  n5 = ScalarParameter() {defaultValue=1, group=Radial Blur, parameterName=Blur Radius, sortPriority=2}
  n6 = ScalarParameter() {defaultValue=1, group=Radial Blur, parameterName=Blur Density, sortPriority=3}
  LinearInterpolate(
   A=ComponentMask(None=SceneTexture() {sceneTextureId=PPI_PostProcessInput0}.Color) {b=true, g=true, r=true}
   B=LinearInterpolate(
    A=ComponentMask(None=@n1.Color) {b=true, g=true, r=true}
    B=@n2
    Alpha=MaterialFunctionCall(UVs=@n3.ViewportUV, CenterPosition=@n4, Radius=@n5, Density=@n6) {materialFunction=&RadialGradientExponential}.RadialGradientExponential
   )
   Alpha=ScalarParameter() {group=Radial Blur, parameterName=Master FX Control, sortPriority=0}
  )
 ]
}
