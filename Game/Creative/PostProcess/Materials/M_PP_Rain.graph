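# M_PP_Rain is a post-process material that overlays animated rain drops and drips on the screen, as if on glass.
# Noise over screen-aligned grids builds round drops that fade on two time-offset cycles, plus drip shapes from per-column sine offsets and a sharp power curve.
# These shapes distort screen UVs, with DDY-derived highlights and vignette masks shaping the effect.
# A Master FX Control parameter blends the result with the unmodified scene color.
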Material /Game/Creative/PostProcess/Materials/M_PP_Rain
graph {
 properties {materialDomain=MD_PostProcess, pixelDepthOffsetMode=PDOM_Legacy, usedShadingModels=Default Lit}
 nodes [
  n1 = NamedRerouteUsage()
  n2 = Multiply(A=NamedRerouteUsage(), B=NamedRerouteUsage())
  n3 = MaterialFunctionCall() {materialFunction=/Game/Creative/PostProcess/MaterialFunctions/MF_VignetteMask}
  n4 = NamedRerouteUsage()
  n5 = ScalarParameter() {defaultValue=2, group=Overall, parameterName=Rain Highlight}
  n6 = Subtract(
   A=Add(A=SceneTexture(UVs=@n1) {sceneTextureId=PPI_PostProcessInput0}.Color, B=Multiply(A=@n2, B=@n3.VignetteMask))
   B=Multiply(A=Multiply(A=@n4, B=@n5), B=@n3.VignetteMask)
  ) {desc=had to subtract because the drops would be on the otherside of the glass}
  n7 = NamedRerouteUsage()
  n8 = NamedRerouteUsage()
  n9 = Multiply(
   A=Saturate(None=Multiply(A=@n7, B=@n8))
   B=MaterialFunctionCall() {materialFunction=&MF_VignetteMask}.VignetteMask
  )
  LinearInterpolate(
   A=ComponentMask(None=SceneTexture() {sceneTextureId=PPI_PostProcessInput0}.Color) {b=true, g=true, r=true}
   B=ComponentMask(None=Subtract(A=@n6, B=@n9)) {b=true, g=true, r=true}
   Alpha=ScalarParameter() {group=Main, parameterName=Master FX Control, sortPriority=0}
  )
  NamedRerouteDeclaration(None=DDY(Value=NamedRerouteUsage())) {name=Rain Highlight, nodeColor=<rgba 0.4766 0 1 1>}
  n10 = ScalarParameter() {defaultValue=60, group=Overall, parameterName=Bands}
  n11 = Multiply(
   A=Panner(Coordinate=MaterialFunctionCall() {materialFunction=&ScreenAlignedUVs}."X 100%, Y Scale to Ratio") {speedY=-0.002}
   B=@n10
  )
  n12 = NamedRerouteUsage()
  n13 = Floor(None=NamedRerouteDeclaration(None=Add(A=@n11, B=@n12)) {name=Grid PrePixel, nodeColor=<rgba 1 0 0.2344 1>})
  n14 = NamedRerouteDeclaration(None=@n10) {name=Band Amt, nodeColor=<rgba 1 0.6094 0 1>}
  NamedRerouteDeclaration(None=AppendVector(A=Divide(A=@n13, B=@n14), B=Constant())) {name=Grid Drops, nodeColor=<rgba 0 0.4609 1 1>}
  n15 = AppendVector(A=NamedRerouteUsage(), B=Constant())
  n16 = ScalarParameter() {defaultValue=0.125, group=Main, parameterName=Round Droplets}
  NamedRerouteDeclaration(None=Saturate(None=Noise!(World Position=@n15, FilterWidth=@n16) {levels=1, noiseFunction=NOISEFUNCTION_GradientALU})) {name=Drops Shape, nodeColor=<rgba 1 0.1875 0 1>}
  n17 = ScalarParameter() {defaultValue=3, parameterName=Drop Minimizer}
  n18 = NamedRerouteDeclaration(None=Multiply(A=Time() {bIgnorePause=true}, B=ScalarParameter() {defaultValue=0.075, group=Overall, parameterName=Overall Speed})) {name=Time Input, nodeColor=<rgba 0 0.9062 1 1>}
  n19 = Noise(World Position=NamedRerouteUsage()) {bTiling=true, levels=5, noiseFunction=NOISEFUNCTION_GradientALU, outputMin=0, scale=3}
  n20 = Frac(None=Add(A=Add(A=@n18) {constB=0.66}, B=@n19))
  n21 = OneMinus(None=Saturate(None=Multiply(A=@n17, B=@n20)))
  n22 = Frac(None=Add(A=@n18, B=@n19))
  n23 = OneMinus(None=Saturate(None=Multiply(A=@n17, B=@n22)))
  NamedRerouteDeclaration(None=Saturate(None=Add(A=@n21, B=@n23))) {name=Drops Math, nodeColor=<rgba 1 0.7734 0 1>}
  n24 = MaterialFunctionCall() {materialFunction=&ScreenAlignedUVs}
  n25 = NamedRerouteUsage()
  n26 = Multiply(A=@n24."X 100%, Y Scale to Ratio", B=@n25)
  n27 = Constant()
  n28 = NamedRerouteDeclaration(None=Noise(World Position=AppendVector(A=@n24."X 100%, Y Scale to Ratio", B=@n27)) {levels=10, noiseFunction=NOISEFUNCTION_GradientALU, outputMin=0, scale=6}) {name=Grid Noise, nodeColor=<rgba 1 0 0.6562 1>}
  n29 = Divide(A=Floor(None=Add(A=@n26, B=@n28)), B=@n25)
  n30 = ComponentMask(None=@n29) {r=true}
  n31 = NamedRerouteUsage()
  n32 = Sine(None=@n30)
  n33 = NamedRerouteUsage()
  n34 = ScalarParameter() {defaultValue=1, group=Drips, parameterName=Drips Amt Reduction}
  n35 = Add(
   A=Multiply(A=Multiply(A=@n31) {constB=-0.05}, B=Frac(None=@n32))
   B=Divide(A=ComponentMask(None=@n29) {g=true}, B=Multiply(A=@n33, B=@n34) {constB=2}) {constB=30}
  )
  n36 = Frac(None=Add(A=Add(A=@n30, B=@n35), B=ScalarParameter() {defaultValue=2, parameterName=Drips Amount}) {constB=2})
  n37 = ScalarParameter() {defaultValue=500, group=Drips, parameterName=Drips Smoothness}
  NamedRerouteDeclaration(None=Saturate(None=Power(Base=@n36, Exp=@n37) {constExponent=500})) {name=Drips Shape, nodeColor=<rgba 1 0.8438 0 1>}
  NamedRerouteDeclaration(None=Multiply(A=NamedRerouteUsage(), B=ScalarParameter() {defaultValue=2, group=Drips, parameterName=Drips Size Reduction}) {constB=2}) {name=Drips Amt, nodeColor=<rgba 1 0.1641 0 1>}
  n38 = NamedRerouteDeclaration(None=Multiply(A=NamedRerouteUsage(), B=NamedRerouteUsage())) {name=Drops, nodeColor=<rgba 1 0 0.4453 1>}
  n39 = Saturate(None=MaterialFunctionCall(UVs=Multiply(A=@n38) {constB=5}, Radius=Constant() {r=0.75}, Density=Constant() {r=0.2}) {materialFunction=&RadialGradientExponential}.RadialGradientExponential)
  n40 = NamedRerouteUsage()
  n41 = NamedRerouteUsage()
  n42 = ComponentMask(None=NamedRerouteUsage()) {r=true}
  n43 = ScalarParameter() {defaultValue=0.02, group=Overall, parameterName=Distortion Intensity}
  n44 = Add(
   A=Saturate(None=Multiply(A=@n39, B=@n40))
   B=Multiply(A=Add(A=@n41, B=@n42), B=NamedRerouteDeclaration(None=@n43) {name=Distortion Intensity, nodeColor=<rgba 0.9922 0 1 1>})
  )
  n45 = MaterialFunctionCall() {materialFunction=&MF_VignetteMask}
  NamedRerouteDeclaration(None=Add(
    A=MaterialFunctionCall() {materialFunction=&ScreenAlignedUVs}."X 100%, Y 100%"
    B=Multiply(A=@n44, B=@n45.VignetteMask)
   )) {name=Rain UV Distort, nodeColor=<rgba 0.2188 0 1 1>}
 ]
}
