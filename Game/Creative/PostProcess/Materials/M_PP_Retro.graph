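# Post-process material that restyles the scene as retro three-color art with glowing edges, blended in by Master FX Control.
# Medium-and-higher deferred rendering builds faceted shading from world-position derivatives, luminance and AO, then adds depth-Sobel wireframe, outlines, height fog and distance saturation.
# Jukebox parameter-collection audio bands pulse brightness, edge and fog colors.
# Low quality, mobile and forward use a cheaper palette with four-tap depth outlines; several branches are unconnected.
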Material /Game/Creative/PostProcess/Materials/M_PP_Retro
graph {
 properties {blendableLocation=BL_SceneColorAfterDOF, materialDomain=MD_PostProcess, pixelDepthOffsetMode=PDOM_Legacy, usedShadingModels=Default Lit}
 nodes [
  n1 = VectorParameter() {parameterName=LowColor, defaultValue=<rgba 0.0195 0.031 0.06 1>}
  n2 = VectorParameter() {parameterName=MidColor, defaultValue=<rgba 0.1009 0.0575 0.1059 1>}
  n3 = VectorParameter() {parameterName=HighColor, defaultValue=<rgba 0.0679 0.3251 0.4392 1>}
  n4 = Saturate(None=SceneTexture() {sceneTextureId=PPI_PostProcessInput0}.Color)
  n5 = SceneTexture() {sceneTextureId=PPI_PostProcessInput0}
  n6 = ScreenPosition()
  n7 = SceneTexelSize()
  n8 = Constant2Vector() {g=-1, r=-1}
  n9 = SceneTexture(UVs=Add(A=@n6.ViewportUV, B=Multiply(A=@n7, B=@n8))) {sceneTextureId=PPI_SceneDepth}
  n10 = SceneTexture() {sceneTextureId=PPI_SceneDepth}
  n11 = Constant2Vector() {g=-1, r=1}
  n12 = SceneTexture(UVs=Add(A=@n6.ViewportUV, B=Multiply(A=@n7, B=@n11))) {sceneTextureId=PPI_SceneDepth}
  n13 = Add(A=@n6.ViewportUV, B=Multiply(A=@n7, B=Constant2Vector() {g=1, r=1}))
  n14 = Add(
   A=Add(A=Subtract(A=@n9.Color, B=@n10.Color), B=Subtract(A=@n12.Color, B=@n10.Color))
   B=Subtract(A=SceneTexture(UVs=@n13) {sceneTextureId=PPI_SceneDepth}.Color, B=@n10.Color)
  )
  n15 = Multiply(A=@n7, B=Constant2Vector() {g=1, r=-1})
  n16 = Subtract(A=SceneTexture(UVs=Add(A=@n6.ViewportUV, B=@n15)) {sceneTextureId=PPI_SceneDepth}.Color, B=@n10.Color)
  n17 = Clamp(None=Abs(None=Add(A=@n14, B=@n16)), Max=ScalarParameter() {defaultValue=65000, parameterName=SceneDepthClampMaxMobile}) {clampMode=CMODE_ClampMax}
  n18 = ScalarParameter() {defaultValue=10, parameterName=EdgeSobelDivisorMobile}
  n19 = SceneTexture() {sceneTextureId=PPI_SceneDepth}
  n20 = ScalarParameter() {defaultValue=2, parameterName=EdgeSceneDepthDivisorMobile}
  n21 = Saturate(None=Divide(A=Divide(A=@n17, B=@n18), B=Divide(A=@n19.Color, B=@n20)))
  n22 = LinearInterpolate(
   A=LinearInterpolate(
    A=MaterialFunctionCall(A=@n1.RGB, B=@n2.RGB, C=@n3.RGB, Alpha=@n4) {materialFunction=&3ColorBlend}.Result
    B=ComponentMask(None=@n5.Color) {b=true, g=true, r=true}
   ) {constAlpha=0.2}
   B=VectorParameter() {parameterName=OutlineColor, defaultValue=<rgba 20.1071 1.6234 30 15>}.RGB
   Alpha=ComponentMask(None=MaterialFunctionCall!(In=@n21) {materialFunction=/Game/Effects/Fort_Effects/Materials/_Functions/MF_Smoothstep_01}.Result) {r=true}
  )
  n23 = VectorParameter() {parameterName=FogColor, defaultValue=<rgba 0.0228 0.3927 0.4444 2>}
  n24 = CollectionParameter() {collection=/Game/Creative/Devices/Jukebox/Creative_Jukebox_MPC, parameterName=MidHigh}
  n25 = CollectionParameter() {collection=&Creative_Jukebox_MPC, parameterName=High}
  n26 = LinearInterpolate(A=@n23.RGB, B=Multiply(A=@n23.RGB, B=@n23.RGB), Alpha=Divide(A=Add(A=@n24, B=@n25)))
  n27 = VectorParameter() {parameterName=LowColor, defaultValue=<rgba 0.0195 0.031 0.06 1>}
  n28 = VectorParameter() {parameterName=MidColor, defaultValue=<rgba 0.1009 0.0575 0.1059 1>}
  n29 = VectorParameter() {parameterName=HighColor, defaultValue=<rgba 0.0679 0.3251 0.4392 1>}
  n30 = SceneDepth()
  n31 = CollectionParameter() {collection=&Creative_Jukebox_MPC, parameterName=AverageAmplitude}
  n32 = ScalarParameter() {defaultValue=0.2, parameterName=MidScale}
  n33 = VectorParameter() {parameterName=FacetedFacesViewSpaceNormalizedRotationAxis, defaultValue=<rgba 1 -0.2 0.3 0>}
  n34 = WorldPosition()
  n35 = Normalize(VectorInput=CrossProduct(A=DDX(Value=@n34.XYZ), B=DDY(Value=@n34.XYZ)))
  n36 = DotProduct(
   A=RotateAboutAxis(
    NormalizedRotationAxis=Normalize(VectorInput=@n33.RGB)
    RotationAngle=ScalarParameter() {defaultValue=0.1079, parameterName=FacetedFacesViewSpaceRotationAngle}
    PivotPoint=Constant()
    Position=Transform(None=@n35) {transformSourceType=TRANSFORMSOURCE_World, transformType=TRANSFORM_View}
   )
   B=Constant() {r=1}
  )
  n37 = ScalarParameter() {defaultValue=3, parameterName=FacetedFacesExponent}
  n38 = Desaturation(
   None=ComponentMask(None=SceneTexture() {bFiltered=true, sceneTextureId=PPI_PostProcessInput0}.Color) {b=true, g=true, r=true}
   Fraction=Constant() {r=1}
  )
  n39 = ScalarParameter() {defaultValue=-0.15, parameterName=DiffuseColorBias}
  n40 = Add(
   A=Add(A=Multiply(A=@n31, B=@n32), B=Power(Base=@n36, Exp=@n37))
   B=Multiply(A=Add(A=@n38, B=@n39), B=ScalarParameter() {defaultValue=0.4, desc=.5, parameterName=DiffuseColorScale})
  )
  n41 = SceneTexture() {bFiltered=true, sceneTextureId=PPI_AmbientOcclusion}
  n42 = Multiply(
   A=OneMinus(None=ComponentMask(None=@n41.Color) {b=true, g=true, r=true})
   B=ScalarParameter() {defaultValue=0.3, desc=.15, parameterName=AOScale}
  )
  n43 = Multiply(
   A=MaterialFunctionCall!(
    min=ScalarParameter() {defaultValue=300000, parameterName=FacetFadeFar}
    max=ScalarParameter() {defaultValue=200000, parameterName=FaceFadeNear}
    value=ComponentMask(None=@n30.R) {r=true}
   ) {materialFunction=/Game/Effects/Fort_Effects/Materials/_Functions/InverseLerp_MF}.Clamped
   B=Max(A=Subtract(A=@n40, B=@n42), B=Constant())
  )
  n44 = VectorParameter() {parameterName=WireframeColor, defaultValue=<rgba 0.0966 1.7271 2 2>}
  n45 = Multiply(A=@n44.RGB, B=@n44.RGB)
  n46 = CollectionParameter() {collection=&Creative_Jukebox_MPC, parameterName=Mid}
  n47 = ScreenPosition()
  n48 = SceneTexelSize()
  n49 = Constant2Vector() {g=-1, r=-1}
  n50 = SceneTexture(UVs=Add(A=@n47.ViewportUV, B=Multiply(A=@n48, B=@n49))) {sceneTextureId=PPI_SceneDepth}
  n51 = SceneTexture() {sceneTextureId=PPI_SceneDepth}
  n52 = Constant2Vector() {g=-1}
  n53 = SceneTexture(UVs=Add(A=@n47.ViewportUV, B=Multiply(A=@n48, B=@n52))) {sceneTextureId=PPI_SceneDepth}
  n54 = Add(A=@n47.ViewportUV, B=Multiply(A=@n48, B=Constant2Vector() {g=-1, r=1}))
  n55 = Add(
   A=Add(A=Subtract(A=@n50.Color, B=@n51.Color), B=Subtract(A=@n53.Color, B=@n51.Color))
   B=Subtract(A=SceneTexture(UVs=@n54) {sceneTextureId=PPI_SceneDepth}.Color, B=@n51.Color)
  )
  n56 = Multiply(A=@n48, B=Constant2Vector() {r=1})
  n57 = Subtract(A=SceneTexture(UVs=Add(A=@n47.ViewportUV, B=@n56)) {sceneTextureId=PPI_SceneDepth}.Color, B=@n51.Color)
  n58 = Constant2Vector() {g=1, r=1}
  n59 = SceneTexture(UVs=Add(A=@n47.ViewportUV, B=Multiply(A=@n48, B=@n58))) {sceneTextureId=PPI_SceneDepth}
  n60 = Add(A=@n47.ViewportUV, B=Multiply(A=@n48, B=Constant2Vector() {g=1}))
  n61 = Add(
   A=Add(A=Add(A=@n55, B=@n57), B=Subtract(A=@n59.Color, B=@n51.Color))
   B=Subtract(A=SceneTexture(UVs=@n60) {sceneTextureId=PPI_SceneDepth}.Color, B=@n51.Color)
  )
  n62 = Multiply(A=@n48, B=Constant2Vector() {g=1, r=-1})
  n63 = Subtract(A=SceneTexture(UVs=Add(A=@n47.ViewportUV, B=@n62)) {sceneTextureId=PPI_SceneDepth}.Color, B=@n51.Color)
  n64 = Constant2Vector() {r=-1}
  n65 = SceneTexture(UVs=Add(A=@n47.ViewportUV, B=Multiply(A=@n48, B=@n64))) {sceneTextureId=PPI_SceneDepth}
  n66 = Abs(None=Add(A=Add(A=@n61, B=@n63), B=Subtract(A=@n65.Color, B=@n51.Color)))
  n67 = ScalarParameter() {defaultValue=0.1, parameterName=PolySobelMaxDepth}
  n68 = SceneDepth()
  n69 = SceneDepth()
  n70 = MaterialFunctionCall(
   min=ScalarParameter() {parameterName=PolySceneDepthNear}
   max=ScalarParameter() {defaultValue=10, parameterName=PolySceneDepthFar}
   value=Add(A=ComponentMask(None=@n69.R) {r=true}, B=ScalarParameter() {defaultValue=1000, parameterName=PolySceneDepthBias})
  ) {materialFunction=&InverseLerp_MF}
  n71 = Saturate(None=Divide(A=Clamp(None=@n66, Max=@n67) {clampMode=CMODE_ClampMax}, B=Divide(A=@n68.R, B=@n70.Unclamped)))
  n72 = VectorParameter() {parameterName=OutlineColor, defaultValue=<rgba 20.1071 1.6234 30 15>}
  n73 = Add(
   A=CollectionParameter() {collection=&Creative_Jukebox_MPC, parameterName=Low}
   B=CollectionParameter() {collection=&Creative_Jukebox_MPC, parameterName=LowMid}
  )
  n74 = Divide(A=@n66, B=ScalarParameter() {defaultValue=10, parameterName=EdgeSobelDivisor})
  n75 = Divide(A=SceneDepth().R, B=ScalarParameter() {defaultValue=500, parameterName=EdgeSceneDepthDivisor})
  n76 = MaterialFunctionCall(In=Saturate(None=Divide(A=@n74, B=@n75))) {materialFunction=&MF_Smoothstep_01}
  n77 = SceneTexture() {bFiltered=true, sceneTextureId=PPI_CustomDepth}
  n78 = Saturate(None=Subtract(A=ComponentMask(None=@n77.Color) {r=true}, B=SceneDepth().R))
  n79 = OneMinus(None=@n78)
  n80 = Power(Base=Divide(A=SceneDepth().R, B=ScalarParameter() {defaultValue=512, parameterName=DepthOutlineDrawDistance}) {constB=512})
  n81 = LinearInterpolate(A=@n78, B=OneMinus(None=Clamp(None=@n80)), Alpha=ScalarParameter() {defaultValue=0.99, parameterName=DepthLerpDrawDistanceControl}) {constAlpha=0.96}
  n82 = LinearInterpolate(
   A=LinearInterpolate(
    A=MaterialFunctionCall(A=@n27.RGB, B=@n28.RGB, C=@n29.RGB, Alpha=@n43) {materialFunction=&3ColorBlend}.Result
    B=LinearInterpolate(A=@n44.RGB, B=@n45, Alpha=@n46)
    Alpha=ComponentMask(None=@n71) {r=true}
   )
   B=LinearInterpolate(A=@n72.RGB, B=Multiply(A=@n72.RGB, B=@n72.RGB), Alpha=Divide(A=@n73))
   Alpha=Multiply(A=ComponentMask(None=@n76.Result) {r=true}, B=Add(A=@n79, B=@n81))
  )
  n83 = Add(
   A=ComponentMask(None=SceneTexture() {bFiltered=true, sceneTextureId=PPI_SceneDepth}.Color) {r=true}
   B=ScalarParameter() {defaultValue=-500, parameterName=FogBias}
  )
  n84 = ScalarParameter() {defaultValue=8000, parameterName=FogDepthDivisor}
  n85 = Multiply(A=Saturate(None=Divide(A=@n83, B=@n84)), B=ScalarParameter() {defaultValue=0.2, parameterName=FogScale})
  n86 = Add(
   A=MaterialFunctionCall(Float3=CameraPositionWS()) {materialFunction=&BreakOutFloat3Components}.B
   B=ScalarParameter() {defaultValue=2400, parameterName=HeightFogZBias}
  )
  n87 = MaterialFunctionCall(Float3=WorldPosition().XYZ) {materialFunction=&BreakOutFloat3Components}
  n88 = Saturate(None=Divide(A=Subtract(A=@n86, B=@n87.B), B=ScalarParameter() {defaultValue=4000, parameterName=HeightFogDivisor}))
  n89 = SceneTexture() {bFiltered=true, sceneTextureId=PPI_SceneDepth}
  n90 = Divide(
   A=Add(A=ComponentMask(None=@n89.Color) {r=true}, B=ScalarParameter() {defaultValue=-1000, parameterName=WorldSaturationBias})
   B=ScalarParameter() {defaultValue=4500, parameterName=WorldSaturationDivisor}
  )
  n91 = Desaturation(
   None=LinearInterpolate(A=@n82, B=Add(A=@n82, B=@n26), Alpha=Multiply(A=@n85, B=@n88))
   Fraction=Multiply(A=ScalarParameter() {defaultValue=-0.8, parameterName=WorldSaturation}, B=Saturate(None=@n90))
  )
  n92 = ComponentMask(None=SceneTexture() {bFiltered=true, sceneTextureId=PPI_PostProcessInput0}.Color) {b=true, g=true, r=true}
  n93 = ComponentMask(None=SceneTexture() {bFiltered=true, sceneTextureId=PPI_PostProcessInput1}.Color) {a=true}
  n94 = Saturate(None=Add(A=ScalarParameter() {defaultValue=0.2, parameterName=SceneTextureLerp}, B=OneMinus(None=@n93)))
  n95 = ShadingPathSwitch(
   Default=@n22
   Deferred=ShadingPathSwitch(Default=@n22, Deferred=LinearInterpolate(A=@n91, B=@n92, Alpha=@n94), Mobile=@n22) {desc=Migrated from FeatureLevelSwitch}
   Forward=@n22
   Mobile=@n22
  )
  LinearInterpolate(
   A=ComponentMask(None=SceneTexture() {sceneTextureId=PPI_PostProcessInput0}.Color) {b=true, g=true, r=true}
   B=QualitySwitch(Default=@n22, Low=@n22, High=@n95, Medium=@n95, Epic=@n95)
   Alpha=ScalarParameter() {parameterName=Master FX Control, sortPriority=0}
  )
  n96 = CameraPositionWS()
  n97 = WorldPosition()
  MaterialFunctionCall(
   Brightness=ScalarParameter() {defaultValue=2, parameterName=FresnelBrightness}
   Exponent=ScalarParameter() {defaultValue=4, parameterName=FresnelExponent}
   Camera Vector=Normalize(VectorInput=Subtract(A=@n96, B=@n97.XYZ))
   Vertex / Pixel Normal=ComponentMask(None=SceneTexture() {bFiltered=true, sceneTextureId=PPI_WorldNormal}.Color) {b=true, g=true, r=true}
  ) {materialFunction=&MF_Fresnel_FX}
  n98 = ScalarParameter() {defaultValue=20, parameterName=CustomDepthDivisor}
  n99 = SceneTexture() {sceneTextureId=PPI_WorldNormal}
  n100 = Subtract(A=CameraPositionWS(), B=WorldPosition().XYZ)
  Multiply(A=@n98, B=DotProduct(A=ComponentMask(None=@n99.Color) {b=true, g=true, r=true}, B=Normalize(VectorInput=@n100)))
  n101 = ScreenPosition()
  n102 = SceneTexelSize()
  n103 = Add(A=@n101.ViewportUV, B=Multiply(A=@n102, B=Constant2Vector() {g=-1, r=-1}))
  n104 = SceneTexture() {sceneTextureId=PPI_DiffuseColor}
  n105 = Add(A=@n101.ViewportUV, B=Multiply(A=@n102, B=Constant2Vector() {g=-1}))
  n106 = Add(
   A=Subtract(A=SceneTexture(UVs=@n103) {sceneTextureId=PPI_DiffuseColor}.Color, B=@n104.Color)
   B=Subtract(A=SceneTexture(UVs=@n105) {sceneTextureId=PPI_DiffuseColor}.Color, B=@n104.Color)
  )
  n107 = Multiply(A=@n102, B=Constant2Vector() {g=-1, r=1})
  n108 = Subtract(A=SceneTexture(UVs=Add(A=@n101.ViewportUV, B=@n107)) {sceneTextureId=PPI_DiffuseColor}.Color, B=@n104.Color)
  n109 = Constant2Vector() {r=1}
  n110 = SceneTexture(UVs=Add(A=@n101.ViewportUV, B=Multiply(A=@n102, B=@n109))) {sceneTextureId=PPI_DiffuseColor}
  n111 = Add(A=@n101.ViewportUV, B=Multiply(A=@n102, B=Constant2Vector() {g=1, r=1}))
  n112 = Add(
   A=Add(A=Add(A=@n106, B=@n108), B=Subtract(A=@n110.Color, B=@n104.Color))
   B=Subtract(A=SceneTexture(UVs=@n111) {sceneTextureId=PPI_DiffuseColor}.Color, B=@n104.Color)
  )
  n113 = Multiply(A=@n102, B=Constant2Vector() {g=1})
  n114 = Subtract(A=SceneTexture(UVs=Add(A=@n101.ViewportUV, B=@n113)) {sceneTextureId=PPI_DiffuseColor}.Color, B=@n104.Color)
  n115 = Constant2Vector() {g=1, r=-1}
  n116 = SceneTexture(UVs=Add(A=@n101.ViewportUV, B=Multiply(A=@n102, B=@n115))) {sceneTextureId=PPI_DiffuseColor}
  n117 = Add(A=@n101.ViewportUV, B=Multiply(A=@n102, B=Constant2Vector() {r=-1}))
  n118 = Add(
   A=Add(A=Add(A=@n112, B=@n114), B=Subtract(A=@n116.Color, B=@n104.Color))
   B=Subtract(A=SceneTexture(UVs=@n117) {sceneTextureId=PPI_DiffuseColor}.Color, B=@n104.Color)
  )
  Multiply(A=Saturate(None=ComponentMask(None=@n118) {r=true}), B=ScalarParameter() {defaultValue=0.05, parameterName=DiffuselScale})
  n119 = SceneTexture() {bFiltered=true, sceneTextureId=PPI_CustomStencil}
  n120 = Constant()
  Multiply(
   A=If(
    A=ComponentMask(None=@n119.Color) {r=true}
    B=ScalarParameter() {parameterName=Stencil}
    "A > B"=@n120
    "A == B"=Constant() {r=1}
    "A < B"=@n120
   )
   B=@n79
  )
  Reroute(None=@n79)
  n121 = ScreenPosition()
  n122 = SceneTexelSize()
  n123 = Constant2Vector() {g=-1, r=-1}
  n124 = SceneTexture(UVs=Add(A=@n121.ViewportUV, B=Multiply(A=@n122, B=@n123))) {sceneTextureId=PPI_SceneDepth}
  n125 = SceneTexture() {sceneTextureId=PPI_SceneDepth}
  n126 = Constant2Vector() {g=-1}
  n127 = SceneTexture(UVs=Add(A=@n121.ViewportUV, B=Multiply(A=@n122, B=@n126))) {sceneTextureId=PPI_SceneDepth}
  n128 = Add(A=@n121.ViewportUV, B=Multiply(A=@n122, B=Constant2Vector() {g=-1, r=1}))
  n129 = Add(
   A=Add(A=Subtract(A=@n124.Color, B=@n125.Color), B=Subtract(A=@n127.Color, B=@n125.Color))
   B=Subtract(A=SceneTexture(UVs=@n128) {sceneTextureId=PPI_SceneDepth}.Color, B=@n125.Color)
  )
  n130 = Multiply(A=@n122, B=Constant2Vector() {r=1})
  n131 = Subtract(A=SceneTexture(UVs=Add(A=@n121.ViewportUV, B=@n130)) {sceneTextureId=PPI_SceneDepth}.Color, B=@n125.Color)
  n132 = Constant2Vector() {g=1, r=1}
  n133 = SceneTexture(UVs=Add(A=@n121.ViewportUV, B=Multiply(A=@n122, B=@n132))) {sceneTextureId=PPI_SceneDepth}
  n134 = Add(A=@n121.ViewportUV, B=Multiply(A=@n122, B=Constant2Vector() {g=1}))
  n135 = Add(
   A=Add(A=Add(A=@n129, B=@n131), B=Subtract(A=@n133.Color, B=@n125.Color))
   B=Subtract(A=SceneTexture(UVs=@n134) {sceneTextureId=PPI_SceneDepth}.Color, B=@n125.Color)
  )
  n136 = Multiply(A=@n122, B=Constant2Vector() {g=1, r=-1})
  n137 = Subtract(A=SceneTexture(UVs=Add(A=@n121.ViewportUV, B=@n136)) {sceneTextureId=PPI_SceneDepth}.Color, B=@n125.Color)
  n138 = Constant2Vector() {r=-1}
  n139 = SceneTexture(UVs=Add(A=@n121.ViewportUV, B=Multiply(A=@n122, B=@n138))) {sceneTextureId=PPI_SceneDepth}
  n140 = Clamp(
   None=Add(A=Add(A=@n135, B=@n137), B=Subtract(A=@n139.Color, B=@n125.Color))
   Max=ScalarParameter() {defaultValue=2, parameterName=SobelMaxDepth}
  ) {clampMode=CMODE_ClampMax}
  n141 = Divide(A=SceneTexture() {sceneTextureId=PPI_SceneDepth}.Color, B=@n98) {constB=1000}
  Saturate(None=ComponentMask(None=Divide(A=@n140, B=@n141)) {r=true})
 ]
}
