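# Post-process material that blurs or sharpens the scene color after depth of field.
# It averages a weighted center tap with six taps in a hexagon around it, offset in pixels using the visible screen resolution.
# An unclamped lerp by Sharpen Or Blur extrapolates between original and blurred color, so negative values sharpen like an unsharp mask.
# Master FX Control, zero by default, fades the result in over the untouched image.
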Material /Game/Creative/PostProcess/Materials/M_PP_SimpleBlurSharpen
graph {
 properties {blendableLocation=BL_SceneColorAfterDOF, materialDomain=MD_PostProcess, pixelDepthOffsetMode=PDOM_Legacy, usedShadingModels=Default Lit}
 nodes [
  n1 = SceneTexture() {sceneTextureId=PPI_PostProcessInput0}
  n2 = TextureCoordinate()
  n3 = Divide(
   A=Constant() {r=1}
   B=MaterialFunctionCall() {materialFunction=&ScreenResolution}.Visible Resolution
  )
  n4 = Constant2Vector()
  n5 = SceneTexture(UVs=Add(A=@n2, B=Multiply(A=@n3, B=@n4))) {sceneTextureId=PPI_PostProcessInput0}
  n6 = Constant2Vector() {r=2}
  n7 = SceneTexture(UVs=Add(A=@n2, B=Multiply(A=@n3, B=@n6))) {sceneTextureId=PPI_PostProcessInput0}
  n8 = Add(A=@n2, B=Multiply(A=@n3, B=Constant2Vector() {r=-2}))
  n9 = Add(
   A=Add(A=Multiply(A=@n5.Color) {constB=0.3}, B=Multiply(A=@n7.Color) {constB=0.1167})
   B=Multiply(A=SceneTexture(UVs=@n8) {sceneTextureId=PPI_PostProcessInput0}.Color) {constB=0.1167}
  )
  n10 = Multiply(A=@n3, B=Constant2Vector() {g=2, r=1})
  n11 = Multiply(A=SceneTexture(UVs=Add(A=@n2, B=@n10)) {sceneTextureId=PPI_PostProcessInput0}.Color) {constB=0.1167}
  n12 = Constant2Vector() {g=-2, r=1}
  n13 = SceneTexture(UVs=Add(A=@n2, B=Multiply(A=@n3, B=@n12))) {sceneTextureId=PPI_PostProcessInput0}
  n14 = Add(A=@n2, B=Multiply(A=@n3, B=Constant2Vector() {g=2, r=-1}))
  n15 = Add(
   A=Add(A=Add(A=@n9, B=@n11), B=Multiply(A=@n13.Color) {constB=0.1167})
   B=Multiply(A=SceneTexture(UVs=@n14) {sceneTextureId=PPI_PostProcessInput0}.Color) {constB=0.1167}
  )
  n16 = Multiply(A=@n3, B=Constant2Vector() {g=-2, r=-1})
  n17 = Multiply(A=SceneTexture(UVs=Add(A=@n2, B=@n16)) {sceneTextureId=PPI_PostProcessInput0}.Color) {constB=0.1167}
  LinearInterpolate(
   A=@n1.Color
   B=LinearInterpolate(
    A=@n1.Color
    B=Add(A=@n15, B=@n17)
    Alpha=ScalarParameter() {defaultValue=2, group=Sharpen Or Blur, parameterName=Sharpen Or Blur, sliderMax=3, sliderMin=-3}
   )
   Alpha=ScalarParameter() {group=Sharpen Or Blur, parameterName=Master FX Control, sortPriority=0}
  )
 ]
}
