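# Post-process material that applies a spooky colour grade to the scene by partly desaturating it.
# On deferred shading at Medium quality and above, desaturation feeds a colour-preserving contrast function, then an exposure multiplier.
# Low and default quality, forward and mobile paths use only desaturation with its own fraction parameter.
# A Master FX Control scalar blends between the untouched scene colour and the graded result.
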Material /Game/Creative/PostProcess/Materials/M_PP_Spooky
graph {
 properties {materialDomain=MD_PostProcess, pixelDepthOffsetMode=PDOM_Legacy, usedShadingModels=Default Lit}
 nodes [
  n1 = SceneTexture(UVs=TextureCoordinate()) {sceneTextureId=PPI_PostProcessInput0}
  n2 = MaterialFunctionCall!(
   Color=Desaturation(
    None=ComponentMask(None=@n1.Color) {b=true, g=true, r=true}
    Fraction=ScalarParameter() {defaultValue=0.4, parameterName=DesaturationFraction}
   ) {luminanceFactors=<rgba 0.5637 0.59 0.502 0>}
   Contrast=ScalarParameter() {defaultValue=1.45, parameterName=Contrast Power}
  ) {materialFunction=/Engine/Functions/Engine_MaterialFunctions01/ImageAdjustment/Contrast_Preserve_Color}
  n3 = ScalarParameter() {defaultValue=1.375, parameterName=Exposure Controller}
  n4 = TextureCoordinate()
  n5 = Desaturation(
   None=ComponentMask(None=SceneTexture(UVs=@n4) {sceneTextureId=PPI_PostProcessInput0}.Color) {b=true, g=true, r=true}
   Fraction=ScalarParameter() {defaultValue=0.4, parameterName="DesaturationFraction(Mobile)"}
  ) {luminanceFactors=<rgba 0.5637 0.59 0.502 0>}
  n6 = ShadingPathSwitch(
   Default=@n5
   Deferred=ShadingPathSwitch(Default=@n5, Deferred=Multiply(A=@n2.Result, B=@n3), Mobile=@n5) {desc=Migrated from FeatureLevelSwitch}
   Forward=@n5
   Mobile=@n5
  )
  LinearInterpolate(
   A=ComponentMask(None=SceneTexture() {sceneTextureId=PPI_PostProcessInput0}.Color) {b=true, g=true, r=true}
   B=QualitySwitch(Default=@n5, Low=@n5, High=@n6, Medium=@n6, Epic=@n6)
   Alpha=ScalarParameter() {parameterName=Master FX Control, sortPriority=0}
  )
 ]
}
