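# MF_Hash23 is a material function that hashes a 2D input into a pseudo-random 3-component value between 0 and 1.
# It dots three different orderings of the input's X and Y with separate constant vectors, then applies the classic sine, large-multiplier and frac trick.
# The saturated fraction is the single output, and texture coordinates serve as the default input.
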MaterialFunction /Game/Creative/PostProcess/MaterialFunctions/MF_Hash23
graph {nodes [
  n1 = FunctionInput(Preview=TextureCoordinate()) {bUsePreviewValueAsDefault=true, inputType=FunctionInput_Vector2}
  n2 = ComponentMask(None=@n1) {r=true}
  n3 = ComponentMask(None=@n1) {g=true}
  n4 = DotProduct(
   A=MaterialFunctionCall(R=@n2, G=@n3, B=@n2) {materialFunction=&AppendMany}.RGB
   B=Constant3Vector() {constant=<rgba 127.1 311.7 74.7 0>}
  )
  n5 = DotProduct(
   A=MaterialFunctionCall(R=@n3, G=@n2, B=@n2) {materialFunction=&AppendMany}.RGB
   B=Constant3Vector() {constant=<rgba 269.5 183.3 246.1 0>}
  )
  n6 = DotProduct(
   A=MaterialFunctionCall(R=@n2, G=@n3, B=@n3) {materialFunction=&AppendMany}.RGB
   B=Constant3Vector() {constant=<rgba 113.5 271.9 124.6 0>}
  )
  n7 = Multiply(
   A=Sine(None=MaterialFunctionCall(R=@n4, G=@n5, B=@n6) {materialFunction=&AppendMany}.RGB)
   B=Constant() {r=43758.5469}
  ) {constB=43758.5469}
  FunctionOutput(None=Saturate(None=Frac(None=@n7)))
 ]}
