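# Post-process material that applies a VHS tape filter over the scene, blended in by the Master FX Control parameter.
# Time-driven hashes jitter UVs for tape wear and scrolling tracking-line glitches, then two scene samples split red from green-blue with alternating-line offsets.
# Pixelated hash noise adds interference streaks.
# The color is converted to luminance-chrominance, range-adjusted, chroma-rotated and converted back before the final lerp with the unfiltered scene.
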Material /Game/Creative/PostProcess/Materials/M_PP_VHS_Filter
graph {
 properties {materialDomain=MD_PostProcess, pixelDepthOffsetMode=PDOM_Legacy, usedShadingModels=Default Lit}
 nodes [
  n1 = NamedRerouteUsage()
  n2 = Add(A=Multiply(A=NamedRerouteUsage()) {constB=0.7}) {constB=0.3}
  n3 = Constant3Vector() {constant=<rgba 0 1 1 1>}
  n4 = TextureCoordinate()
  n5 = MaterialFunctionCall!(In=AppendVector(A=ComponentMask(None=@n4) {g=true}, B=Time())) {materialFunction=/Game/Creative/PostProcess/MaterialFunctions/MF_Hash23}
  n6 = ComponentMask(None=@n5.Result) {r=true}
  n7 = ComponentMask(None=ViewSize()) {r=true}
  n8 = OneMinus(None=ScalarParameter() {defaultValue=0.33, group=VHS Filter, parameterName=VHS Tape Wear, sliderMax=0.99})
  n9 = Add(A=TextureCoordinate(), B=Divide(A=Subtract(A=@n6) {constB=0.5}, B=Multiply(A=@n7, B=@n8))) {desc=uvn}
  n10 = NamedRerouteDeclaration(None=ComponentMask(None=@n9) {r=true}) {name=AnimatedLinesX, nodeColor=<rgba 0 1 0.0156 1>}
  n11 = Constant2Vector() {g=0.7, r=0.3}
  n12 = SmoothStep(Min=@n11.R, Max=@n11.R, Value=ComponentMask(None=@n5.Result) {g=true})
  n13 = ViewSize()
  n14 = ScalarParameter() {defaultValue=0.33, group=VHS Filter, parameterName=VHS Tracking Line Glitch Intensity, sliderMax=0.999, sortPriority=33}
  n15 = Divide(
   A=Multiply(A=NamedRerouteUsage(), B=@n12) {desc=Mult lines}
   B=Multiply(A=ComponentMask(None=@n13) {r=true}, B=OneMinus(None=@n14)) {constB=0.02}
  )
  n16 = ComponentMask(None=@n9) {g=true}
  n17 = NamedRerouteDeclaration(None=AppendVector(A=Subtract(A=@n10, B=@n15), B=NamedRerouteDeclaration(None=@n16) {name=Animated LinesY, nodeColor=<rgba 1 0.0234 0 1>})) {name=MuitilineInterlacedEffect, nodeColor=<rgba 0 1 0.0859 1>}
  n18 = Divide(B=ComponentMask(None=ViewSize()) {r=true}) {constB=1}
  n19 = ViewSize()
  n20 = Multiply(
   A=Add(A=Time(), B=TextureCoordinate())
   B=ScalarParameter() {defaultValue=0.33, group=VHS Filter, parameterName=VHS Lines Speed, sliderMax=1, sliderMin=-1, sortPriority=0}
  )
  n21 = Floor(None=Multiply(A=ComponentMask(None=@n19) {g=true}, B=ComponentMask(None=@n20) {g=true}))
  n22 = Constant() {desc=EveryOtherLine, r=2}
  n23 = Multiply(
   A=@n18
   B=Multiply(
    A=ScalarParameter() {defaultValue=3.33, group=VHS Filter, parameterName=VHS Chromatic Ab Offset, sliderMax=50, sortPriority=1}
    B=Fmod(A=@n21, B=@n22)
   )
  ) {constB=5}
  n24 = Constant()
  n25 = SceneTexture(UVs=Add(A=@n17, B=AppendVector(A=@n23, B=@n24)) {desc=Add Tape Wear}) {sceneTextureId=PPI_PostProcessInput0}
  n26 = Multiply(A=@n18) {constB=-1}
  n27 = Constant()
  n28 = SceneTexture(UVs=Add(A=@n17, B=AppendVector(A=@n26, B=@n27)) {desc=Add Tape Wear}) {sceneTextureId=PPI_PostProcessInput0}
  n29 = Add(
   A=Multiply(A=@n3, B=ComponentMask(None=@n25.Color) {b=true, g=true, r=true})
   B=Multiply(A=OneMinus(None=@n3), B=ComponentMask(None=@n28.Color) {b=true, g=true, r=true})
  )
  n30 = NamedRerouteUsage()
  n31 = Time()
  n32 = ComponentMask(None=MaterialFunctionCall(In=AppendVector(A=@n30, B=@n31)) {materialFunction=&MF_Hash23}.Result) {r=true}
  n33 = Constant2Vector() {r=1}
  n34 = NamedRerouteUsage()
  n35 = ScalarParameter() {defaultValue=8, group=VHS Filter, parameterName=VHS Noise Pixel Size, sliderMax=16, sliderMin=4, sortPriority=46}
  n36 = Divide(A=ViewSize(), B=@n35)
  n37 = Divide(A=Floor(None=Multiply(A=@n34, B=@n36)), B=@n36)
  n38 = Multiply(A=Add(A=Multiply(A=@n32, B=@n33), B=ComponentMask(None=@n37) {g=true, r=true}), B=Constant2Vector() {g=1, r=0.1})
  n39 = Power(
   Base=ComponentMask(None=MaterialFunctionCall(In=@n38) {materialFunction=&MF_Hash23}.Result) {r=true}
   Exp=Constant() {r=10}
  ) {constExponent=10}
  n40 = ScalarParameter() {defaultValue=0.33, group=VHS Filter, parameterName=VHS Noise Interference, sliderMax=5, sortPriority=44}
  n41 = If(
   A=Constant() {r=0.29}
   B=Multiply(A=Multiply(A=@n1, B=@n2))
   "A > B"=@n29
   "A == B"=@n29
   "A < B"=LinearInterpolate(A=@n29, B=Constant3Vector() {constant=<rgba 2 2 2 0>}, Alpha=Multiply(A=@n39, B=@n40))
  )
  n42 = Constant3Vector() {desc="B&&W", constant=<rgba 0.299 0.596 0.211 0>}
  n43 = Constant3Vector() {desc=Luminance, constant=<rgba 0.587 -0.274 -0.523 0>}
  n44 = Constant3Vector() {desc=Chrominance, constant=<rgba 0.144 -0.322 0.312 0>}
  n45 = Add(
   A=Constant3Vector() {desc="Luminance and 2 Chomanance ", constant=<rgba 0.1 -0.1 0 0>}
   B=Multiply(
    A=Constant3Vector() {desc="Shorten R Range, Expand G and B range", constant=<rgba 0.9 1.1 1.5 0>}
    B=MaterialFunctionCall(VectorToTransform=@n41, BasisX=@n42, BasisY=@n43, BasisZ=@n44) {materialFunction=&Transform3x3Matrix}.Transformed Vector
   )
  )
  n46 = ComponentMask(None=@n45) {r=true}
  n47 = Rotator(
   Coordinate=ComponentMask(None=@n45) {b=true, g=true}
   Time=Multiply(
    A=NamedRerouteUsage()
    B=ScalarParameter() {defaultValue=0.33, group=VHS Filter, parameterName=VHS Tracking Line Saturation Intensity, sortPriority=33}
   ) {constB=0.3}
  ) {centerX=0, centerY=0, speed=1}
  LinearInterpolate(
   A=ComponentMask(None=SceneTexture() {sceneTextureId=PPI_PostProcessInput0}.Color) {b=true, g=true, r=true}
   B=MaterialFunctionCall(
    VectorToTransform=AppendVector(A=@n46, B=@n47)
    BasisX=Constant3Vector() {constant=<rgba 1 1 1 0>}
    BasisY=Constant3Vector() {constant=<rgba 0.956 -0.272 -1.11 0>}
    BasisZ=Constant3Vector() {constant=<rgba 0.621 -0.647 1.7 0>}
   ) {materialFunction=&Transform3x3Matrix}.Transformed Vector
   Alpha=ScalarParameter() {group=VHS Filter, parameterName=Master FX Control, sortPriority=0}
  ) {constAlpha=0}
  n48 = Constant3Vector() {constant=<rgba 0 1 1 1>}
  n49 = NamedRerouteDeclaration() {name=AnimatedLinesX_1, nodeColor=<rgba 0 1 0.0156 1>}
  n50 = Divide(B=ComponentMask(None=ViewSize()) {g=true}) {constB=1}
  n51 = ScalarParameter() {group=VHS Filter, parameterName=VHS Chromatic Offset Y, sliderMax=50, sortPriority=9}
  n52 = ComponentMask(None=ViewSize()) {r=true}
  n53 = Time()
  n54 = TextureCoordinate()
  n55 = ComponentMask(None=Multiply(
    A=Add(A=@n53, B=@n54)
    B=ScalarParameter() {group=VHS Filter, parameterName=VHS Lines Speed Y, sliderMax=0.44, sliderMin=-0.44, sortPriority=10}
   )) {r=true}
  n56 = Fmod(A=Floor(None=Multiply(A=@n52, B=@n55)), B=Constant() {desc=EveryOtherLine})
  n57 = AppendVector(A=Multiply(A=@n50, B=Multiply(A=@n51, B=@n56)) {constB=5}, B=Constant())
  n58 = ComponentMask(None=SceneTexture(UVs=Add(A=@n49, B=@n57) {desc=Add Tape Wear}) {sceneTextureId=PPI_PostProcessInput0}.Color) {b=true, g=true, r=true}
  n59 = OneMinus(None=@n48)
  n60 = NamedRerouteDeclaration() {name=Animated LinesY_1, nodeColor=<rgba 1 0.0234 0 1>}
  n61 = AppendVector(A=Multiply(A=@n50) {constB=-1}, B=Constant())
  n62 = ComponentMask(None=SceneTexture(UVs=Add(A=@n60, B=@n61) {desc=Add Tape Wear}) {sceneTextureId=PPI_PostProcessInput0}.Color) {b=true, g=true, r=true}
  Max(A=@n29, B=Add(A=Multiply(A=@n48, B=@n58), B=Multiply(A=@n59, B=@n62)))
  NamedRerouteDeclaration(None=@n12) {name=G Tracking Noise, nodeColor=<rgba 1 0.6328 0 1>}
  n63 = Constant2Vector() {g=0.96, r=0.9}
  n64 = Time()
  n65 = Multiply(A=Time(), B=Constant2Vector() {g=0.57, r=0.67})
  n66 = Multiply(A=ComponentMask(None=MaterialFunctionCall(Vec2=@n65) {materialFunction=&MF_Hash}.Vec2) {r=true}) {constB=0.15}
  n67 = ScalarParameter() {defaultValue=-0.33, group=VHS Filter, parameterName=VHS Tracking Line Direction / Speed, sliderMax=1, sliderMin=-1, sortPriority=34}
  n68 = Subtract(A=Multiply(A=ComponentMask(None=@n9) {g=true}) {constB=0.33}, B=Multiply(A=Add(A=@n64, B=@n66), B=@n67) {constB=-0.2})
  NamedRerouteDeclaration(None=SmoothStep(Min=@n63.R, Max=@n63.R, Value=Sine(None=@n68) {period=0.28})) {name=TrackingPhase, nodeColor=<rgba 0 1 0.6016 1>}
  n69 = Constant2Vector() {g=0.7, r=0.3}
  NamedRerouteDeclaration(None=SmoothStep(Min=@n69.R, Max=@n69.R, Value=ComponentMask(None=@n5.Result) {r=true})) {name=R Tracking Noise, nodeColor=<rgba 0 1 0.625 1>}
  NamedRerouteDeclaration(None=@n41) {name=DEBUG Preview Noise Lerp, nodeColor=<rgba 1 0 0.375 1>}
  NamedRerouteUsage()
 ]
}
