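# Post-process material that lays a vignette over the scene, filled with either a color or a tiling texture.
# MF_VignetteMask, remapped by inverted In and Out intensity parameters and saturated, blends the fill with the scene.
# A static switch picks the fill: center-scaled texture or VignetteColor, each using only its red channel, so grayscale.
# Master FX Control blends from the untouched scene to the effect and defaults to off.
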Material /Game/Creative/PostProcess/Materials/M_PP_VignetteMaskedColored
graph {
 properties {materialDomain=MD_PostProcess, pixelDepthOffsetMode=PDOM_Legacy, usedShadingModels=Default Lit}
 nodes [
  n1 = Constant2Vector() {g=0.5, r=0.5}
  n2 = Subtract(A=TextureCoordinate(), B=@n1)
  n3 = AppendVector(
   A=ScalarParameter() {defaultValue=1, group=Use Texture, parameterName=U Tiling, sortPriority=2}
   B=ScalarParameter() {defaultValue=1, group=Use Texture, parameterName=V Tiling, sortPriority=3}
  )
  n4 = TextureSampleParameter2D(UVs=Add(A=Multiply(A=@n2, B=@n3), B=@n1)) {
   group=Use Texture
   parameterName=Texture Mask
   samplerType=SAMPLERTYPE_LinearColor
   sortPriority=1
   texture=/Game/Global/Textures/Default/Tileable_Patterns/Cloud_Diffuse
  }
  n5 = MaterialFunctionCall(X=@n4.R, Y=@n4.R, Z=@n4.R, A=@n4.R) {materialFunction=&MakeFloat4}
  n6 = ScalarParameter() {group=Vignette, parameterName=Vignette In Intensity, sliderMax=1, sortPriority=2}
  n7 = ScalarParameter() {defaultValue=0.666, group=Vignette, parameterName=Vignette Out Intensity, sliderMax=1, sortPriority=1}
  n8 = Saturate(None=MaterialFunctionCall(
    Input=MaterialFunctionCall() {materialFunction=/Game/Creative/PostProcess/MaterialFunctions/MF_VignetteMask}.VignetteMask
    Input Low=Constant()
    Input High=Constant() {r=1}
    Target Low=OneMinus(None=@n6)
    Target High=OneMinus(None=@n7)
   ) {materialFunction=&RemapValueRange}.Result)
  n9 = SceneTexture() {sceneTextureId=PPI_PostProcessInput0}
  n10 = VectorParameter() {group=Vignette, parameterName=VignetteColor, sortPriority=0, defaultValue=<rgba 0 0 0 1>}
  n11 = MaterialFunctionCall(X=@n10.R, Y=@n10.R, Z=@n10.R, A=@n10.R) {materialFunction=&MakeFloat4}
  LinearInterpolate(
   A=SceneTexture() {sceneTextureId=PPI_PostProcessInput0}.Color
   B=StaticSwitchParameter(True=LinearInterpolate(A=@n5.Result, B=@n9.Color, Alpha=@n8), False=LinearInterpolate(A=@n11.Result, B=@n9.Color, Alpha=@n8)) {group=Use Texture, parameterName=UseTextureMaskForVingette, sortPriority=0}
   Alpha=ScalarParameter() {group=Vignette, parameterName=Master FX Control, sortPriority=0}
  )
 ]
}
